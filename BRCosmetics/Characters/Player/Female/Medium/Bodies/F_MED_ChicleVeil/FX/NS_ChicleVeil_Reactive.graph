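# Reactive Niagara effect for the ChicleVeil female body cosmetic: a short sprite burst plus goo-style ribbons attached to arms, hands and legs.
# Burst sprites spawn around six indexed bone/socket points, drift upward with curl noise and dim by time of day.
# Per-limb ribbon emitters re-sample the skeletal mesh, then GooDynamics modules pull points toward goals along wind-blended, upward-bending axes.
# User parameters drive ribbon size and wind; ribbons fade out over their lifetime.
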NiagaraSystem /BRCosmetics/Characters/Player/Female/Medium/Bodies/F_MED_ChicleVeil/FX/NS_ChicleVeil_Reactive
dependencies {
 usedDataInterfaces [
  &NiagaraDataInterfaceArrayInt32
  &NiagaraDataInterfaceArrayFloat3
  &NiagaraDataInterfaceCurve
  &NiagaraDataInterfaceArrayFloat
 ]
 usedDynamicInputs [&MakeFloatFromVector]
 usedModules [
  &self:SystemSpawnScript.SystemScriptSource.SystemScriptGraph.NiagaraNodeAssignment_0.SetVariables_#1
  &SystemState
 ]
}
emitters {
 Burst {
  emitterUpdateScript [
   &EmitterState()
   &SpawnBurst_Instantaneous(Spawn Count=30, Use Spawn Probability=false, UseLoopCountLimit=false)
   &SpawnRate(
    SpawnRate=&FloatFromCurve(
     FloatCurve=DataInterface() {propertyValues {Curve {keys [{interpMode=RCIM_Cubic, tangentMode=RCTM_User, time=0, value=1}, {interpMode=RCIM_Cubic, tangentMode=RCTM_User, time=0.35, value=0}]}}}
     CurveIndex=$Emitter.NormalizedLoopAge
     Scale Curve=35
    )
    Use Spawn Probability=false
   )
  ]
  particleSpawnScript [
   &FN_SetSpawnGroup(Number of Spawn Groups=6)
   &InitializeParticle(
    Color Mode=::NewEnumerator0
    Lifetime Max=0.5
    Lifetime Min=0.3
    Lifetime Mode=::NewEnumerator1
    Sprite Size Max=<6 12>
    Sprite Size Min=<3 7>
    Sprite Size Mode=::NewEnumerator2
    UsePositionOffset=false
    Write Material Random=true
   )
   /SetVariables(Particles.SpriteSize=&Multiply_Vector2D_ByFloat(
     Float=&SelectFloatFromArray(Direct Array Index=$Particles.SpawnGroup, Float Selection Array=DataInterface() {propertyValues {FloatData [1,1,1,1,2,2]}})
     Vector2D=$Particles.SpriteSize
    )) {isSetParameters=true}
   &SkeletalMeshLocation(
    Bone / Socket Index=$Particles.SpawnGroup
    Bone Sampling Mode=::NewEnumerator4
    MaskSpawnGroup=false
    Sampled Position Offset=&SelectVectorFromArray(
     Direct Array Index=$Particles.SpawnGroup
     Vector Selection Array=DataInterface() {propertyValues {FloatData [<5 0 3>, <-5 0 -3>, <-3 -2 0>, <3 2 0>, <-20 0 -5>, <20 0 5>]}}
    )
    Skeletal Mesh=$System.Skeletal Mesh
   )
   &ShapeLocation(
    MaskSpawnGroup=false
    Non Uniform Scale=&SelectVectorFromArray(
     Direct Array Index=$Particles.SpawnGroup
     Vector Selection Array=DataInterface() {propertyValues {FloatData [<1 1 0.1>, <1 1 1>, <1 1 1>, <1 1 1>, <0.1 1 1>, <0.1 1 1>]}}
    )
    Override Seed=false
    Sphere Radius=&SelectFloatFromArray(Direct Array Index=$Particles.SpawnGroup, Float Selection Array=DataInterface() {propertyValues {FloatData [3,3,3,3,7,7]}})
    UseOwnerScale=true
   )
   &AddVelocity(Velocity=<0 0 15>)
   &AddVelocity(
    Module.Use Falloff Distance=false
    Origin Offset=&SelectVectorFromArray(
     Direct Array Index=$Particles.SpawnGroup
     Vector Selection Array=DataInterface() {propertyValues {FloatData [<0 0 0>, <0 0 0>, <0 0 0>, <0 0 0>, <0 0 -25>, <0 0 -25>]}}
    )
    Override Seed=false
    Velocity Mode=::NewEnumerator1
    Velocity Origin=$Particles.SkeletalMeshLocation.SampledPosition
   ) {module=AddVelocity001}
   /SetVariables(Particles.Velocity=&Multiply_VectorByFloat(
     Vector=$Particles.Velocity
     Float=&SelectFloatFromArray(Direct Array Index=$Particles.SpawnGroup, Float Selection Array=DataInterface() {propertyValues {FloatData [1,1,1,1,2,2]}})
    )) {isSetParameters=true}
  ]
  particleUpdateScript [
   &ParticleState()
   &ScaleSpriteSize(Uniform Curve Sprite Scale=DataInterface() {propertyValues {Curve {keys [{interpMode=RCIM_Cubic, tangentMode=RCTM_User, time=0.3, value=1}, {tangentMode=RCTM_User, time=1, value=0}]}}})
   &DynamicMaterialParameters(
    Index 0 Param 1=&RandomRangeFloat(Minimum=0, Maximum=1)
    Module.Param0WriteEnabled=true
    Module.Param1WriteEnabled=true
    Module.Param2WriteEnabled=true
    Module.Param3WriteEnabled=true
   ) {module=DynamicMaterialParameters001}
   &GravityForce(Gravity=<0 0 15>)
   &Drag(Use Linear Drag=true, Use Rotational Drag=true)
   &CurlNoiseForce(Noise Frequency=250, Noise Strength=50, Pan Noise=true, Pan Noise Field=<3 0 0>)
   &SolveForcesAndVelocity(Clamp Velocity=false, Limit Acceleration=false)
   &ScaleColor(
    Scale RGB=&Multiply_VectorByFloat(
     Vector=<1 1 1>
     Float=&FNTimeOfDayScalar(Night Scale=0.5, Morning Scale=0.75, Evening Scale=0.7)
    )
    ScaleA=false
    ScaleRGB=true
   )
  ]
  rendererClasses [&NiagaraSpriteRendererProperties]
  renderers [{rendererClass=&NiagaraSpriteRendererProperties}]
 }
 Left_Arm {
  emitterSpawnScript [/Game/Effects/Niagara/Modules/GooDynamics/FN_GooGlobalProperties(
    EndBoneIndexArray=DataInterface() {propertyValues {GpuSyncMode=SyncCpuToGpu, MaxElements=0, IntData [0]}}
    EndBoneOffsetArray=DataInterface() {propertyValues {GpuSyncMode=SyncCpuToGpu, MaxElements=0, FloatData [<0 0 0>]}}
    MaxDistance:NiagaraFloat=$User.ArmRibbonLength
    MaxDistanceCompressed=10
    MaxDistance_RandFactorHigh=1
    MaxDistance_RandFactorLow=1
    RibbonLength:NiagaraFloat=$User.ArmRibbonLength
    RibbonLengthRandHigh=1
    RibbonLengthRandLow=1
    StartBoneIndexArray=DataInterface() {propertyValues {GpuSyncMode=SyncCpuToGpu, MaxElements=0, IntData [0,1]}}
    StartBoneOffsetArray=DataInterface() {propertyValues {GpuSyncMode=SyncCpuToGpu, MaxElements=0, FloatData [<4 -2 2>, <3 0 0>]}}
   )]
  emitterUpdateScript [
   &EmitterState()
   &SpawnBurst_Instantaneous(
    Spawn Count=&Multiply_Int(A=$Emitter.ParticlesPerRibbon, B=$Emitter.NumberOfRibbons)
    Use Spawn Probability=true
    UseLoopCountLimit=false
   )
  ]
  particleSpawnScript [
   /SetVariables(Particles.SpawnGroup=/Niagara/DynamicInputs/Divide/Divide_Int(A:NiagaraInt32=&ReturnExecIndex(), B:NiagaraInt32=$Emitter.ParticlesPerRibbon)) {isSetParameters=true, module=SetVariables_#2}
   /SetVariables(
    Particles.RibbonID=/Niagara/DynamicInputs/Meshes/MakeRibbonIDFromSpawnGroup()
    Particles.RibbonLinkOrder:NiagaraFloat=&Divide_Float(
     A=&MakeFloatFromInt(Int=&ModuloInt(A=&ReturnExecIndex(), B=$Emitter.ParticlesPerRibbon))
     B=&Subtract_Float(A=&MakeFloatFromInt(Int=$Emitter.ParticlesPerRibbon), B=1)
    )
   ) {isSetParameters=true, module=SetVariables_#3}
   /SetVariables(Particles.RibbonRandom:NiagaraFloat=&RandomRangeFloat(
     Minimum=0
     Maximum=1
     Randomness Mode=::NewEnumerator1
     Random Seed=&Add_Integer(A=$Engine.System.RandomSeed, B=$Particles.SpawnGroup)
     Fixed Random Seed=true
    )) {isSetParameters=true, module=SetVariables_#4}
   &InitializeRibbon(
    Lifetime=&Lerp_Float(A=3, B=5, Alpha=$Particles.RibbonRandom)
    Ribbon Width=&FloatFromCurve(
     FloatCurve=DataInterface() {propertyValues {Curve {keys [<at 0 1>,<at 0.5 0.5>,<at 1 1>]}}}
     CurveIndex=$Particles.RibbonLinkOrder
     Scale Curve=$User.ArmRibbonWidth
    )
    UsePositionOffset=false
    Write Material Random=true
   )
   /SetVariables(Particles.MaterialRandom:NiagaraFloat=$Particles.RibbonRandom) {isSetParameters=true, module=SetVariables_#5}
   &SkeletalMeshLocation(
    Bone / Socket Index=&SelectIntFromArray(Direct Array Index=$Particles.SpawnGroup, Direct Array Mode=::NewEnumerator1, Int Selection Array=$Emitter.StartBoneIndexArray)
    Bone Sampling Mode=::NewEnumerator4
    MaskSpawnGroup=false
    Orientation Sampling=::NewEnumerator1
    Position Sampling=::NewEnumerator2
    Sampled Position Offset=&SelectVectorFromArray(Direct Array Index=$Particles.SpawnGroup, Direct Array Mode=::NewEnumerator1, Vector Selection Array=$Emitter.StartBoneOffsetArray)
    Skeletal Mesh=$User.SkeletalMesh
    Use Sampled Mesh Attribute as Ribbon ID=false
   )
   /Game/Effects/Niagara/Modules/GooDynamics/FN_GooCalculateGoalPosition(
    Axis:Vector3f=$Emitter.RibbonAxis
    Scale:Vector3f=&Multiply_VectorByFloat(
     Vector=&Multiply_VectorByFloat(
      Vector=&VectorFromFloat(Value=&Multiply_Float(A=$Emitter.RibbonLength, B=0))
      Float=&Lerp_Float(A=$Emitter.RibbonLengthRandMin, B=$Emitter.RibbonLengthRandMax, Alpha=$Particles.RibbonRandom)
     )
     Float=$Particles.RibbonLinkOrder
    )
    StartPosition:Vector3f=$Particles.SkeletalMeshLocation.SampledPosition
   ) {module=FN_CalculateGoalPosition}
   /SetVariables(Particles.Position=$Particles.GoalPosition) {isSetParameters=true, module=SetVariables_#6}
   /SetVariables(Particles.PreviousRotation:Quat4f=$Engine.Owner.Rotation) {isSetParameters=true, module=SetVariables_#7}
   &DynamicMaterialParameters(
    Index 0 Param 1=&Multiply_Float(A=$Particles.RibbonRandom, B=10)
    Module.Param0WriteEnabled=true
    Module.Param1WriteEnabled=true
    Module.Param2WriteEnabled=true
    Module.Param3WriteEnabled=true
   ) {enabled=false}
   /SetVariables(Particles.SpriteSize=<10 10>) {isSetParameters=true, module=SetVariables_#8}
   &DynamicMaterialParameters(
    Index 0 Param 1=&SelectFloatFromArray(Direct Array Index=&SpawnGroup(), Float Selection Array=DataInterface() {propertyValues {FloatData [0.3,0.78]}})
    Module.Index0Param0Write=true
    Module.Index0Param1Write=true
    Module.Index0Param2Write=true
    Module.Index0Param3Write=true
   ) {module=DynamicMaterialParameters002}
  ]
  particleUpdateScript [
   &ParticleState()
   &SkeletalMeshLocation(
    Bone / Socket Index=&SelectIntFromArray(Direct Array Index=$Particles.SpawnGroup, Direct Array Mode=::NewEnumerator1, Int Selection Array=$Emitter.StartBoneIndexArray)
    Bone Sampling Mode=::NewEnumerator4
    MaskSpawnGroup=false
    Orientation Sampling=::NewEnumerator1
    Position Sampling=::NewEnumerator2
    Sampled Position Offset=&SelectVectorFromArray(Direct Array Index=$Particles.SpawnGroup, Direct Array Mode=::NewEnumerator1, Vector Selection Array=$Emitter.StartBoneOffsetArray)
    Skeletal Mesh=$User.SkeletalMesh
    Use Sampled Mesh Attribute as Ribbon ID=false
   ) {module=SkeletalMeshLocation001}
   &SolveForcesAndVelocity(Clamp Velocity=false, Limit Acceleration=false)
   &FN_GooCalculateGoalPosition(
    Axis:Vector3f=/Niagara/DynamicInputs/Math/SlerpVector(
     Vector A:Vector3f=&SlerpVector(
      Vector A:Vector3f=<1 0 0>
      Vector B:Vector3f=&TransformVector(Vector=$User.WorldSpaceWindVector, Source Space=::World, Destination Space=::Local)
      Alpha:NiagaraFloat=$User.WorldSpaceWindVectorFactor
      Scale:NiagaraFloat=1
     )
     Vector B:Vector3f=<0 0 1>
     Alpha:NiagaraFloat=&FloatFromCurve(
      FloatCurve=DataInterface() {propertyValues {Curve {keys [
          {arriveTangent=1.3547, interpMode=RCIM_Cubic, leaveTangent=1.3547, tangentMode=RCTM_User, time=0, value=0.3}
          {arriveTangent=0.7, interpMode=RCIM_Cubic, leaveTangent=0.7, time=0.4, value=0.9}
          <at 1 1 RCIM_Cubic>
         ]}}}
      CurveIndex=$Particles.RibbonLinkOrder
      Scale Curve=0.7
     )
     Scale:NiagaraFloat=1
    )
    CustomOrientation:Quat4f=$Engine.Owner.Rotation
    Scale:Vector3f=&Multiply_VectorByFloat(
     Vector=&Multiply_VectorByFloat(
      Vector=&VectorFromFloat(Value=&Multiply_Float(
        A=$Emitter.RibbonLength
        B=&FloatFromCurve(FloatCurve=DataInterface() {propertyValues {Curve {keys [
             {arriveTangent=17.1487, interpMode=RCIM_Cubic, leaveTangent=17.1487, tangentMode=RCTM_User, time=0, value=1}
             {interpMode=RCIM_Cubic, tangentMode=RCTM_User, time=0.2, value=1.5}
             <at 0.25 1 RCIM_Cubic>
             {arriveTangent=-3.4902, interpMode=RCIM_Cubic, leaveTangent=-3.4902, tangentMode=RCTM_User, time=1, value=0}
            ]}}})
       ))
      Float=&Lerp_Float(A=$Emitter.RibbonLengthRandMin, B=$Emitter.RibbonLengthRandMax, Alpha=$Particles.RibbonRandom)
     )
     Float=$Particles.RibbonLinkOrder
    )
    StartPosition:Vector3f=$Particles.SkeletalMeshLocation001.SampledPosition
    Use Quaternion Rotation=true
   ) {module=FN_CalculateGoalPosition001}
   /Game/Effects/Niagara/Modules/GooDynamics/FN_GooApproachGoalPosition(
    AttractionForce:NiagaraFloat=&Lerp_Float(A=$Emitter.ResolveSpeedLow, B=$Emitter.ResolveSpeedHigh, Alpha=$Particles.RibbonRandom)
    GoalPosition:NiagaraPosition=$Particles.GoalPosition
    MaximumDistance:NiagaraFloat=&Multiply_Float(
     A=&FloatFromCurve(
      FloatCurve=DataInterface() {propertyValues {Curve {keys [
          {arriveTangent=4.1785, interpMode=RCIM_Cubic, time=0, value=0}
          {arriveTangent=2.3315, interpMode=RCIM_Cubic, leaveTangent=2.3315, tangentMode=RCTM_User, time=1, value=1}
         ]}}}
      CurveIndex=$Particles.RibbonLinkOrder
      Scale Curve=&Multiply_Float(
       A=$Emitter.MaxDistance
       B=&FloatFromCurve(FloatCurve=DataInterface() {propertyValues {Curve {keys [<at 0.75 1>,<at 1 0>]}}})
      )
     )
     B=&Lerp_Float(A=$Emitter.MaxDistance_RandFactorLow, B=$Emitter.MaxDistance_RandFactorHigh, Alpha=$Particles.RibbonRandom)
    )
    RibbonAxis:Vector3f=$Particles.RibbonAxis
   )
   /SetVariables(Particles.PreviousRotation:Quat4f=$Engine.Owner.Rotation) {isSetParameters=true, module=SetVariables_#9}
   &ScaleColor(
    Scale Alpha=&FloatFromCurve(FloatCurve=DataInterface() {propertyValues {Curve {keys [{interpMode=RCIM_Cubic, tangentMode=RCTM_User, time=0.6, value=1}, {interpMode=RCIM_Cubic, tangentMode=RCTM_User, time=1, value=0}]}}})
    ScaleA=true
    ScaleRGB=false
   )
  ]
  rendererClasses [&NiagaraRibbonRendererProperties]
  renderers [{rendererClass=&NiagaraRibbonRendererProperties}]
 }
 Left_Hand {
  emitterSpawnScript [&FN_GooGlobalProperties(
    EndBoneIndexArray=DataInterface() {propertyValues {GpuSyncMode=SyncCpuToGpu, MaxElements=0, IntData [0]}}
    EndBoneOffsetArray=DataInterface() {propertyValues {GpuSyncMode=SyncCpuToGpu, MaxElements=0, FloatData [<0 0 0>]}}
    MaxDistance:NiagaraFloat=$User.ArmRibbonLength
    MaxDistanceCompressed=10
    MaxDistance_RandFactorHigh=1
    MaxDistance_RandFactorLow=1
    RibbonLength:NiagaraFloat=$User.ArmRibbonLength
    RibbonLengthRandHigh=1
    RibbonLengthRandLow=1
    StartBoneIndexArray=DataInterface() {propertyValues {GpuSyncMode=SyncCpuToGpu, MaxElements=0, IntData [2]}}
    StartBoneOffsetArray=DataInterface() {propertyValues {GpuSyncMode=SyncCpuToGpu, MaxElements=0, FloatData [<3 0 0>]}}
   )]
  emitterUpdateScript [
   &EmitterState()
   &SpawnBurst_Instantaneous(
    Spawn Count=&Multiply_Int(A=$Emitter.ParticlesPerRibbon, B=$Emitter.NumberOfRibbons)
    Use Spawn Probability=false
    UseLoopCountLimit=false
   )
  ]
  particleSpawnScript [
   /SetVariables(Particles.SpawnGroup=&Divide_Int(A:NiagaraInt32=&ReturnExecIndex(), B:NiagaraInt32=$Emitter.ParticlesPerRibbon)) {isSetParameters=true, module=SetVariables_#2}
   /SetVariables(
    Particles.RibbonID=&MakeRibbonIDFromSpawnGroup()
    Particles.RibbonLinkOrder:NiagaraFloat=&Divide_Float(
     A=&MakeFloatFromInt(Int=&ModuloInt(A=&ReturnExecIndex(), B=$Emitter.ParticlesPerRibbon))
     B=&Subtract_Float(A=&MakeFloatFromInt(Int=$Emitter.ParticlesPerRibbon), B=1)
    )
   ) {isSetParameters=true, module=SetVariables_#3}
   /SetVariables(Particles.RibbonRandom:NiagaraFloat=&RandomRangeFloat(
     Minimum=0
     Maximum=1
     Randomness Mode=::NewEnumerator1
     Random Seed=&Add_Integer(A=$Engine.System.RandomSeed, B=$Particles.SpawnGroup)
     Fixed Random Seed=true
    )) {isSetParameters=true, module=SetVariables_#4}
   &InitializeRibbon(
    Lifetime=&Lerp_Float(A=3, B=5, Alpha=$Particles.RibbonRandom)
    Ribbon Width=&FloatFromCurve(
     FloatCurve=DataInterface() {propertyValues {Curve {keys [<at 0 1>,<at 0.5 0.5>,<at 1 1>]}}}
     CurveIndex=$Particles.RibbonLinkOrder
     Scale Curve=15
    )
    UsePositionOffset=false
    Write Material Random=true
   )
   /SetVariables(Particles.MaterialRandom:NiagaraFloat=$Particles.RibbonRandom) {isSetParameters=true, module=SetVariables_#5}
   &SkeletalMeshLocation(
    Bone / Socket Index=&SelectIntFromArray(Direct Array Index=$Particles.SpawnGroup, Direct Array Mode=::NewEnumerator1, Int Selection Array=$Emitter.StartBoneIndexArray)
    Bone Sampling Mode=::NewEnumerator4
    MaskSpawnGroup=false
    Orientation Sampling=::NewEnumerator1
    Position Sampling=::NewEnumerator2
    Sampled Position Offset=&SelectVectorFromArray(Direct Array Index=$Particles.SpawnGroup, Direct Array Mode=::NewEnumerator1, Vector Selection Array=$Emitter.StartBoneOffsetArray)
    Skeletal Mesh=$User.SkeletalMesh
    Use Sampled Mesh Attribute as Ribbon ID=false
   )
   &FN_GooCalculateGoalPosition(
    Axis:Vector3f=$Emitter.RibbonAxis
    Scale:Vector3f=&Multiply_VectorByFloat(
     Vector=&Multiply_VectorByFloat(
      Vector=&VectorFromFloat(Value=&Multiply_Float(A=$Emitter.RibbonLength, B=0))
      Float=&Lerp_Float(A=$Emitter.RibbonLengthRandMin, B=$Emitter.RibbonLengthRandMax, Alpha=$Particles.RibbonRandom)
     )
     Float=$Particles.RibbonLinkOrder
    )
    StartPosition:Vector3f=$Particles.SkeletalMeshLocation.SampledPosition
   ) {module=FN_CalculateGoalPosition}
   /SetVariables(Particles.Position=$Particles.GoalPosition) {isSetParameters=true, module=SetVariables_#6}
   /SetVariables(Particles.PreviousRotation:Quat4f=$Engine.Owner.Rotation) {isSetParameters=true, module=SetVariables_#7}
   &DynamicMaterialParameters(
    Index 0 Param 1=&Multiply_Float(A=$Particles.RibbonRandom, B=10)
    Module.Param0WriteEnabled=true
    Module.Param1WriteEnabled=true
    Module.Param2WriteEnabled=true
    Module.Param3WriteEnabled=true
   ) {enabled=false}
   /SetVariables(Particles.SpriteSize=<10 10>) {isSetParameters=true, module=SetVariables_#8}
  ]
  particleUpdateScript [
   &ParticleState()
   &SkeletalMeshLocation(
    Bone / Socket Index=&SelectIntFromArray(Direct Array Index=$Particles.SpawnGroup, Direct Array Mode=::NewEnumerator1, Int Selection Array=$Emitter.StartBoneIndexArray)
    Bone Sampling Mode=::NewEnumerator4
    MaskSpawnGroup=false
    Orientation Sampling=::NewEnumerator1
    Position Sampling=::NewEnumerator2
    Sampled Position Offset=&SelectVectorFromArray(Direct Array Index=$Particles.SpawnGroup, Direct Array Mode=::NewEnumerator1, Vector Selection Array=$Emitter.StartBoneOffsetArray)
    Skeletal Mesh=$User.SkeletalMesh
    Use Sampled Mesh Attribute as Ribbon ID=false
   ) {module=SkeletalMeshLocation001}
   &SolveForcesAndVelocity(Clamp Velocity=false, Limit Acceleration=false)
   &FN_GooCalculateGoalPosition(
    Axis:Vector3f=&SlerpVector(
     Vector A:Vector3f=&SlerpVector(
      Vector A:Vector3f=<1 0 0>
      Vector B:Vector3f=&TransformVector(Vector=$User.WorldSpaceWindVector, Source Space=::World, Destination Space=::Local)
      Alpha:NiagaraFloat=$User.WorldSpaceWindVectorFactor
      Scale:NiagaraFloat=1
     )
     Vector B:Vector3f=<0 0 1>
     Alpha:NiagaraFloat=&FloatFromCurve(
      FloatCurve=DataInterface() {propertyValues {Curve {keys [
          {arriveTangent=1.3547, interpMode=RCIM_Cubic, leaveTangent=1.3547, tangentMode=RCTM_User, time=0, value=0.3}
          {arriveTangent=0.7, interpMode=RCIM_Cubic, leaveTangent=0.7, time=0.4, value=0.9}
          <at 1 1 RCIM_Cubic>
         ]}}}
      CurveIndex=$Particles.RibbonLinkOrder
      Scale Curve=0.7
     )
     Scale:NiagaraFloat=1
    )
    CustomOrientation:Quat4f=$Engine.Owner.Rotation
    Scale:Vector3f=&Multiply_VectorByFloat(
     Vector=&Multiply_VectorByFloat(
      Vector=&VectorFromFloat(Value=&Multiply_Float(
        A=$Emitter.RibbonLength
        B=&FloatFromCurve(FloatCurve=DataInterface() {propertyValues {Curve {keys [
             {arriveTangent=17.1487, interpMode=RCIM_Cubic, leaveTangent=17.1487, tangentMode=RCTM_User, time=0, value=1}
             {interpMode=RCIM_Cubic, tangentMode=RCTM_User, time=0.2, value=1.5}
             <at 0.25 1 RCIM_Cubic>
             {arriveTangent=-3.4902, interpMode=RCIM_Cubic, leaveTangent=-3.4902, tangentMode=RCTM_User, time=1, value=0}
            ]}}})
       ))
      Float=&Lerp_Float(A=$Emitter.RibbonLengthRandMin, B=$Emitter.RibbonLengthRandMax, Alpha=$Particles.RibbonRandom)
     )
     Float=$Particles.RibbonLinkOrder
    )
    StartPosition:Vector3f=$Particles.SkeletalMeshLocation001.SampledPosition
    Use Quaternion Rotation=true
   ) {module=FN_CalculateGoalPosition001}
   &FN_GooApproachGoalPosition(
    AttractionForce:NiagaraFloat=&Lerp_Float(A=$Emitter.ResolveSpeedLow, B=$Emitter.ResolveSpeedHigh, Alpha=$Particles.RibbonRandom)
    GoalPosition:NiagaraPosition=$Particles.GoalPosition
    MaximumDistance:NiagaraFloat=&Multiply_Float(
     A=&FloatFromCurve(
      FloatCurve=DataInterface() {propertyValues {Curve {keys [
          {arriveTangent=4.1785, interpMode=RCIM_Cubic, time=0, value=0}
          {arriveTangent=2.3315, interpMode=RCIM_Cubic, leaveTangent=2.3315, tangentMode=RCTM_User, time=1, value=1}
         ]}}}
      CurveIndex=$Particles.RibbonLinkOrder
      Scale Curve=&Multiply_Float(
       A=$Emitter.MaxDistance
       B=&FloatFromCurve(FloatCurve=DataInterface() {propertyValues {Curve {keys [<at 0.75 1>,<at 1 0>]}}})
      )
     )
     B=&Lerp_Float(A=$Emitter.MaxDistance_RandFactorLow, B=$Emitter.MaxDistance_RandFactorHigh, Alpha=$Particles.RibbonRandom)
    )
    RibbonAxis:Vector3f=$Particles.RibbonAxis
   )
   /SetVariables(Particles.PreviousRotation:Quat4f=$Engine.Owner.Rotation) {isSetParameters=true, module=SetVariables_#9}
   &ScaleColor(
    Scale Alpha=&FloatFromCurve(FloatCurve=DataInterface() {propertyValues {Curve {keys [{interpMode=RCIM_Cubic, tangentMode=RCTM_User, time=0.6, value=1}, {interpMode=RCIM_Cubic, tangentMode=RCTM_User, time=1, value=0}]}}})
    ScaleA=true
    ScaleRGB=false
   )
  ]
  rendererClasses [&NiagaraRibbonRendererProperties]
  renderers [{rendererClass=&NiagaraRibbonRendererProperties}]
 }
 Left_Leg {
  emitterSpawnScript [&FN_GooGlobalProperties(
    EndBoneIndexArray=DataInterface() {propertyValues {GpuSyncMode=SyncCpuToGpu, MaxElements=0, IntData [0]}}
    EndBoneOffsetArray=DataInterface() {propertyValues {GpuSyncMode=SyncCpuToGpu, MaxElements=0, FloatData [<0 0 0>]}}
    MaxDistance:NiagaraFloat=$User.LegRibbonLength
    MaxDistanceCompressed=10
    MaxDistance_RandFactorHigh=1
    MaxDistance_RandFactorLow=1
    NumberOfRibbons=2
    ParticlesPerRibbon=3
    ResolveSpeedHigh=3
    RibbonLength:NiagaraFloat=$User.LegRibbonLength
    RibbonLengthRandHigh=1
    RibbonLengthRandLow=1
    StartBoneIndexArray=DataInterface() {propertyValues {GpuSyncMode=SyncCpuToGpu, MaxElements=0, IntData [6,7]}}
    StartBoneOffsetArray=DataInterface() {propertyValues {GpuSyncMode=SyncCpuToGpu, MaxElements=0, FloatData [<-20 0 -5>, <15 0 -6>]}}
   )]
  emitterUpdateScript [
   &EmitterState()
   &SpawnBurst_Instantaneous(
    Spawn Count=&Multiply_Int(A=$Emitter.ParticlesPerRibbon, B=$Emitter.NumberOfRibbons)
    Use Spawn Probability=false
    UseLoopCountLimit=false
   )
  ]
  particleSpawnScript [
   /SetVariables(Particles.SpawnGroup=&Divide_Int(A:NiagaraInt32=&ReturnExecIndex(), B:NiagaraInt32=$Emitter.ParticlesPerRibbon)) {isSetParameters=true, module=SetVariables_#2}
   /SetVariables(
    Particles.RibbonID=&MakeRibbonIDFromSpawnGroup()
    Particles.RibbonLinkOrder:NiagaraFloat=&Divide_Float(
     A=&MakeFloatFromInt(Int=&ModuloInt(A=&ReturnExecIndex(), B=$Emitter.ParticlesPerRibbon))
     B=&Subtract_Float(A=&MakeFloatFromInt(Int=$Emitter.ParticlesPerRibbon), B=1)
    )
   ) {isSetParameters=true, module=SetVariables_#3}
   /SetVariables(Particles.RibbonRandom:NiagaraFloat=&RandomRangeFloat(
     Minimum=0
     Maximum=1
     Randomness Mode=::NewEnumerator1
     Random Seed=&Add_Integer(A=$Engine.System.RandomSeed, B=$Particles.SpawnGroup)
     Fixed Random Seed=true
    )) {isSetParameters=true, module=SetVariables_#4}
   &InitializeRibbon(
    Lifetime=&Lerp_Float(A=3, B=5, Alpha=$Particles.RibbonRandom)
    Ribbon Width=$User.LegRibbonWidth
    UsePositionOffset=false
    Write Material Random=true
   )
   /SetVariables(Particles.MaterialRandom:NiagaraFloat=$Particles.RibbonRandom) {isSetParameters=true, module=SetVariables_#5}
   &SkeletalMeshLocation(
    Bone / Socket Index=&SelectIntFromArray(Direct Array Index=$Particles.SpawnGroup, Direct Array Mode=::NewEnumerator1, Int Selection Array=$Emitter.StartBoneIndexArray)
    Bone Sampling Mode=::NewEnumerator4
    MaskSpawnGroup=false
    Orientation Sampling=::NewEnumerator1
    Position Sampling=::NewEnumerator2
    Sampled Position Offset=&SelectVectorFromArray(Direct Array Index=$Particles.SpawnGroup, Direct Array Mode=::NewEnumerator1, Vector Selection Array=$Emitter.StartBoneOffsetArray)
    Skeletal Mesh=$User.SkeletalMesh
    Use Sampled Mesh Attribute as Ribbon ID=false
   )
   &FN_GooCalculateGoalPosition(
    Axis:Vector3f=$Emitter.RibbonAxis
    Scale:Vector3f=&Multiply_VectorByFloat(
     Vector=&Multiply_VectorByFloat(
      Vector=&VectorFromFloat(Value=&Multiply_Float(A=$Emitter.RibbonLength, B=0))
      Float=&Lerp_Float(A=$Emitter.RibbonLengthRandMin, B=$Emitter.RibbonLengthRandMax, Alpha=$Particles.RibbonRandom)
     )
     Float=$Particles.RibbonLinkOrder
    )
    StartPosition:Vector3f=$Particles.SkeletalMeshLocation.SampledPosition
   ) {module=FN_CalculateGoalPosition}
   /SetVariables(Particles.Position=$Particles.GoalPosition) {isSetParameters=true, module=SetVariables_#6}
   /SetVariables(Particles.PreviousRotation:Quat4f=$Engine.Owner.Rotation) {isSetParameters=true, module=SetVariables_#7}
   &DynamicMaterialParameters(
    Index 0 Param 1=&Multiply_Float(A=$Particles.RibbonRandom, B=10)
    Module.Param0WriteEnabled=true
    Module.Param1WriteEnabled=true
    Module.Param2WriteEnabled=true
    Module.Param3WriteEnabled=true
   )
   /SetVariables(Particles.SpriteSize=<10 10>) {isSetParameters=true, module=SetVariables_#8}
   &DynamicMaterialParameters(
    Index 0 Param 1=&SelectFloatFromArray(
     Direct Array Index=&SpawnGroup()
     Float Selection Array=DataInterface() {propertyValues {FloatData [0.25,0.75]}}
    )
    Module.Index0Param0Write=true
    Module.Index0Param1Write=true
    Module.Index0Param2Write=true
    Module.Index0Param3Write=true
   ) {module=DynamicMaterialParameters002}
  ]
  particleUpdateScript [
   &ParticleState()
   &SkeletalMeshLocation(
    Bone / Socket Index=&SelectIntFromArray(Direct Array Index=$Particles.SpawnGroup, Direct Array Mode=::NewEnumerator1, Int Selection Array=$Emitter.StartBoneIndexArray)
    Bone Sampling Mode=::NewEnumerator4
    MaskSpawnGroup=false
    Orientation Sampling=::NewEnumerator1
    Position Sampling=::NewEnumerator2
    Sampled Position Offset=&SelectVectorFromArray(Direct Array Index=$Particles.SpawnGroup, Direct Array Mode=::NewEnumerator1, Vector Selection Array=$Emitter.StartBoneOffsetArray)
    Skeletal Mesh=$User.SkeletalMesh
    Use Sampled Mesh Attribute as Ribbon ID=false
   ) {module=SkeletalMeshLocation001}
   &SolveForcesAndVelocity(Clamp Velocity=false, Limit Acceleration=false)
   &FN_GooCalculateGoalPosition(
    Axis:Vector3f=&SlerpVector(
     Vector A:Vector3f=&SlerpVector(
      Vector A:Vector3f=<1 0 0>
      Vector B:Vector3f=&TransformVector(Vector=$User.WorldSpaceWindVector, Source Space=::World, Destination Space=::Local)
      Alpha:NiagaraFloat=$User.WorldSpaceWindVectorFactor
      Scale:NiagaraFloat=1
     )
     Vector B:Vector3f=<0 0 1>
     Alpha:NiagaraFloat=&FloatFromCurve(
      FloatCurve=DataInterface() {propertyValues {Curve {keys [
          {arriveTangent=1.3547, interpMode=RCIM_Cubic, leaveTangent=1.3547, tangentMode=RCTM_User, time=0, value=0.3}
          {arriveTangent=0.7, interpMode=RCIM_Cubic, leaveTangent=0.7, time=0.4, value=0.75}
          <at 1 1 RCIM_Cubic>
         ]}}}
      CurveIndex=$Particles.RibbonLinkOrder
     )
     Scale:NiagaraFloat=1
    )
    CustomOrientation:Quat4f=$Engine.Owner.Rotation
    Scale:Vector3f=&Multiply_VectorByFloat(
     Vector=&Multiply_VectorByFloat(
      Vector=&VectorFromFloat(Value=&Multiply_Float(
        A=$Emitter.RibbonLength
        B=&FloatFromCurve(FloatCurve=DataInterface() {propertyValues {Curve {keys [
             {arriveTangent=30.27, interpMode=RCIM_Cubic, leaveTangent=30.27, tangentMode=RCTM_User, time=0, value=0}
             {interpMode=RCIM_Cubic, tangentMode=RCTM_User, time=0.15, value=1.5}
             <at 0.2 1 RCIM_Cubic>
             {arriveTangent=-3.4902, interpMode=RCIM_Cubic, leaveTangent=-3.4902, tangentMode=RCTM_User, time=1, value=0}
            ]}}})
       ))
      Float=&Lerp_Float(A=$Emitter.RibbonLengthRandMin, B=$Emitter.RibbonLengthRandMax, Alpha=$Particles.RibbonRandom)
     )
     Float=$Particles.RibbonLinkOrder
    )
    StartPosition:Vector3f=$Particles.SkeletalMeshLocation001.SampledPosition
    Use Quaternion Rotation=true
   ) {module=FN_CalculateGoalPosition001}
   &FN_GooApproachGoalPosition(
    AttractionForce:NiagaraFloat=&Lerp_Float(A=$Emitter.ResolveSpeedLow, B=$Emitter.ResolveSpeedHigh, Alpha=$Particles.RibbonRandom)
    GoalPosition:NiagaraPosition=$Particles.GoalPosition
    MaximumDistance:NiagaraFloat=&Multiply_Float(
     A=&FloatFromCurve(
      FloatCurve=DataInterface() {propertyValues {Curve {keys [
          {arriveTangent=4.1785, interpMode=RCIM_Cubic, time=0, value=0}
          {arriveTangent=2.3315, interpMode=RCIM_Cubic, leaveTangent=2.3315, tangentMode=RCTM_User, time=1, value=1}
         ]}}}
      CurveIndex=$Particles.RibbonLinkOrder
      Scale Curve=&Multiply_Float(
       A=$Emitter.MaxDistance
       B=&FloatFromCurve(FloatCurve=DataInterface() {propertyValues {Curve {keys [<at 0.75 1>,<at 1 0>]}}})
      )
     )
     B=&Lerp_Float(A=$Emitter.MaxDistance_RandFactorLow, B=$Emitter.MaxDistance_RandFactorHigh, Alpha=$Particles.RibbonRandom)
    )
    RibbonAxis:Vector3f=$Particles.RibbonAxis
   )
   /SetVariables(Particles.PreviousRotation:Quat4f=$Engine.Owner.Rotation) {isSetParameters=true, module=SetVariables_#9}
   &ScaleColor(
    Scale Alpha=&FloatFromCurve(FloatCurve=DataInterface() {propertyValues {Curve {keys [{interpMode=RCIM_Cubic, tangentMode=RCTM_User, time=0.6, value=1}, {interpMode=RCIM_Cubic, tangentMode=RCTM_User, time=1, value=0}]}}})
    ScaleA=true
    ScaleRGB=false
   )
  ]
  rendererClasses [&NiagaraRibbonRendererProperties]
  renderers [{rendererClass=&NiagaraRibbonRendererProperties}]
 }
 Right_Arm {
  emitterSpawnScript [&FN_GooGlobalProperties(
    EndBoneIndexArray=DataInterface() {propertyValues {GpuSyncMode=SyncCpuToGpu, MaxElements=0, IntData [0]}}
    EndBoneOffsetArray=DataInterface() {propertyValues {GpuSyncMode=SyncCpuToGpu, MaxElements=0, FloatData [<0 0 0>]}}
    MaxDistance:NiagaraFloat=$User.ArmRibbonLength
    MaxDistanceCompressed=10
    MaxDistance_RandFactorHigh=1
    MaxDistance_RandFactorLow=1
    RibbonLength:NiagaraFloat=$User.ArmRibbonLength
    RibbonLengthRandHigh=1
    RibbonLengthRandLow=1
    StartBoneIndexArray=DataInterface() {propertyValues {GpuSyncMode=SyncCpuToGpu, MaxElements=0, IntData [3,4]}}
    StartBoneOffsetArray=DataInterface() {propertyValues {GpuSyncMode=SyncCpuToGpu, MaxElements=0, FloatData [<-4 2 -2>, <-3 0 0>]}}
   )]
  emitterUpdateScript [
   &EmitterState()
   &SpawnBurst_Instantaneous(
    Spawn Count=&Multiply_Int(A=$Emitter.ParticlesPerRibbon, B=$Emitter.NumberOfRibbons)
    Use Spawn Probability=true
    UseLoopCountLimit=false
   )
  ]
  particleSpawnScript [
   /SetVariables(Particles.SpawnGroup=&Divide_Int(A:NiagaraInt32=&ReturnExecIndex(), B:NiagaraInt32=$Emitter.ParticlesPerRibbon)) {isSetParameters=true, module=SetVariables_#10}
   /SetVariables(
    Particles.RibbonID=&MakeRibbonIDFromSpawnGroup()
    Particles.RibbonLinkOrder:NiagaraFloat=&Divide_Float(
     A=&MakeFloatFromInt(Int=&ModuloInt(A=&ReturnExecIndex(), B=$Emitter.ParticlesPerRibbon))
     B=&Subtract_Float(A=&MakeFloatFromInt(Int=$Emitter.ParticlesPerRibbon), B=1)
    )
   ) {isSetParameters=true, module=SetVariables_#11}
   /SetVariables(Particles.RibbonRandom:NiagaraFloat=&RandomRangeFloat(
     Minimum=0
     Maximum=1
     Randomness Mode=::NewEnumerator1
     Random Seed=&Add_Integer(A=$Engine.System.RandomSeed, B=$Particles.SpawnGroup)
     Fixed Random Seed=true
    )) {isSetParameters=true, module=SetVariables_#4}
   &InitializeRibbon(
    Lifetime=&Lerp_Float(A=3, B=5, Alpha=$Particles.RibbonRandom)
    Ribbon Width=&FloatFromCurve(
     FloatCurve=DataInterface() {propertyValues {Curve {keys [<at 0 1>,<at 0.5 0.5>,<at 1 1>]}}}
     CurveIndex=$Particles.RibbonLinkOrder
     Scale Curve=$User.ArmRibbonWidth
    )
    UsePositionOffset=false
    Write Material Random=true
   )
   /SetVariables(Particles.MaterialRandom:NiagaraFloat=$Particles.RibbonRandom) {isSetParameters=true, module=SetVariables_#5}
   &SkeletalMeshLocation(
    Bone / Socket Index=&SelectIntFromArray(Direct Array Index=$Particles.SpawnGroup, Direct Array Mode=::NewEnumerator1, Int Selection Array=$Emitter.StartBoneIndexArray)
    Bone Sampling Mode=::NewEnumerator4
    MaskSpawnGroup=false
    Orientation Sampling=::NewEnumerator1
    Position Sampling=::NewEnumerator2
    Sampled Position Offset=&SelectVectorFromArray(Direct Array Index=$Particles.SpawnGroup, Direct Array Mode=::NewEnumerator1, Vector Selection Array=$Emitter.StartBoneOffsetArray)
    Skeletal Mesh=$User.SkeletalMesh
    Use Sampled Mesh Attribute as Ribbon ID=false
   )
   &FN_GooCalculateGoalPosition(
    Axis:Vector3f=$Emitter.RibbonAxis
    Scale:Vector3f=&Multiply_VectorByFloat(
     Vector=&Multiply_VectorByFloat(
      Vector=&VectorFromFloat(Value=&Multiply_Float(A=$Emitter.RibbonLength, B=0))
      Float=&Lerp_Float(A=$Emitter.RibbonLengthRandMin, B=$Emitter.RibbonLengthRandMax, Alpha=$Particles.RibbonRandom)
     )
     Float=$Particles.RibbonLinkOrder
    )
    StartPosition:Vector3f=$Particles.SkeletalMeshLocation.SampledPosition
   ) {module=FN_CalculateGoalPosition}
   /SetVariables(Particles.Position=$Particles.GoalPosition) {isSetParameters=true, module=SetVariables_#6}
   /SetVariables(Particles.PreviousRotation:Quat4f=$Engine.Owner.Rotation) {isSetParameters=true, module=SetVariables_#7}
   &DynamicMaterialParameters(
    Index 0 Param 1=&Multiply_Float(A=$Particles.RibbonRandom, B=10)
    Module.Param0WriteEnabled=true
    Module.Param1WriteEnabled=true
    Module.Param2WriteEnabled=true
    Module.Param3WriteEnabled=true
   ) {enabled=false}
   /SetVariables(Particles.SpriteSize=<10 10>) {isSetParameters=true, module=SetVariables_#8}
   &DynamicMaterialParameters(
    Index 0 Param 1=&SelectFloatFromArray(Direct Array Index=&SpawnGroup(), Float Selection Array=DataInterface() {propertyValues {FloatData [0,0.5]}})
    Module.Index0Param0Write=true
    Module.Index0Param1Write=true
    Module.Index0Param2Write=true
    Module.Index0Param3Write=true
   ) {module=DynamicMaterialParameters002}
  ]
  particleUpdateScript [
   &ParticleState()
   &SkeletalMeshLocation(
    Bone / Socket Index=&SelectIntFromArray(Direct Array Index=$Particles.SpawnGroup, Direct Array Mode=::NewEnumerator1, Int Selection Array=$Emitter.StartBoneIndexArray)
    Bone Sampling Mode=::NewEnumerator4
    MaskSpawnGroup=false
    Orientation Sampling=::NewEnumerator1
    Position Sampling=::NewEnumerator2
    Sampled Position Offset=&SelectVectorFromArray(Direct Array Index=$Particles.SpawnGroup, Direct Array Mode=::NewEnumerator1, Vector Selection Array=$Emitter.StartBoneOffsetArray)
    Skeletal Mesh=$User.SkeletalMesh
    Use Sampled Mesh Attribute as Ribbon ID=false
   ) {module=SkeletalMeshLocation001}
   &SolveForcesAndVelocity(Clamp Velocity=false, Limit Acceleration=false)
   &FN_GooCalculateGoalPosition(
    Axis:Vector3f=&SlerpVector(
     Vector A:Vector3f=&SlerpVector(
      Vector A:Vector3f=<-1 0 0>
      Vector B:Vector3f=&TransformVector(Vector=$User.WorldSpaceWindVector, Source Space=::World, Destination Space=::Local)
      Alpha:NiagaraFloat=$User.WorldSpaceWindVectorFactor
      Scale:NiagaraFloat=1
     )
     Vector B:Vector3f=<0 0 1>
     Alpha:NiagaraFloat=&FloatFromCurve(
      FloatCurve=DataInterface() {propertyValues {Curve {keys [
          {arriveTangent=1.3547, interpMode=RCIM_Cubic, leaveTangent=1.3547, tangentMode=RCTM_User, time=0, value=0.3}
          {arriveTangent=0.7, interpMode=RCIM_Cubic, leaveTangent=0.7, time=0.4, value=0.9}
          <at 1 1 RCIM_Cubic>
         ]}}}
      CurveIndex=$Particles.RibbonLinkOrder
      Scale Curve=0.7
     )
     Scale:NiagaraFloat=1
    )
    CustomOrientation:Quat4f=$Engine.Owner.Rotation
    Scale:Vector3f=&Multiply_VectorByFloat(
     Vector=&Multiply_VectorByFloat(
      Vector=&VectorFromFloat(Value=&Multiply_Float(
        A=$Emitter.RibbonLength
        B=&FloatFromCurve(FloatCurve=DataInterface() {propertyValues {Curve {keys [
             {arriveTangent=17.1487, interpMode=RCIM_Cubic, leaveTangent=17.1487, tangentMode=RCTM_User, time=0, value=0}
             {interpMode=RCIM_Cubic, tangentMode=RCTM_User, time=0.2, value=1.5}
             <at 0.25 1 RCIM_Cubic>
             {arriveTangent=-3.4902, interpMode=RCIM_Cubic, leaveTangent=-3.4902, tangentMode=RCTM_User, time=1, value=0}
            ]}}})
       ))
      Float=&Lerp_Float(A=$Emitter.RibbonLengthRandMin, B=$Emitter.RibbonLengthRandMax, Alpha=$Particles.RibbonRandom)
     )
     Float=$Particles.RibbonLinkOrder
    )
    StartPosition:Vector3f=$Particles.SkeletalMeshLocation001.SampledPosition
    Use Quaternion Rotation=true
   ) {module=FN_CalculateGoalPosition001}
   &FN_GooApproachGoalPosition(
    AttractionForce:NiagaraFloat=&Lerp_Float(A=$Emitter.ResolveSpeedLow, B=$Emitter.ResolveSpeedHigh, Alpha=$Particles.RibbonRandom)
    GoalPosition:NiagaraPosition=$Particles.GoalPosition
    MaximumDistance:NiagaraFloat=&Multiply_Float(
     A=&FloatFromCurve(
      FloatCurve=DataInterface() {propertyValues {Curve {keys [
          {arriveTangent=4.1785, interpMode=RCIM_Cubic, time=0, value=0}
          {arriveTangent=2.3315, interpMode=RCIM_Cubic, leaveTangent=2.3315, tangentMode=RCTM_User, time=1, value=1}
         ]}}}
      CurveIndex=$Particles.RibbonLinkOrder
      Scale Curve=&Multiply_Float(
       A=$Emitter.MaxDistance
       B=&FloatFromCurve(FloatCurve=DataInterface() {propertyValues {Curve {keys [<at 0.75 1>,<at 1 0>]}}})
      )
     )
     B=&Lerp_Float(A=$Emitter.MaxDistance_RandFactorLow, B=$Emitter.MaxDistance_RandFactorHigh, Alpha=$Particles.RibbonRandom)
    )
    RibbonAxis:Vector3f=$Particles.RibbonAxis
   )
   /SetVariables(Particles.PreviousRotation:Quat4f=$Engine.Owner.Rotation) {isSetParameters=true, module=SetVariables_#9}
   &ScaleColor(
    Scale Alpha=&FloatFromCurve(FloatCurve=DataInterface() {propertyValues {Curve {keys [{interpMode=RCIM_Cubic, tangentMode=RCTM_User, time=0.6, value=1}, {interpMode=RCIM_Cubic, tangentMode=RCTM_User, time=1, value=0}]}}})
    ScaleA=true
    ScaleRGB=false
   )
  ]
  rendererClasses [&NiagaraRibbonRendererProperties]
  renderers [{rendererClass=&NiagaraRibbonRendererProperties}]
 }
 Right_Hand {
  emitterSpawnScript [&FN_GooGlobalProperties(
    EndBoneIndexArray=DataInterface() {propertyValues {GpuSyncMode=SyncCpuToGpu, MaxElements=0, IntData [0]}}
    EndBoneOffsetArray=DataInterface() {propertyValues {GpuSyncMode=SyncCpuToGpu, MaxElements=0, FloatData [<0 0 0>]}}
    MaxDistance:NiagaraFloat=$User.ArmRibbonLength
    MaxDistanceCompressed=10
    MaxDistance_RandFactorHigh=1
    MaxDistance_RandFactorLow=1
    RibbonLength:NiagaraFloat=$User.ArmRibbonLength
    RibbonLengthRandHigh=1
    RibbonLengthRandLow=1
    StartBoneIndexArray=DataInterface() {propertyValues {GpuSyncMode=SyncCpuToGpu, MaxElements=0, IntData [5]}}
    StartBoneOffsetArray=DataInterface() {propertyValues {GpuSyncMode=SyncCpuToGpu, MaxElements=0, FloatData [<-3 0 0>]}}
   )]
  emitterUpdateScript [
   &EmitterState()
   &SpawnBurst_Instantaneous(
    Spawn Count=&Multiply_Int(A=$Emitter.ParticlesPerRibbon, B=$Emitter.NumberOfRibbons)
    Use Spawn Probability=false
    UseLoopCountLimit=false
   )
  ]
  particleSpawnScript [
   /SetVariables(Particles.SpawnGroup=&Divide_Int(A:NiagaraInt32=&ReturnExecIndex(), B:NiagaraInt32=$Emitter.ParticlesPerRibbon)) {isSetParameters=true, module=SetVariables_#10}
   /SetVariables(
    Particles.RibbonID=&MakeRibbonIDFromSpawnGroup()
    Particles.RibbonLinkOrder:NiagaraFloat=&Divide_Float(
     A=&MakeFloatFromInt(Int=&ModuloInt(A=&ReturnExecIndex(), B=$Emitter.ParticlesPerRibbon))
     B=&Subtract_Float(A=&MakeFloatFromInt(Int=$Emitter.ParticlesPerRibbon), B=1)
    )
   ) {isSetParameters=true, module=SetVariables_#11}
   /SetVariables(Particles.RibbonRandom:NiagaraFloat=&RandomRangeFloat(
     Minimum=0
     Maximum=1
     Randomness Mode=::NewEnumerator1
     Random Seed=&Add_Integer(A=$Engine.System.RandomSeed, B=$Particles.SpawnGroup)
     Fixed Random Seed=true
    )) {isSetParameters=true, module=SetVariables_#4}
   &InitializeRibbon(
    Lifetime=&Lerp_Float(A=3, B=5, Alpha=$Particles.RibbonRandom)
    Ribbon Width=&FloatFromCurve(
     FloatCurve=DataInterface() {propertyValues {Curve {keys [<at 0 1>,<at 0.5 0.5>,<at 1 1>]}}}
     CurveIndex=$Particles.RibbonLinkOrder
     Scale Curve=15
    )
    UsePositionOffset=false
    Write Material Random=true
   )
   /SetVariables(Particles.MaterialRandom:NiagaraFloat=$Particles.RibbonRandom) {isSetParameters=true, module=SetVariables_#5}
   &SkeletalMeshLocation(
    Bone / Socket Index=&SelectIntFromArray(Direct Array Index=$Particles.SpawnGroup, Direct Array Mode=::NewEnumerator1, Int Selection Array=$Emitter.StartBoneIndexArray)
    Bone Sampling Mode=::NewEnumerator4
    MaskSpawnGroup=false
    Orientation Sampling=::NewEnumerator1
    Position Sampling=::NewEnumerator2
    Sampled Position Offset=&SelectVectorFromArray(Direct Array Index=$Particles.SpawnGroup, Direct Array Mode=::NewEnumerator1, Vector Selection Array=$Emitter.StartBoneOffsetArray)
    Skeletal Mesh=$User.SkeletalMesh
    Use Sampled Mesh Attribute as Ribbon ID=false
   )
   &FN_GooCalculateGoalPosition(
    Axis:Vector3f=$Emitter.RibbonAxis
    Scale:Vector3f=&Multiply_VectorByFloat(
     Vector=&Multiply_VectorByFloat(
      Vector=&VectorFromFloat(Value=&Multiply_Float(A=$Emitter.RibbonLength, B=0))
      Float=&Lerp_Float(A=$Emitter.RibbonLengthRandMin, B=$Emitter.RibbonLengthRandMax, Alpha=$Particles.RibbonRandom)
     )
     Float=$Particles.RibbonLinkOrder
    )
    StartPosition:Vector3f=$Particles.SkeletalMeshLocation.SampledPosition
   ) {module=FN_CalculateGoalPosition}
   /SetVariables(Particles.Position=$Particles.GoalPosition) {isSetParameters=true, module=SetVariables_#6}
   /SetVariables(Particles.PreviousRotation:Quat4f=$Engine.Owner.Rotation) {isSetParameters=true, module=SetVariables_#7}
   &DynamicMaterialParameters(
    Index 0 Param 1=&Multiply_Float(A=$Particles.RibbonRandom, B=10)
    Module.Param0WriteEnabled=true
    Module.Param1WriteEnabled=true
    Module.Param2WriteEnabled=true
    Module.Param3WriteEnabled=true
   ) {enabled=false}
   /SetVariables(Particles.SpriteSize=<10 10>) {isSetParameters=true, module=SetVariables_#8}
  ]
  particleUpdateScript [
   &ParticleState()
   &SkeletalMeshLocation(
    Bone / Socket Index=&SelectIntFromArray(Direct Array Index=$Particles.SpawnGroup, Direct Array Mode=::NewEnumerator1, Int Selection Array=$Emitter.StartBoneIndexArray)
    Bone Sampling Mode=::NewEnumerator4
    MaskSpawnGroup=false
    Orientation Sampling=::NewEnumerator1
    Position Sampling=::NewEnumerator2
    Sampled Position Offset=&SelectVectorFromArray(Direct Array Index=$Particles.SpawnGroup, Direct Array Mode=::NewEnumerator1, Vector Selection Array=$Emitter.StartBoneOffsetArray)
    Skeletal Mesh=$User.SkeletalMesh
    Use Sampled Mesh Attribute as Ribbon ID=false
   ) {module=SkeletalMeshLocation001}
   &SolveForcesAndVelocity(Clamp Velocity=false, Limit Acceleration=false)
   &FN_GooCalculateGoalPosition(
    Axis:Vector3f=&SlerpVector(
     Vector A:Vector3f=&SlerpVector(
      Vector A:Vector3f=<-1 0 0>
      Vector B:Vector3f=&TransformVector(Vector=$User.WorldSpaceWindVector, Source Space=::World, Destination Space=::Local)
      Alpha:NiagaraFloat=$User.WorldSpaceWindVectorFactor
      Scale:NiagaraFloat=1
     )
     Vector B:Vector3f=<0 0 1>
     Alpha:NiagaraFloat=&FloatFromCurve(
      FloatCurve=DataInterface() {propertyValues {Curve {keys [
          {arriveTangent=1.3547, interpMode=RCIM_Cubic, leaveTangent=1.3547, tangentMode=RCTM_User, time=0, value=0.3}
          {arriveTangent=0.7, interpMode=RCIM_Cubic, leaveTangent=0.7, time=0.4, value=0.9}
          <at 1 1 RCIM_Cubic>
         ]}}}
      CurveIndex=$Particles.RibbonLinkOrder
      Scale Curve=0.7
     )
     Scale:NiagaraFloat=1
    )
    CustomOrientation:Quat4f=$Engine.Owner.Rotation
    Scale:Vector3f=&Multiply_VectorByFloat(
     Vector=&Multiply_VectorByFloat(
      Vector=&VectorFromFloat(Value=&Multiply_Float(
        A=$Emitter.RibbonLength
        B=&FloatFromCurve(FloatCurve=DataInterface() {propertyValues {Curve {keys [
             {arriveTangent=17.1487, interpMode=RCIM_Cubic, leaveTangent=17.1487, tangentMode=RCTM_User, time=0, value=0}
             {interpMode=RCIM_Cubic, tangentMode=RCTM_User, time=0.2, value=1.5}
             <at 0.25 1 RCIM_Cubic>
             {arriveTangent=-3.4902, interpMode=RCIM_Cubic, leaveTangent=-3.4902, tangentMode=RCTM_User, time=1, value=0}
            ]}}})
       ))
      Float=&Lerp_Float(A=$Emitter.RibbonLengthRandMin, B=$Emitter.RibbonLengthRandMax, Alpha=$Particles.RibbonRandom)
     )
     Float=$Particles.RibbonLinkOrder
    )
    StartPosition:Vector3f=$Particles.SkeletalMeshLocation001.SampledPosition
    Use Quaternion Rotation=true
   ) {module=FN_CalculateGoalPosition001}
   &FN_GooApproachGoalPosition(
    AttractionForce:NiagaraFloat=&Lerp_Float(A=$Emitter.ResolveSpeedLow, B=$Emitter.ResolveSpeedHigh, Alpha=$Particles.RibbonRandom)
    GoalPosition:NiagaraPosition=$Particles.GoalPosition
    MaximumDistance:NiagaraFloat=&Multiply_Float(
     A=&FloatFromCurve(
      FloatCurve=DataInterface() {propertyValues {Curve {keys [
          {arriveTangent=4.1785, interpMode=RCIM_Cubic, time=0, value=0}
          {arriveTangent=2.3315, interpMode=RCIM_Cubic, leaveTangent=2.3315, tangentMode=RCTM_User, time=1, value=1}
         ]}}}
      CurveIndex=$Particles.RibbonLinkOrder
      Scale Curve=&Multiply_Float(
       A=$Emitter.MaxDistance
       B=&FloatFromCurve(FloatCurve=DataInterface() {propertyValues {Curve {keys [<at 0.75 1>,<at 1 0>]}}})
      )
     )
     B=&Lerp_Float(A=$Emitter.MaxDistance_RandFactorLow, B=$Emitter.MaxDistance_RandFactorHigh, Alpha=$Particles.RibbonRandom)
    )
    RibbonAxis:Vector3f=$Particles.RibbonAxis
   )
   /SetVariables(Particles.PreviousRotation:Quat4f=$Engine.Owner.Rotation) {isSetParameters=true, module=SetVariables_#9}
   &ScaleColor(
    Scale Alpha=&FloatFromCurve(FloatCurve=DataInterface() {propertyValues {Curve {keys [{interpMode=RCIM_Cubic, tangentMode=RCTM_User, time=0.6, value=1}, {interpMode=RCIM_Cubic, tangentMode=RCTM_User, time=1, value=0}]}}})
    ScaleA=true
    ScaleRGB=false
   )
  ]
  rendererClasses [&NiagaraRibbonRendererProperties]
  renderers [{rendererClass=&NiagaraRibbonRendererProperties}]
 }
 Right_Leg {
  emitterSpawnScript [&FN_GooGlobalProperties(
    EndBoneIndexArray=DataInterface() {propertyValues {GpuSyncMode=SyncCpuToGpu, MaxElements=0, IntData [0]}}
    EndBoneOffsetArray=DataInterface() {propertyValues {GpuSyncMode=SyncCpuToGpu, MaxElements=0, FloatData [<0 0 0>]}}
    MaxDistance:NiagaraFloat=$User.LegRibbonLength
    MaxDistanceCompressed=10
    MaxDistance_RandFactorHigh=1
    MaxDistance_RandFactorLow=1
    NumberOfRibbons=2
    ParticlesPerRibbon=3
    ResolveSpeedHigh=3
    RibbonLength:NiagaraFloat=$User.LegRibbonLength
    RibbonLengthRandHigh=1
    RibbonLengthRandLow=1
    StartBoneIndexArray=DataInterface() {propertyValues {GpuSyncMode=SyncCpuToGpu, MaxElements=0, IntData [8,9]}}
    StartBoneOffsetArray=DataInterface() {propertyValues {GpuSyncMode=SyncCpuToGpu, MaxElements=0, FloatData [<20 0 5>, <-15 0 2>]}}
   )]
  emitterUpdateScript [
   &EmitterState()
   &SpawnBurst_Instantaneous(
    Spawn Count=&Multiply_Int(A=$Emitter.ParticlesPerRibbon, B=$Emitter.NumberOfRibbons)
    Use Spawn Probability=false
    UseLoopCountLimit=false
   )
  ]
  particleSpawnScript [
   /SetVariables(Particles.SpawnGroup=&Divide_Int(A:NiagaraInt32=&ReturnExecIndex(), B:NiagaraInt32=$Emitter.ParticlesPerRibbon)) {isSetParameters=true, module=SetVariables_#10}
   /SetVariables(
    Particles.RibbonID=&MakeRibbonIDFromSpawnGroup()
    Particles.RibbonLinkOrder:NiagaraFloat=&Divide_Float(
     A=&MakeFloatFromInt(Int=&ModuloInt(A=&ReturnExecIndex(), B=$Emitter.ParticlesPerRibbon))
     B=&Subtract_Float(A=&MakeFloatFromInt(Int=$Emitter.ParticlesPerRibbon), B=1)
    )
   ) {isSetParameters=true, module=SetVariables_#11}
   /SetVariables(Particles.RibbonRandom:NiagaraFloat=&RandomRangeFloat(
     Minimum=0
     Maximum=1
     Randomness Mode=::NewEnumerator1
     Random Seed=&Add_Integer(A=$Engine.System.RandomSeed, B=$Particles.SpawnGroup)
     Fixed Random Seed=true
    )) {isSetParameters=true, module=SetVariables_#4}
   &InitializeRibbon(
    Lifetime=&Lerp_Float(A=3, B=5, Alpha=$Particles.RibbonRandom)
    Ribbon Width=$User.LegRibbonWidth
    UsePositionOffset=false
    Write Material Random=true
   )
   /SetVariables(Particles.MaterialRandom:NiagaraFloat=$Particles.RibbonRandom) {isSetParameters=true, module=SetVariables_#5}
   &SkeletalMeshLocation(
    Bone / Socket Index=&SelectIntFromArray(Direct Array Index=$Particles.SpawnGroup, Direct Array Mode=::NewEnumerator1, Int Selection Array=$Emitter.StartBoneIndexArray)
    Bone Sampling Mode=::NewEnumerator4
    MaskSpawnGroup=false
    Orientation Sampling=::NewEnumerator1
    Position Sampling=::NewEnumerator2
    Sampled Position Offset=&SelectVectorFromArray(Direct Array Index=$Particles.SpawnGroup, Direct Array Mode=::NewEnumerator1, Vector Selection Array=$Emitter.StartBoneOffsetArray)
    Skeletal Mesh=$User.SkeletalMesh
    Use Sampled Mesh Attribute as Ribbon ID=false
   )
   &FN_GooCalculateGoalPosition(
    Axis:Vector3f=$Emitter.RibbonAxis
    Scale:Vector3f=&Multiply_VectorByFloat(
     Vector=&Multiply_VectorByFloat(
      Vector=&VectorFromFloat(Value=&Multiply_Float(A=$Emitter.RibbonLength, B=0))
      Float=&Lerp_Float(A=$Emitter.RibbonLengthRandMin, B=$Emitter.RibbonLengthRandMax, Alpha=$Particles.RibbonRandom)
     )
     Float=$Particles.RibbonLinkOrder
    )
    StartPosition:Vector3f=$Particles.SkeletalMeshLocation.SampledPosition
   ) {module=FN_CalculateGoalPosition}
   /SetVariables(Particles.Position=$Particles.GoalPosition) {isSetParameters=true, module=SetVariables_#6}
   /SetVariables(Particles.PreviousRotation:Quat4f=$Engine.Owner.Rotation) {isSetParameters=true, module=SetVariables_#7}
   &DynamicMaterialParameters(
    Index 0 Param 1=&Multiply_Float(A=$Particles.RibbonRandom, B=10)
    Module.Param0WriteEnabled=true
    Module.Param1WriteEnabled=true
    Module.Param2WriteEnabled=true
    Module.Param3WriteEnabled=true
   ) {enabled=false}
   /SetVariables(Particles.SpriteSize=<10 10>) {isSetParameters=true, module=SetVariables_#8}
   &DynamicMaterialParameters(
    Index 0 Param 1=&SelectFloatFromArray(Direct Array Index=&SpawnGroup(), Float Selection Array=DataInterface() {propertyValues {FloatData [0.5,0.9]}})
    Module.Index0Param0Write=true
    Module.Index0Param1Write=true
    Module.Index0Param2Write=true
    Module.Index0Param3Write=true
   ) {module=DynamicMaterialParameters002}
  ]
  particleUpdateScript [
   &ParticleState()
   &SkeletalMeshLocation(
    Bone / Socket Index=&SelectIntFromArray(Direct Array Index=$Particles.SpawnGroup, Direct Array Mode=::NewEnumerator1, Int Selection Array=$Emitter.StartBoneIndexArray)
    Bone Sampling Mode=::NewEnumerator4
    MaskSpawnGroup=false
    Orientation Sampling=::NewEnumerator1
    Position Sampling=::NewEnumerator2
    Sampled Position Offset=&SelectVectorFromArray(Direct Array Index=$Particles.SpawnGroup, Direct Array Mode=::NewEnumerator1, Vector Selection Array=$Emitter.StartBoneOffsetArray)
    Skeletal Mesh=$User.SkeletalMesh
    Use Sampled Mesh Attribute as Ribbon ID=false
   ) {module=SkeletalMeshLocation001}
   &SolveForcesAndVelocity(Clamp Velocity=false, Limit Acceleration=false)
   &FN_GooCalculateGoalPosition(
    Axis:Vector3f=&SlerpVector(
     Vector A:Vector3f=&SlerpVector(
      Vector A:Vector3f=<-1 0 0>
      Vector B:Vector3f=&TransformVector(Vector=$User.WorldSpaceWindVector, Source Space=::World, Destination Space=::Local)
      Alpha:NiagaraFloat=$User.WorldSpaceWindVectorFactor
      Scale:NiagaraFloat=1
     )
     Vector B:Vector3f=<0 0 1>
     Alpha:NiagaraFloat=&FloatFromCurve(
      FloatCurve=DataInterface() {propertyValues {Curve {keys [
          {arriveTangent=1.3547, interpMode=RCIM_Cubic, leaveTangent=1.3547, tangentMode=RCTM_User, time=0, value=0.3}
          {arriveTangent=0.7, interpMode=RCIM_Cubic, leaveTangent=0.7, time=0.4, value=0.8}
          <at 1 1 RCIM_Cubic>
         ]}}}
      CurveIndex=$Particles.RibbonLinkOrder
     )
     Scale:NiagaraFloat=1
    )
    CustomOrientation:Quat4f=$Engine.Owner.Rotation
    Scale:Vector3f=&Multiply_VectorByFloat(
     Vector=&Multiply_VectorByFloat(
      Vector=&VectorFromFloat(Value=&Multiply_Float(
        A=$Emitter.RibbonLength
        B=&FloatFromCurve(FloatCurve=DataInterface() {propertyValues {Curve {keys [
             {arriveTangent=30.27, interpMode=RCIM_Cubic, leaveTangent=30.27, tangentMode=RCTM_User, time=0, value=0}
             {interpMode=RCIM_Cubic, tangentMode=RCTM_User, time=0.15, value=1.5}
             <at 0.2 1 RCIM_Cubic>
             {arriveTangent=-3.4902, interpMode=RCIM_Cubic, leaveTangent=-3.4902, tangentMode=RCTM_User, time=1, value=0}
            ]}}})
       ))
      Float=&Lerp_Float(A=$Emitter.RibbonLengthRandMin, B=$Emitter.RibbonLengthRandMax, Alpha=$Particles.RibbonRandom)
     )
     Float=$Particles.RibbonLinkOrder
    )
    StartPosition:Vector3f=$Particles.SkeletalMeshLocation001.SampledPosition
    Use Quaternion Rotation=true
   ) {module=FN_CalculateGoalPosition001}
   &FN_GooApproachGoalPosition(
    AttractionForce:NiagaraFloat=&Lerp_Float(A=$Emitter.ResolveSpeedLow, B=$Emitter.ResolveSpeedHigh, Alpha=$Particles.RibbonRandom)
    GoalPosition:NiagaraPosition=$Particles.GoalPosition
    MaximumDistance:NiagaraFloat=&Multiply_Float(
     A=&FloatFromCurve(
      FloatCurve=DataInterface() {propertyValues {Curve {keys [
          {arriveTangent=4.1785, interpMode=RCIM_Cubic, time=0, value=0}
          {arriveTangent=2.3315, interpMode=RCIM_Cubic, leaveTangent=2.3315, tangentMode=RCTM_User, time=1, value=1}
         ]}}}
      CurveIndex=$Particles.RibbonLinkOrder
      Scale Curve=&Multiply_Float(
       A=$Emitter.MaxDistance
       B=&FloatFromCurve(FloatCurve=DataInterface() {propertyValues {Curve {keys [<at 0.75 1>,<at 1 0>]}}})
      )
     )
     B=&Lerp_Float(A=$Emitter.MaxDistance_RandFactorLow, B=$Emitter.MaxDistance_RandFactorHigh, Alpha=$Particles.RibbonRandom)
    )
    RibbonAxis:Vector3f=$Particles.RibbonAxis
   )
   /SetVariables(Particles.PreviousRotation:Quat4f=$Engine.Owner.Rotation) {isSetParameters=true, module=SetVariables_#9}
   &ScaleColor(
    Scale Alpha=&FloatFromCurve(FloatCurve=DataInterface() {propertyValues {Curve {keys [{interpMode=RCIM_Cubic, tangentMode=RCTM_User, time=0.6, value=1}, {interpMode=RCIM_Cubic, tangentMode=RCTM_User, time=1, value=0}]}}})
    ScaleA=true
    ScaleRGB=false
   )
  ]
  rendererClasses [&NiagaraRibbonRendererProperties]
  renderers [{rendererClass=&NiagaraRibbonRendererProperties}]
 }
 SmallBits {
  emitterUpdateScript [
   &EmitterState()
   &SpawnRate(
    SpawnRate=&FloatFromCurve(
     FloatCurve=DataInterface() {propertyValues {Curve {keys [{interpMode=RCIM_Cubic, tangentMode=RCTM_User, time=3, value=1}, {interpMode=RCIM_Cubic, tangentMode=RCTM_User, time=5, value=0}]}}}
     CurveIndex=$Emitter.Age
    )
    Use Spawn Probability=false
   ) {module=SpawnRate001}
  ]
  particleSpawnScript [
   &FN_SetSpawnGroup(Number of Spawn Groups=6)
   &InitializeParticle(
    Color Mode=::NewEnumerator0
    Lifetime Max=0.5
    Lifetime Min=0.3
    Lifetime Mode=::NewEnumerator1
    Sprite Size Max=<5 9>
    Sprite Size Min=<2 5>
    Sprite Size Mode=::NewEnumerator2
    UsePositionOffset=false
    Write Material Random=true
   )
   /SetVariables(Particles.SpriteSize=&Multiply_Vector2D_ByFloat(
     Float=&SelectFloatFromArray(Direct Array Index=$Particles.SpawnGroup, Float Selection Array=DataInterface() {propertyValues {FloatData [1,1,1,1,2,2]}})
     Vector2D=$Particles.SpriteSize
    )) {isSetParameters=true}
   &SkeletalMeshLocation(
    Bone / Socket Index=$Particles.SpawnGroup
    Bone Sampling Mode=::NewEnumerator4
    MaskSpawnGroup=false
    Sampled Position Offset=&SelectVectorFromArray(
     Direct Array Index=$Particles.SpawnGroup
     Vector Selection Array=DataInterface() {propertyValues {FloatData [<5 0 3>, <-5 0 -3>, <-3 -2 0>, <3 2 0>, <-20 0 -5>, <20 0 5>]}}
    )
    Skeletal Mesh=$System.Skeletal Mesh
   )
   &ShapeLocation(
    MaskSpawnGroup=false
    Non Uniform Scale=&SelectVectorFromArray(
     Direct Array Index=$Particles.SpawnGroup
     Vector Selection Array=DataInterface() {propertyValues {FloatData [<1 1 0.1>, <1 1 1>, <1 1 1>, <1 1 1>, <0.1 1 1>, <0.1 1 1>]}}
    )
    Override Seed=false
    Sphere Radius=&SelectFloatFromArray(Direct Array Index=$Particles.SpawnGroup, Float Selection Array=DataInterface() {propertyValues {FloatData [3,3,3,3,7,7]}})
    UseOwnerScale=true
   )
   &AddVelocity(Velocity=<0 0 5>)
   &AddVelocity(
    Module.Use Falloff Distance=false
    Origin Offset=&SelectVectorFromArray(
     Direct Array Index=$Particles.SpawnGroup
     Vector Selection Array=DataInterface() {propertyValues {FloatData [<0 0 0>, <0 0 0>, <0 0 0>, <0 0 0>, <0 0 -25>, <0 0 -25>]}}
    )
    Override Seed=false
    Velocity Mode=::NewEnumerator1
    Velocity Origin=$Particles.SkeletalMeshLocation.SampledPosition
    Velocity Speed=20
   ) {module=AddVelocity001}
   /SetVariables(Particles.Velocity=&Multiply_VectorByFloat(
     Vector=$Particles.Velocity
     Float=&SelectFloatFromArray(Direct Array Index=$Particles.SpawnGroup, Float Selection Array=DataInterface() {propertyValues {FloatData [1,1,1,1,2,2]}})
    )) {isSetParameters=true}
  ]
  particleUpdateScript [
   &ParticleState()
   &ScaleSpriteSize(Uniform Curve Sprite Scale=DataInterface() {propertyValues {Curve {keys [{interpMode=RCIM_Cubic, tangentMode=RCTM_User, time=0.3, value=1}, {tangentMode=RCTM_User, time=1, value=0}]}}})
   &DynamicMaterialParameters(
    Index 0 Param 1=&RandomRangeFloat(Minimum=0, Maximum=1)
    Module.Param0WriteEnabled=true
    Module.Param1WriteEnabled=true
    Module.Param2WriteEnabled=true
    Module.Param3WriteEnabled=true
   ) {module=DynamicMaterialParameters001}
   &GravityForce(Gravity=<0 0 15>)
   &Drag(Use Linear Drag=true, Use Rotational Drag=true)
   &CurlNoiseForce(Noise Frequency=250, Noise Strength=50, Pan Noise=true, Pan Noise Field=<3 0 0>) {enabled=false}
   &SolveForcesAndVelocity(Clamp Velocity=false, Limit Acceleration=false)
   &ScaleColor(
    Scale RGB=&Multiply_VectorByFloat(
     Vector=<1 1 1>
     Float=&FNTimeOfDayScalar(Night Scale=0.5, Morning Scale=0.75, Evening Scale=0.7)
    )
    ScaleA=false
    ScaleRGB=true
   )
  ]
  rendererClasses [&NiagaraSpriteRendererProperties]
  renderers [{rendererClass=&NiagaraSpriteRendererProperties}]
 }
}
summary {
 systemName=NS_ChicleVeil_Reactive
 emitters [
  {emitterName=Left_Arm, rendererClasses [&NiagaraRibbonRendererProperties]}
  {emitterName=Right_Arm, rendererClasses [&NiagaraRibbonRendererProperties]}
  {emitterName=Left_Leg, rendererClasses [&NiagaraRibbonRendererProperties]}
  {emitterName=Right_Leg, rendererClasses [&NiagaraRibbonRendererProperties]}
  {emitterName=Left_Hand, rendererClasses [&NiagaraRibbonRendererProperties]}
  {emitterName=Right_Hand, rendererClasses [&NiagaraRibbonRendererProperties]}
  {emitterName=Burst, rendererClasses [&NiagaraSpriteRendererProperties]}
  {emitterName=SmallBits, rendererClasses [&NiagaraSpriteRendererProperties]}
 ]
 userVariables [
  User.ArmRibbonLength {type=NiagaraFloat, defaultValue=17, description=""}
  User.ArmRibbonWidth {type=NiagaraFloat, defaultValue=15, description=""}
  User.BariumPaintColor {type=LinearColor, description="", defaultValue {struct=LinearColor, value=<rgba 1 1 1 1>}}
  User.EnergyFillFactor {type=NiagaraFloat, defaultValue=0.5, description=""}
  User.HipRibbonLength {type=NiagaraFloat, defaultValue=20, description=""}
  User.HipRibbonWidth {type=NiagaraFloat, defaultValue=15, description=""}
  User.LegRibbonLength {type=NiagaraFloat, defaultValue=35, description=""}
  User.LegRibbonWidth {type=NiagaraFloat, defaultValue=20, description=""}
  User.SkeletalMesh {
   type=NiagaraDataInterfaceSkeletalMesh
   defaultValue {
    struct=NiagaraExt_VariableValue_DataInterface
    value {dataInterface=&self:NiagaraDataInterfaceSkeletalMesh_0}
   }
  }
  User.SpawnHelmetFX {type=NiagaraFloat, defaultValue=0, description=""}
  User.SpawnJacketFX {type=NiagaraFloat, defaultValue=0, description=""}
  User.WorldSpaceWindVector {type=Vector3f, description="", defaultValue {struct=Vector3f, value=<1 0 0>}}
  User.WorldSpaceWindVectorFactor {type=NiagaraFloat, defaultValue=0, description=""}
 ]
}
