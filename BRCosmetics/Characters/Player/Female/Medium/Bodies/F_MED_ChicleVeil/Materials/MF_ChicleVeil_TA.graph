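# Material function for the ChicleVeil character body, outputting animated swirl color, fresnel glow, alpha and noise.
# A flowmapped swirl texture plus sine strands drives a three-color lerp, with overridable balance and strand-based hue, brightness and desaturation.
# Fresnel glow uses counter-panning fire noise, simplified on lower quality, with a base-to-reactive intensity blend.
# SwirlAlpha effectively passes SwirlFresnel through because its contrast blend is fixed at zero; Strands2 and SwirlOpacityInfluence are unconnected.
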MaterialFunction /BRCosmetics/Characters/Player/Female/Medium/Bodies/F_MED_ChicleVeil/Materials/MF_ChicleVeil_TA
graph {nodes [
  n1 = TextureObject() {samplerType=SAMPLERTYPE_LinearColor, texture=/Game/Global/Textures/Default/Flowmaps/T_SwirlFlow_02}
  n2 = Multiply(A=TextureCoordinate(), B=ScalarParameter() {defaultValue=4, group=Swirl, parameterName=SwirlDistortionTiling}) {constB=4}
  n3 = Constant2Vector() {g=1, r=0.25}
  n4 = Multiply(
   A=ComponentMask(None=TextureSample(UVs=@n2, Tex=@n1) {samplerSource=SSM_Wrap_WorldGroupSettings}.RGB) {g=true, r=true}
   B=StaticSwitchParameter(True=MaterialFunctionCall(XY=@n3) {materialFunction=&Swizzle}.YX, False=@n3) {parameterName=SwizzlePanning}
  )
  n5 = ScalarParameter() {defaultValue=1, group=Swirl, parameterName=SwirlDistortionIntensity}
  n6 = Time()
  n7 = ScalarParameter() {defaultValue=0.2, group=Swirl, parameterName=SwirlSpeed}
  n8 = Multiply(A=TextureCoordinate() {vTiling=1.25}, B=ScalarParameter() {defaultValue=8, group=Swirl, parameterName=SwirlTiling}) {constB=8}
  n9 = Multiply(A=Time(), B=ComponentMask(None=VectorParameter() {parameterName=SwirlPanSpeed, defaultValue=<rgba 0 0 0 1>}.RGB) {g=true, r=true})
  n10 = ComponentMask(None=MaterialFunctionCall!(
    Diffuse=@n1
    Normal Map
    "Flow Vector Map (see tooltip)"=Multiply(A=@n4, B=@n5)
    UV texture 2 offset
    "Time (see tooltip)"=Multiply(A=@n6, B=@n7) {constB=0.2}
    useMipLevel
    UVs=Add(A=@n8, B=@n9)
    Texture Size
    Texture Mip Bias
   ) {materialFunction=/Engine/Functions/Engine_MaterialFunctions02/Texturing/Flowmaps.FlowMaps}.Diffuse) {g=true}
  n11 = Sine(None=Multiply(A=@n10, B=ScalarParameter() {defaultValue=8, group=Swirl, parameterName=StrandsSize}))
  n12 = Add(
   A=@n10
   B=Multiply(
    A=NamedRerouteDeclaration(None=@n11) {name=Strands, nodeColor=<rgba 0.4297 0 1 1>}
    B=ScalarParameter() {defaultValue=0.05, group=Swirl, parameterName=StrandsIntensity}
   ) {constB=0.5}
  )
  n13 = ScalarParameter() {defaultValue=1.2, group=Swirl, parameterName=SwirlColorBalance}
  n14 = ScalarParameter() {defaultValue=1.2, group=Swirl, parameterName=SwirlColorBalance_LOW}
  n15 = Add(
   A=Power(Base=NamedRerouteDeclaration(None=@n12) {name=Flow, nodeColor=<rgba 0 0.6484 1 1>}, Exp=QualitySwitch(Default=@n13, Low=@n14)) {constExponent=0.8}
   B=FunctionInput() {inputName=SwirlColorBias, inputType=FunctionInput_Scalar}
  )
  n16 = Constant()
  n17 = FunctionInput(Preview=@n16) {bUsePreviewValueAsDefault=true, inputName=SwirlColorBalanceOverride, inputType=FunctionInput_Scalar, sortPriority=10}
  n18 = FunctionInput(Preview=@n16) {bUsePreviewValueAsDefault=true, inputName=SwirlColorBalanceOverrideMask, inputType=FunctionInput_Scalar, sortPriority=20}
  n19 = MaterialFunctionCall(
   A=VectorParameter() {group=Swirl, parameterName=SwirlColorB, defaultValue=<rgba 0.0223 1 0.1018 1>}.RGB
   B=VectorParameter() {group=Swirl, parameterName=SwirlColorC, defaultValue=<rgba 1 0.7315 0 1>}.RGB
   C=VectorParameter() {group=Swirl, parameterName=SwirlColorA, defaultValue=<rgba 1 0.0441 0.0872 1>}.RGB
   Alpha=SmoothStep(
    Min=ScalarParameter() {group=Swirl, parameterName=ColorBalance_SmoothMin}
    Max=ScalarParameter() {defaultValue=1, group=Swirl, parameterName=ColorBalance_SmoothMax}
    Value=LinearInterpolate(A=@n15, B=@n17, Alpha=@n18)
   )
  ) {materialFunction=&Lerp_3Color}
  n20 = NamedRerouteUsage()
  n21 = Multiply(A=ScalarParameter() {group=Swirl, parameterName=StrandsBrighnessFactor}, B=@n20)
  n22 = Multiply(A=@n20, B=ScalarParameter() {group=Swirl, parameterName=StrandsSaturationFactor})
  n23 = ScalarParameter() {group=Swirl, parameterName=SwirlColorDesaturation}
  n24 = MaterialFunctionCall(
   Hue Shift Percentage=Multiply(A=NamedRerouteUsage(), B=ScalarParameter() {group=Swirl, parameterName=StrandsHueFactor})
   Texture=Desaturation(None=Add(A=@n19.Result, B=@n21), Fraction=Add(A=@n22, B=@n23))
  ) {materialFunction=&HueShift}
  FunctionOutput(None=NamedRerouteDeclaration(None=LinearInterpolate(A=@n24.Result, B=@n19.Result, Alpha=@n18)) {name=Swirl, nodeColor=<rgba 0 0.5078 1 1>}) {outputName=Swirl}
  n25 = FunctionInput() {inputName=FresnelMax, inputType=FunctionInput_Scalar}
  n26 = VertexNormalWS()
  n27 = CameraVectorWS()
  n28 = Power(Base=OneMinus(None=DotProduct(A=@n26, B=@n27))) {constExponent=4}
  n29 = Multiply(A=TextureCoordinate(), B=ScalarParameter() {defaultValue=2, parameterName=FresnelNoise_Tiling}) {constB=16}
  n30 = Time()
  n31 = Constant2Vector() {g=1}
  n32 = Multiply(
   A=StaticSwitchParameter(True=MaterialFunctionCall(XY=@n31) {materialFunction=&Swizzle}.YX, False=@n31) {parameterName=SwizzlePanning}
   B=ScalarParameter() {defaultValue=0.075, parameterName=FresnelNoise_Speed}
  )
  n33 = TextureSample(UVs=Add(A=@n29, B=Multiply(A=@n30, B=@n32))) {samplerSource=SSM_Wrap_WorldGroupSettings, texture=/Game/Global/Textures/Default/Noise/T_FireNoise_01}
  n34 = Add(A=@n29, B=Multiply(A=@n30, B=Multiply(A=@n32) {constB=-1}))
  n35 = Max(
   A=@n33.R
   B=TextureSample(UVs=Multiply(A=@n34) {constB=1.1}) {samplerSource=SSM_Wrap_WorldGroupSettings, texture=&T_FireNoise_01}.R
  )
  n36 = NamedRerouteDeclaration(None=QualitySwitch(Default=@n35, Low=Constant() {r=1}, High=@n35, Medium=@n33.R, Epic=@n35)) {name=FresnelNoise, nodeColor=<rgba 0.7812 0 1 1>}
  n37 = Multiply(
   A=Multiply(A=VectorParameter() {parameterName=GlowFresnelColor, defaultValue=<rgba 0.5195 0 1 1>}.RGB, B=Max(A=@n25, B=@n28))
   B=Multiply(A=@n36, B=ScalarParameter() {defaultValue=6, parameterName=FresnelNoise_Intensity}) {constB=4}
  )
  n38 = LinearInterpolate(
   A=ScalarParameter() {parameterName=GlowFresnelIntensityBase}
   B=ScalarParameter() {defaultValue=1, parameterName=GlowFresnelIntensityReactive}
   Alpha=ScalarParameter() {parameterName=FresnelBlend}
  )
  FunctionOutput(None=NamedRerouteDeclaration(None=Multiply(A=@n37, B=@n38)) {name=FresnelGlow, nodeColor=<rgba 0 1 1 1>}) {outputName=FesnelGlow}
  n39 = Constant() {r=1}
  n40 = MaterialFunctionCall(
   In=Multiply(A=NamedRerouteUsage(), B=ScalarParameter() {defaultValue=1.2, group=Swirl, parameterName=SwirlOpacityInfluenceBrightness}) {constB=1.2}
   Contrast=ScalarParameter() {defaultValue=1, group=Swirl, parameterName=SwirlOpacityInfluenceContrast}
  ) {materialFunction=&CheapContrast}
  n41 = Constant()
  FunctionOutput(None=Multiply(A=FunctionInput() {inputName=SwirlFresnel, inputType=FunctionInput_Scalar}, B=LinearInterpolate(A=@n39, B=@n40.Result, Alpha=@n41))) {outputName=SwirlAlpha}
  FunctionOutput(None=@n36) {outputName=FresnelNoise}
  n42 = ScalarParameter() {defaultValue=16, group=Swirl, parameterName=StrandsSize2}
  NamedRerouteDeclaration(None=Sine(None=Multiply(A=@n10, B=@n42))) {name=Strands2, nodeColor=<rgba 0.6875 1 0 1>}
  ScalarParameter() {defaultValue=0.8, group=Swirl, parameterName=SwirlOpacityInfluence}
 ]}
