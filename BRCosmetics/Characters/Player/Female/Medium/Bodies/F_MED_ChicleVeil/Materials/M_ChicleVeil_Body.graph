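# Body material for the ChicleVeil character cosmetic, layering swirl, thin-film, Fresnel-glow and crystal-gem effects between the 2023 character parent PreFX and PostFX functions.
# MF_ChicleVeil_TA generates the swirls, a curve atlas colors the thin film, and gems reuse the ChicleVeil pickaxe's crystal function.
# Named masks separate tattoo, fabric, gem and pants regions, alongside gradients computed from position.
# Time-of-day scalars and quality switches also feed the effects.
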Material /BRCosmetics/Characters/Player/Female/Medium/Bodies/F_MED_ChicleVeil/Materials/M_ChicleVeil_Body
graph {
 expressions {
  MaterialExpressionAbs_2 {class=MaterialExpressionAbs, inputs [None {from=MaterialExpressionNamedRerouteUsage_31, output=""}], properties {}}
  MaterialExpressionAbs_3 {class=MaterialExpressionAbs, inputs [None {from=MaterialExpressionAdd_21, output=""}], properties {}}
  MaterialExpressionAbs_4 {class=MaterialExpressionAbs, inputs [None {from=MaterialExpressionAdd_17, output=""}], properties {}}
  MaterialExpressionAbs_5 {class=MaterialExpressionAbs, inputs [None {from=MaterialExpressionAdd_29, output=""}], properties {}}
  MaterialExpressionAdd_0 {
   class=MaterialExpressionAdd
   inputs [
    A {from=MaterialExpressionMultiply_3}
    B {from=MaterialExpressionScalarParameter_1}
   ]
  }
  MaterialExpressionAdd_1 {
   class=MaterialExpressionAdd
   inputs [
    A {from=MaterialExpressionMultiply_17}
    B {from=MaterialExpressionMultiply_23}
   ]
  }
  MaterialExpressionAdd_12 {
   class=MaterialExpressionAdd
   inputs [
    A {from=MaterialExpressionSubtract_10}
    B {from=""}
   ]
   properties {constB=-15}
  }
  MaterialExpressionAdd_14 {
   class=MaterialExpressionAdd
   inputs [
    A {from=MaterialExpressionMultiply_46}
    B {from=MaterialExpressionMultiply_5}
   ]
  }
  MaterialExpressionAdd_17 {
   class=MaterialExpressionAdd
   inputs [
    A {from=MaterialExpressionAdd_21}
    B {from=""}
   ]
  }
  MaterialExpressionAdd_19 {
   class=MaterialExpressionAdd
   inputs [
    A {from=MaterialExpressionMultiply_18}
    B {from=MaterialExpressionNamedRerouteUsage_45}
   ]
  }
  MaterialExpressionAdd_21 {
   class=MaterialExpressionAdd
   inputs [
    A {from=MaterialExpressionComponentMask_12}
    B {from=MaterialExpressionScalarParameter_52}
   ]
   properties {constB=-35}
  }
  MaterialExpressionAdd_23 {
   class=MaterialExpressionAdd
   inputs [
    A {from=MaterialExpressionScalarParameter_49}
    B {from=MaterialExpressionMultiply_68}
   ]
  }
  MaterialExpressionAdd_24 {
   class=MaterialExpressionAdd
   inputs [
    A {from=MaterialExpressionAbs_3}
    B {from=MaterialExpressionMultiply_69}
   ]
  }
  MaterialExpressionAdd_25 {
   class=MaterialExpressionAdd
   inputs [
    A {from=MaterialExpressionMultiply_89}
    B {from=MaterialExpressionMultiply_90}
   ]
  }
  MaterialExpressionAdd_26 {
   class=MaterialExpressionAdd
   inputs [
    A {from=MaterialExpressionMultiply_70}
    B {from=MaterialExpressionMultiply_92}
   ]
  }
  MaterialExpressionAdd_27 {
   class=MaterialExpressionAdd
   inputs [
    A {from=MaterialExpressionAdd_14}
    B {from=MaterialExpressionMultiply_97}
   ]
  }
  MaterialExpressionAdd_29 {
   class=MaterialExpressionAdd
   inputs [
    A {from=MaterialExpressionComponentMask_10}
    B {from=MaterialExpressionScalarParameter_93}
   ]
   properties {constB=-80}
  }
  MaterialExpressionAdd_30 {
   class=MaterialExpressionAdd
   inputs [
    A {from=MaterialExpressionComponentMask_10}
    B {from=""}
   ]
   properties {constB=-5}
  }
  MaterialExpressionAdd_7 {
   class=MaterialExpressionAdd
   inputs [
    A {from=MaterialExpressionMultiply_20}
    B {from=MaterialExpressionLinearInterpolate_8}
   ]
  }
  MaterialExpressionAdd_8 {
   class=MaterialExpressionAdd
   inputs [
    A {from=MaterialExpressionNamedRerouteUsage_6}
    B {from=MaterialExpressionMultiply_9}
   ]
  }
  MaterialExpressionComponentMask_0 {class=MaterialExpressionComponentMask, inputs [None {from=MaterialExpressionAbs_2}], properties {r=true}}
  MaterialExpressionComponentMask_1 {class=MaterialExpressionComponentMask, inputs [None {from=MaterialExpressionNamedRerouteUsage_7}], properties {b=true}}
  MaterialExpressionComponentMask_10 {class=MaterialExpressionComponentMask, inputs [None {from=MaterialExpressionNamedRerouteUsage_67}], properties {b=true}}
  MaterialExpressionComponentMask_12 {class=MaterialExpressionComponentMask, inputs [None {from=MaterialExpressionNamedRerouteUsage_34}], properties {b=true}}
  MaterialExpressionComponentMask_14 {class=MaterialExpressionComponentMask, inputs [None {from=MaterialExpressionNamedRerouteUsage_7}], properties {g=true}}
  MaterialExpressionComponentMask_2 {class=MaterialExpressionComponentMask, inputs [None {from=MaterialExpressionNamedRerouteUsage_21}], properties {g=true}}
  MaterialExpressionComponentMask_3 {class=MaterialExpressionComponentMask, inputs [None {from=MaterialExpressionNamedRerouteUsage_19}], properties {b=true}}
  MaterialExpressionComponentMask_5 {class=MaterialExpressionComponentMask, inputs [None {from=MaterialExpressionNamedRerouteUsage_21}], properties {b=true}}
  MaterialExpressionComponentMask_8 {class=MaterialExpressionComponentMask, inputs [None {from=MaterialExpressionNamedRerouteUsage_63}], properties {g=true, r=true}}
  MaterialExpressionComposite_0 {
   class=MaterialExpressionComposite
   inputs [
    Fresnel {from=MaterialExpressionOneMinus_0, output=""}
    Alpha {from=MaterialExpressionMultiply_96, output=""}
    ExponentFactor {from=MaterialExpressionMax_0, output=""}
   ]
   properties {subgraphName=Thinfilm}
  }
  MaterialExpressionComposite_1 {class=MaterialExpressionComposite, inputs [], properties {subgraphName=LowerPantsGradient}}
  MaterialExpressionComposite_3 {class=MaterialExpressionComposite, inputs [], properties {subgraphName=NeckGradient}}
  MaterialExpressionComposite_5 {class=MaterialExpressionComposite, inputs [], properties {subgraphName=PantsSeamGradient}}
  MaterialExpressionConstant3Vector_0 {class=MaterialExpressionConstant3Vector, properties {constant=<rgba 0.02 0.0052 0.02 1>}}
  MaterialExpressionConstant3Vector_1 {class=MaterialExpressionConstant3Vector, properties {constant=<rgba 0 0 150 0>}}
  MaterialExpressionConstant3Vector_2 {class=MaterialExpressionConstant3Vector, properties {constant=<rgba 0 -45 10 0>}}
  MaterialExpressionConstant_1 {class=MaterialExpressionConstant, properties {r=26}}
  MaterialExpressionConstant_11 {class=MaterialExpressionConstant}
  MaterialExpressionConstant_13 {class=MaterialExpressionConstant}
  MaterialExpressionConstant_2 {class=MaterialExpressionConstant}
  MaterialExpressionConstant_3 {class=MaterialExpressionConstant}
  MaterialExpressionConstant_32 {class=MaterialExpressionConstant, properties {r=105}}
  MaterialExpressionConstant_33 {class=MaterialExpressionConstant, properties {r=-30}}
  MaterialExpressionConstant_34 {class=MaterialExpressionConstant, properties {r=0.4}}
  MaterialExpressionConstant_35 {class=MaterialExpressionConstant, properties {r=-0.27}}
  MaterialExpressionConstant_36 {class=MaterialExpressionConstant, properties {r=6}}
  MaterialExpressionConstant_4 {class=MaterialExpressionConstant, properties {r=0.6}}
  MaterialExpressionConstant_5 {class=MaterialExpressionConstant, properties {r=1.7}}
  MaterialExpressionConstant_6 {class=MaterialExpressionConstant}
  MaterialExpressionCurveAtlasRowParameter_0 {
   class=MaterialExpressionCurveAtlasRowParameter
   inputs [CurveTime {from=MaterialExpressionAdd_0, output=""}]
   properties {
    atlas=/Game/Characters/Player/Female/Medium/Bodies/F_MED_ChicleVeil/Textures/CA_ChicleVeil
    curve=/Game/Characters/Player/Female/Medium/Bodies/F_MED_ChicleVeil/Textures/C_ChicleVeil
    group=06_Thin Film
    parameterName=ThinFilm1
   }
  }
  MaterialExpressionDesaturation_0 {
   class=MaterialExpressionDesaturation
   inputs [
    None {from=MaterialExpressionMaterialFunctionCall_19, output=Swirl}
    Fraction {from=MaterialExpressionMultiply_94, output=""}
   ]
   properties {}
  }
  MaterialExpressionDistance_2 {
   class=MaterialExpressionDistance
   inputs [
    A {from=MaterialExpressionComponentMask_8, output=""}
    B {from=MaterialExpressionConstant_13, output=""}
   ]
   properties {}
  }
  MaterialExpressionDivide_0 {
   class=MaterialExpressionDivide
   inputs [
    A {from=MaterialExpressionAdd_12}
    B {from=""}
   ]
   properties {constB=20}
  }
  MaterialExpressionDivide_11 {
   class=MaterialExpressionDivide
   inputs [
    A {from=MaterialExpressionDistance_2}
    B {from=""}
   ]
   properties {constB=36}
  }
  MaterialExpressionDivide_12 {
   class=MaterialExpressionDivide
   inputs [
    A {from=MaterialExpressionMaterialFunctionCall_16, output=B}
    B {from=""}
   ]
   properties {constB=108}
  }
  MaterialExpressionDivide_13 {
   class=MaterialExpressionDivide
   inputs [
    A {from=MaterialExpressionAbs_5}
    B {from=""}
   ]
   properties {constB=35}
  }
  MaterialExpressionDivide_14 {
   class=MaterialExpressionDivide
   inputs [
    A {from=MaterialExpressionAdd_30}
    B {from=""}
   ]
   properties {constB=5}
  }
  MaterialExpressionDivide_5 {
   class=MaterialExpressionDivide
   inputs [
    A {from=MaterialExpressionAdd_24}
    B {from=MaterialExpressionScalarParameter_63}
   ]
   properties {constB=13}
  }
  MaterialExpressionDivide_6 {
   class=MaterialExpressionDivide
   inputs [
    A {from=MaterialExpressionAdd_21}
    B {from=""}
   ]
   properties {constB=10}
  }
  MaterialExpressionDivide_7 {
   class=MaterialExpressionDivide
   inputs [
    A {from=MaterialExpressionSubtract_6}
    B {from=MaterialExpressionScalarParameter_67}
   ]
   properties {constB=5}
  }
  MaterialExpressionDivide_8 {
   class=MaterialExpressionDivide
   inputs [
    A {from=MaterialExpressionAbs_4}
    B {from=""}
   ]
   properties {constB=10}
  }
  MaterialExpressionFloor_4 {class=MaterialExpressionFloor, inputs [None {from=MaterialExpressionDivide_11, output=""}], properties {}}
  MaterialExpressionFloor_5 {class=MaterialExpressionFloor, inputs [None {from=MaterialExpressionDivide_12, output=""}], properties {}}
  MaterialExpressionGetMaterialAttributes_0 {
   class=MaterialExpressionGetMaterialAttributes
   inputs [None {from=MaterialExpressionMaterialFunctionCall_0, output=Result}]
   properties {attributeGetTypes [#1, #2, #3]}
  }
  MaterialExpressionLinearInterpolate_1 {
   class=MaterialExpressionLinearInterpolate
   inputs [
    A {from=MaterialExpressionAdd_7}
    B {from=MaterialExpressionMultiply_98}
    Alpha {from=MaterialExpressionMultiply_16}
   ]
  }
  MaterialExpressionLinearInterpolate_12 {
   class=MaterialExpressionLinearInterpolate
   inputs [
    A {from=MaterialExpressionScalarParameter_48}
    B {from=MaterialExpressionScalarParameter_66}
    Alpha {from=MaterialExpressionSaturate_2}
   ]
  }
  MaterialExpressionLinearInterpolate_14 {
   class=MaterialExpressionLinearInterpolate
   inputs [
    A {from=MaterialExpressionScalarParameter_8}
    B {from=MaterialExpressionScalarParameter_72}
    Alpha {from=MaterialExpressionNamedRerouteUsage_47}
   ]
  }
  MaterialExpressionLinearInterpolate_16 {
   class=MaterialExpressionLinearInterpolate
   inputs [
    A {from=MaterialExpressionScalarParameter_75}
    B {from=MaterialExpressionScalarParameter_76}
    Alpha {from=MaterialExpressionNamedRerouteUsage_54}
   ]
  }
  MaterialExpressionLinearInterpolate_2 {
   class=MaterialExpressionLinearInterpolate
   inputs [
    A {from=MaterialExpressionNamedRerouteUsage_10}
    B {from=MaterialExpressionScalarParameter_7}
    Alpha {from=MaterialExpressionNamedRerouteUsage_15}
   ]
  }
  MaterialExpressionLinearInterpolate_8 {
   class=MaterialExpressionLinearInterpolate
   inputs [
    A {from=MaterialExpressionLinearInterpolate_9}
    B {from=MaterialExpressionConstant3Vector_0}
    Alpha {from=MaterialExpressionNamedRerouteUsage_28}
   ]
  }
  MaterialExpressionLinearInterpolate_9 {
   class=MaterialExpressionLinearInterpolate
   inputs [
    A {from=MaterialExpressionNamedRerouteUsage_0}
    B {from=MaterialExpressionVectorParameter_5, output=RGB}
    Alpha {from=MaterialExpressionMultiply_79}
   ]
  }
  MaterialExpressionMaterialFunctionCall_0 {
   class=MaterialExpressionMaterialFunctionCall
   inputs [
    Overide Texture Set UVs {from=""}
    Override Tech Art Mask UVs {from=""}
    Override Basic Emissive UVs {from=""}
   ]
   properties {materialFunction=&MF_CharacterParent_2023_PreFX}
  }
  MaterialExpressionMaterialFunctionCall_1 {
   class=MaterialExpressionMaterialFunctionCall
   inputs [
    Material {from=MaterialExpressionNamedRerouteUsage_2}
    M {from=MaterialExpressionMaterialFunctionCall_0, output=M}
    Pre Skinned Local Position {from=MaterialExpressionMaterialFunctionCall_0, output=M}
    Distance Blend {from=MaterialExpressionMaterialFunctionCall_0, output=M}
    Fresnel {from=MaterialExpressionMaterialFunctionCall_0, output=M}
   ]
   properties {materialFunction=&MF_CharacterParent_2023_PostFX}
  }
  MaterialExpressionMaterialFunctionCall_16 {
   class=MaterialExpressionMaterialFunctionCall
   inputs [Float3 {from=MaterialExpressionNamedRerouteUsage_63}]
   properties {materialFunction=&BreakOutFloat3Components}
  }
  MaterialExpressionMaterialFunctionCall_18 {
   class=MaterialExpressionMaterialFunctionCall
   inputs [
    X {from=MaterialExpressionMaterialFunctionCall_16, output=R}
    Y {from=MaterialExpressionMaterialFunctionCall_16, output=R}
    Z {from=MaterialExpressionAdd_25}
   ]
   properties {materialFunction=&MakeFloat3}
  }
  MaterialExpressionMaterialFunctionCall_19 {
   class=MaterialExpressionMaterialFunctionCall
   inputs [
    FresnelMax {from=MaterialExpressionMax_4}
    SwirlColorBias {from=MaterialExpressionAdd_27}
    SwirlFresnel {from=MaterialExpressionMultiply_100}
    SwirlColorBalanceOverride {from=MaterialExpressionSmoothStep_3}
    SwirlColorBalanceOverrideMask {from=MaterialExpressionReroute_5, output=Output}
   ]
   properties {materialFunction=&/MF_ChicleVeil_TA}
  }
  MaterialExpressionMaterialFunctionCall_2 {
   class=MaterialExpressionMaterialFunctionCall
   inputs [Float3 {from=MaterialExpressionNamedRerouteUsage_5}]
   properties {materialFunction=&BreakOutFloat3Components}
  }
  MaterialExpressionMaterialFunctionCall_20 {
   class=MaterialExpressionMaterialFunctionCall
   inputs [
    In {from=MaterialExpressionMultiply_74}
    Contrast {from=MaterialExpressionScalarParameter_62}
   ]
   properties {materialFunction=&CheapContrast}
  }
  MaterialExpressionMaterialFunctionCall_3 {
   class=MaterialExpressionMaterialFunctionCall
   inputs [
    X {from=MaterialExpressionMultiply_13}
    Y {from=MaterialExpressionMaterialFunctionCall_2, output=G}
    Z {from=MaterialExpressionMaterialFunctionCall_2, output=G}
   ]
   properties {materialFunction=&MakeFloat3}
  }
  MaterialExpressionMaterialFunctionCall_4 {
   class=MaterialExpressionMaterialFunctionCall
   inputs [
    Material {from=MaterialExpressionSetMaterialAttributes_1}
    Alpha {from=MaterialExpressionNamedRerouteUsage_8}
    Reactive {from=MaterialExpressionScalarParameter_10}
    ColorProgress {from=MaterialExpressionScalarParameter_10}
    SpellSpeed {from=MaterialExpressionConstant_6}
   ]
   properties {materialFunction=/BRCosmetics/Weapons/FORT_Melee/Pickaxe_ChicleVeil/Materials/MF_ChicleVeil_Crystal}
  }
  MaterialExpressionMaterialFunctionCall_5 {
   class=MaterialExpressionMaterialFunctionCall
   inputs [Float3 {from=MaterialExpressionNamedRerouteUsage_17}]
   properties {materialFunction=&BreakOutFloat3Components}
  }
  MaterialExpressionMaterialFunctionCall_6 {
   class=MaterialExpressionMaterialFunctionCall
   inputs [
    Value_Morning {from=MaterialExpressionScalarParameter_59}
    Value_Day {from=MaterialExpressionScalarParameter_50}
    Value_Evening {from=MaterialExpressionScalarParameter_60}
    Value_Night {from=MaterialExpressionScalarParameter_64}
   ]
   properties {materialFunction=&Scalar_Time_of_Day}
  }
  MaterialExpressionMaterialFunctionCall_7 {
   class=MaterialExpressionMaterialFunctionCall
   inputs [
    X {from=MaterialExpressionMaterialFunctionCall_5, output=R}
    Y {from=MaterialExpressionAdd_1}
    Z {from=MaterialExpressionMultiply_19}
   ]
   properties {materialFunction=&MakeFloat3}
  }
  MaterialExpressionMax_0 {
   class=MaterialExpressionMax
   inputs [
    A {from=MaterialExpressionNamedRerouteUsage_38}
    B {from=MaterialExpressionScalarParameter_14}
   ]
   properties {constB=0.1}
  }
  MaterialExpressionMax_2 {
   class=MaterialExpressionMax
   inputs [
    A {from=MaterialExpressionMultiply_91}
    B {from=MaterialExpressionOneMinus_23}
   ]
  }
  MaterialExpressionMax_4 {
   class=MaterialExpressionMax
   inputs [
    A {from=MaterialExpressionMultiply_53}
    B {from=MaterialExpressionMultiply_48}
   ]
   properties {constB=0}
  }
  MaterialExpressionMin_0 {
   class=MaterialExpressionMin
   inputs [
    A {from=MaterialExpressionNamedRerouteUsage_61, output=""}
    B {from=MaterialExpressionPower_7, output=""}
   ]
   properties {}
  }
  MaterialExpressionMin_2 {
   class=MaterialExpressionMin
   inputs [
    A {from=MaterialExpressionNamedRerouteUsage_61, output=""}
    B {from=MaterialExpressionPower_15, output=""}
   ]
   properties {}
  }
  MaterialExpressionMultiply_0 {
   class=MaterialExpressionMultiply
   inputs [
    A {from=MaterialExpressionMultiply_2}
    B {from=MaterialExpressionPower_0}
   ]
  }
  MaterialExpressionMultiply_1 {
   class=MaterialExpressionMultiply
   inputs [
    A {from=MaterialExpressionOneMinus_4}
    B {from=MaterialExpressionComponentMask_2}
   ]
  }
  MaterialExpressionMultiply_10 {
   class=MaterialExpressionMultiply
   inputs [
    A {from=MaterialExpressionLinearInterpolate_16}
    B {from=MaterialExpressionNamedRerouteUsage_11}
   ]
   properties {constB=4}
  }
  MaterialExpressionMultiply_100 {
   class=MaterialExpressionMultiply
   inputs [
    A {from=MaterialExpressionNamedRerouteDeclaration_41}
    B {from=MaterialExpressionNamedRerouteUsage_66}
   ]
  }
  MaterialExpressionMultiply_101 {
   class=MaterialExpressionMultiply
   inputs [
    A {from=MaterialExpressionPower_18}
    B {from=MaterialExpressionSaturate_9}
   ]
  }
  MaterialExpressionMultiply_102 {
   class=MaterialExpressionMultiply
   inputs [
    A {from=MaterialExpressionMultiply_101}
    B {from=MaterialExpressionMultiply_8}
   ]
  }
  MaterialExpressionMultiply_11 {
   class=MaterialExpressionMultiply
   inputs [
    A {from=MaterialExpressionNamedRerouteUsage_12}
    B {from=MaterialExpressionMultiply_102}
   ]
  }
  MaterialExpressionMultiply_13 {
   class=MaterialExpressionMultiply
   inputs [
    A {from=MaterialExpressionMaterialFunctionCall_2, output=R}
    B {from=MaterialExpressionConstant_5}
   ]
   properties {constB=1.8}
  }
  MaterialExpressionMultiply_14 {
   class=MaterialExpressionMultiply
   inputs [
    A {from=MaterialExpressionMultiply_7}
    B {from=MaterialExpressionScalarParameter_6}
   ]
  }
  MaterialExpressionMultiply_16 {
   class=MaterialExpressionMultiply
   inputs [
    A {from=MaterialExpressionMin_2}
    B {from=MaterialExpressionScalarParameter_77}
   ]
  }
  MaterialExpressionMultiply_17 {
   class=MaterialExpressionMultiply
   inputs [
    A {from=MaterialExpressionMaterialFunctionCall_5, output=G}
    B {from=MaterialExpressionConstant_36}
   ]
   properties {constB=8}
  }
  MaterialExpressionMultiply_18 {
   class=MaterialExpressionMultiply
   inputs [
    A {from=MaterialExpressionScalarParameter_77}
    B {from=MaterialExpressionMultiply_29}
   ]
  }
  MaterialExpressionMultiply_19 {
   class=MaterialExpressionMultiply
   inputs [
    A {from=MaterialExpressionMaterialFunctionCall_5, output=B}
    B {from=""}
   ]
   properties {constB=0.2}
  }
  MaterialExpressionMultiply_2 {
   class=MaterialExpressionMultiply
   inputs [
    A {from=MaterialExpressionCurveAtlasRowParameter_0}
    B {from=MaterialExpressionScalarParameter_2}
   ]
  }
  MaterialExpressionMultiply_20 {
   class=MaterialExpressionMultiply
   inputs [
    A {from=MaterialExpressionNamedRerouteUsage_44}
    B {from=MaterialExpressionScalarParameter_78}
   ]
  }
  MaterialExpressionMultiply_22 {
   class=MaterialExpressionMultiply
   inputs [
    A {from=MaterialExpressionPower_10}
    B {from=MaterialExpressionNamedRerouteUsage_38}
   ]
  }
  MaterialExpressionMultiply_23 {
   class=MaterialExpressionMultiply
   inputs [
    A {from=MaterialExpressionMaterialFunctionCall_5, output=B}
    B {from=MaterialExpressionConstant_35}
   ]
   properties {constB=-0.5}
  }
  MaterialExpressionMultiply_25 {
   class=MaterialExpressionMultiply
   inputs [
    A {from=MaterialExpressionMultiply_40}
    B {from=MaterialExpressionComponentMask_2}
   ]
  }
  MaterialExpressionMultiply_27 {
   class=MaterialExpressionMultiply
   inputs [
    A {from=MaterialExpressionReroute_0, output=Output}
    B {from=MaterialExpressionMultiply_1}
   ]
  }
  MaterialExpressionMultiply_29 {
   class=MaterialExpressionMultiply
   inputs [
    A {from=MaterialExpressionMultiply_51}
    B {from=MaterialExpressionLinearInterpolate_14}
   ]
   properties {constB=0.5}
  }
  MaterialExpressionMultiply_3 {
   class=MaterialExpressionMultiply
   inputs [
    A {from=MaterialExpressionScalarParameter_3}
    B {from=MaterialExpressionPower_1}
   ]
  }
  MaterialExpressionMultiply_34 {
   class=MaterialExpressionMultiply
   inputs [
    A {from=MaterialExpressionScalarParameter_4}
    B {from=MaterialExpressionReroute_5, output=ExponentFactor}
   ]
  }
  MaterialExpressionMultiply_4 {
   class=MaterialExpressionMultiply
   inputs [
    A {from=MaterialExpressionMultiply_0}
    B {from=MaterialExpressionReroute_6}
   ]
  }
  MaterialExpressionMultiply_40 {
   class=MaterialExpressionMultiply
   inputs [
    A {from=MaterialExpressionStep_2}
    B {from=MaterialExpressionOneMinus_3}
   ]
  }
  MaterialExpressionMultiply_46 {
   class=MaterialExpressionMultiply
   inputs [
    A {from=MaterialExpressionMultiply_65}
    B {from=""}
   ]
   properties {constB=-1}
  }
  MaterialExpressionMultiply_48 {
   class=MaterialExpressionMultiply
   inputs [
    A {from=MaterialExpressionPower_8}
    B {from=MaterialExpressionScalarParameter_31}
   ]
  }
  MaterialExpressionMultiply_5 {
   class=MaterialExpressionMultiply
   inputs [
    A {from=MaterialExpressionOneMinus_1}
    B {from=MaterialExpressionScalarParameter_9}
   ]
  }
  MaterialExpressionMultiply_51 {
   class=MaterialExpressionMultiply
   inputs [
    A {from=MaterialExpressionNamedRerouteUsage_16}
    B {from=MaterialExpressionMin_0}
   ]
  }
  MaterialExpressionMultiply_53 {
   class=MaterialExpressionMultiply
   inputs [
    A {from=MaterialExpressionNamedRerouteUsage_13}
    B {from=MaterialExpressionScalarParameter_44}
   ]
  }
  MaterialExpressionMultiply_6 {
   class=MaterialExpressionMultiply
   inputs [
    A {from=MaterialExpressionComponentMask_2}
    B {from=MaterialExpressionComponentMask_5}
   ]
  }
  MaterialExpressionMultiply_65 {
   class=MaterialExpressionMultiply
   inputs [
    A {from=MaterialExpressionNamedRerouteUsage_32}
    B {from=MaterialExpressionScalarParameter_47}
   ]
  }
  MaterialExpressionMultiply_66 {
   class=MaterialExpressionMultiply
   inputs [
    A {from=MaterialExpressionPower_16}
    B {from=MaterialExpressionOneMinus_12}
   ]
  }
  MaterialExpressionMultiply_67 {
   class=MaterialExpressionMultiply
   inputs [
    A {from=MaterialExpressionNamedRerouteUsage_37}
    B {from=MaterialExpressionPower_9}
   ]
  }
  MaterialExpressionMultiply_68 {
   class=MaterialExpressionMultiply
   inputs [
    A {from=MaterialExpressionSaturate_1}
    B {from=MaterialExpressionScalarParameter_51}
   ]
   properties {constB=12}
  }
  MaterialExpressionMultiply_69 {
   class=MaterialExpressionMultiply
   inputs [
    A {from=MaterialExpressionSaturate_1}
    B {from=MaterialExpressionScalarParameter_53}
   ]
   properties {constB=-10}
  }
  MaterialExpressionMultiply_7 {
   class=MaterialExpressionMultiply
   inputs [
    A {from=MaterialExpressionOneMinus_2}
    B {from=MaterialExpressionComponentMask_5}
   ]
  }
  MaterialExpressionMultiply_70 {
   class=MaterialExpressionMultiply
   inputs [
    A {from=MaterialExpressionLinearInterpolate_12}
    B {from=MaterialExpressionSubtract_5}
   ]
  }
  MaterialExpressionMultiply_73 {
   class=MaterialExpressionMultiply
   inputs [
    A {from=MaterialExpressionOneMinus_15}
    B {from=MaterialExpressionPower_20}
   ]
  }
  MaterialExpressionMultiply_74 {
   class=MaterialExpressionMultiply
   inputs [
    A {from=MaterialExpressionMultiply_73}
    B {from=MaterialExpressionScalarParameter_56}
   ]
  }
  MaterialExpressionMultiply_79 {
   class=MaterialExpressionMultiply
   inputs [
    A {from=MaterialExpressionPower_17}
    B {from=MaterialExpressionMultiply_67}
   ]
  }
  MaterialExpressionMultiply_8 {
   class=MaterialExpressionMultiply
   inputs [
    A {from=MaterialExpressionReroute_4, output=Output}
    B {from=MaterialExpressionMaterialFunctionCall_19, output=FesnelGlow}
   ]
  }
  MaterialExpressionMultiply_80 {
   class=MaterialExpressionMultiply
   inputs [
    A {from=MaterialExpressionPower_19}
    B {from=MaterialExpressionPower_27}
   ]
  }
  MaterialExpressionMultiply_86 {
   class=MaterialExpressionMultiply
   inputs [
    A {from=MaterialExpressionPower_3}
    B {from=MaterialExpressionScalarParameter_46}
   ]
  }
  MaterialExpressionMultiply_87 {
   class=MaterialExpressionMultiply
   inputs [
    A {from=MaterialExpressionNamedRerouteUsage_59}
    B {from=MaterialExpressionSmoothStep_4}
   ]
  }
  MaterialExpressionMultiply_88 {
   class=MaterialExpressionMultiply
   inputs [
    A {from=MaterialExpressionMultiply_80}
    B {from=MaterialExpressionScalarParameter_58}
   ]
  }
  MaterialExpressionMultiply_89 {
   class=MaterialExpressionMultiply
   inputs [
    A {from=MaterialExpressionSubtract_7}
    B {from=""}
   ]
   properties {constB=1.5}
  }
  MaterialExpressionMultiply_9 {
   class=MaterialExpressionMultiply
   inputs [
    A {from=MaterialExpressionAdd_19}
    B {from=MaterialExpressionNamedRerouteUsage_9}
   ]
  }
  MaterialExpressionMultiply_90 {
   class=MaterialExpressionMultiply
   inputs [
    A {from=MaterialExpressionSaturate_6}
    B {from=""}
   ]
   properties {constB=100}
  }
  MaterialExpressionMultiply_91 {
   class=MaterialExpressionMultiply
   inputs [
    A {from=MaterialExpressionSphereMask_1}
    B {from=MaterialExpressionOneMinus_21}
   ]
  }
  MaterialExpressionMultiply_92 {
   class=MaterialExpressionMultiply
   inputs [
    A {from=MaterialExpressionNamedRerouteUsage_14}
    B {from=MaterialExpressionScalarParameter_69}
   ]
  }
  MaterialExpressionMultiply_93 {
   class=MaterialExpressionMultiply
   inputs [
    A {from=MaterialExpressionSubtract_9}
    B {from=""}
   ]
   properties {constB=0.15}
  }
  MaterialExpressionMultiply_94 {
   class=MaterialExpressionMultiply
   inputs [
    A {from=MaterialExpressionScalarParameter_71}
    B {from=MaterialExpressionNamedRerouteUsage_46}
   ]
  }
  MaterialExpressionMultiply_95 {
   class=MaterialExpressionMultiply
   inputs [
    A {from=MaterialExpressionOneMinus_21}
    B {from=MaterialExpressionOneMinus_22}
   ]
  }
  MaterialExpressionMultiply_96 {
   class=MaterialExpressionMultiply
   inputs [
    A {from=MaterialExpressionMultiply_22}
    B {from=MaterialExpressionNamedRerouteUsage_53}
   ]
  }
  MaterialExpressionMultiply_97 {
   class=MaterialExpressionMultiply
   inputs [
    A {from=MaterialExpressionMultiply_99}
    B {from=MaterialExpressionScalarParameter_90}
   ]
  }
  MaterialExpressionMultiply_98 {
   class=MaterialExpressionMultiply
   inputs [
    A {from=MaterialExpressionMultiply_10}
    B {from=MaterialExpressionNamedRerouteUsage_16}
   ]
   properties {constB=4}
  }
  MaterialExpressionMultiply_99 {
   class=MaterialExpressionMultiply
   inputs [
    A {from=MaterialExpressionPower_13}
    B {from=MaterialExpressionNamedRerouteDeclaration_41}
   ]
  }
  MaterialExpressionNamedRerouteDeclaration_0 {
   class=MaterialExpressionNamedRerouteDeclaration
   inputs [None {from=MaterialExpressionGetMaterialAttributes_0, output=Base Color}]
   properties {name=InBaseColor, nodeColor=<rgba 1 0 0.7031 1>}
  }
  MaterialExpressionNamedRerouteDeclaration_1 {
   class=MaterialExpressionNamedRerouteDeclaration
   inputs [None {from=MaterialExpressionGetMaterialAttributes_0, output=Emissive Color}]
   properties {name=InEmissiveColor, nodeColor=<rgba 0 1 0.1562 1>}
  }
  MaterialExpressionNamedRerouteDeclaration_10 {
   class=MaterialExpressionNamedRerouteDeclaration
   inputs [None {from=MaterialExpressionGetMaterialAttributes_0, output=Roughness}]
   properties {name=InRoughness, nodeColor=<rgba 0 0.5312 1 1>}
  }
  MaterialExpressionNamedRerouteDeclaration_11 {
   class=MaterialExpressionNamedRerouteDeclaration
   inputs [None {from=MaterialExpressionQualitySwitch_2}]
   properties {name=PersistentSwirlShapes, nodeColor=<rgba 1 0 0.5156 1>}
  }
  MaterialExpressionNamedRerouteDeclaration_12 {
   class=MaterialExpressionNamedRerouteDeclaration
   inputs [None {from=MaterialExpressionMultiply_25}]
   properties {name=TattooOutline, nodeColor=<rgba 0.5938 0 1 1>}
  }
  MaterialExpressionNamedRerouteDeclaration_14 {
   class=MaterialExpressionNamedRerouteDeclaration
   inputs [None {from=MaterialExpressionMultiply_27}]
   properties {name=FabricMask, nodeColor=<rgba 0 1 0.4375 1>}
  }
  MaterialExpressionNamedRerouteDeclaration_15 {
   class=MaterialExpressionNamedRerouteDeclaration
   inputs [None {from=MaterialExpressionMaterialFunctionCall_19, output=SwirlAlpha}]
   properties {name=SwirlAlpha, nodeColor=<rgba 1 0.4219 0 1>}
  }
  MaterialExpressionNamedRerouteDeclaration_16 {
   class=MaterialExpressionNamedRerouteDeclaration
   inputs [None {from=MaterialExpressionOneMinus_20}]
   properties {name=PantsGradient, nodeColor=<rgba 1 0.0234 0 1>}
  }
  MaterialExpressionNamedRerouteDeclaration_18 {
   class=MaterialExpressionNamedRerouteDeclaration
   inputs [None {from=MaterialExpressionMultiply_95}]
   properties {name=PantsMask, nodeColor=<rgba 0 1 0.4844 1>}
  }
  MaterialExpressionNamedRerouteDeclaration_19 {
   class=MaterialExpressionNamedRerouteDeclaration
   inputs [None {from=MaterialExpressionMultiply_14}]
   properties {name=TattooMask, nodeColor=<rgba 0 0.5547 1 1>}
  }
  MaterialExpressionNamedRerouteDeclaration_2 {
   class=MaterialExpressionNamedRerouteDeclaration
   inputs [None {from=MaterialExpressionGetMaterialAttributes_0, output=MaterialAttributes}]
   properties {name=InMA, nodeColor=<rgba 1 0 0.75 1>}
  }
  MaterialExpressionNamedRerouteDeclaration_20 {
   class=MaterialExpressionNamedRerouteDeclaration
   inputs [None {from=MaterialExpressionNamedRerouteUsage_59}]
   properties {name=SwirlBaseGradient, nodeColor=<rgba 1 0.0938 0 1>}
  }
  MaterialExpressionNamedRerouteDeclaration_23 {
   class=MaterialExpressionNamedRerouteDeclaration
   inputs [None {from=MaterialExpressionMaterialFunctionCall_19, output=FresnelNoise}]
   properties {name=FresnelNoise, nodeColor=<rgba 0.7813 0 1 1>}
  }
  MaterialExpressionNamedRerouteDeclaration_24 {
   class=MaterialExpressionNamedRerouteDeclaration
   inputs [None {from=MaterialExpressionDivide_7}]
   properties {name=SwirlBaseUpperGradient, nodeColor=<rgba 0 1 0.1562 1>}
  }
  MaterialExpressionNamedRerouteDeclaration_25 {
   class=MaterialExpressionNamedRerouteDeclaration
   inputs [None {from=MaterialExpressionMultiply_6}]
   properties {name=UpperBackGradient, nodeColor=<rgba 0.8281 1 0 1>}
  }
  MaterialExpressionNamedRerouteDeclaration_26 {
   class=MaterialExpressionNamedRerouteDeclaration
   inputs [None {from=MaterialExpressionQualitySwitch_1}]
   properties {name=ThinFilm, nodeColor=<rgba 0 1 0.3906 1>}
  }
  MaterialExpressionNamedRerouteDeclaration_27 {
   class=MaterialExpressionNamedRerouteDeclaration
   inputs [None {from=MaterialExpressionMultiply_11}]
   properties {name=FresnelGlow, nodeColor=<rgba 0 1 1 1>}
  }
  MaterialExpressionNamedRerouteDeclaration_29 {
   class=MaterialExpressionNamedRerouteDeclaration
   inputs [None {from=MaterialExpressionMax_2}]
   properties {name=ThinFilmMask, nodeColor=<rgba 0 0.4844 1 1>}
  }
  MaterialExpressionNamedRerouteDeclaration_3 {
   class=MaterialExpressionNamedRerouteDeclaration
   inputs [None {from=MaterialExpressionMaterialFunctionCall_4, output=Result}]
   properties {name=OutMA, nodeColor=<rgba 0 0.1562 1 1>}
  }
  MaterialExpressionNamedRerouteDeclaration_4 {
   class=MaterialExpressionNamedRerouteDeclaration
   inputs [None {from=MaterialExpressionMaterialFunctionCall_0, output=Tech Art Mask}]
   properties {name=InTAMask, nodeColor=<rgba 1 0.2812 0 1>}
  }
  MaterialExpressionNamedRerouteDeclaration_41 {
   class=MaterialExpressionNamedRerouteDeclaration
   inputs [None {from=MaterialExpressionMultiply_86}]
   properties {name=SwirlFresnel, nodeColor=<rgba 1 0 0.6562 1>}
  }
  MaterialExpressionNamedRerouteDeclaration_43 {
   class=MaterialExpressionNamedRerouteDeclaration
   inputs [None {from=MaterialExpressionMultiply_87}]
   properties {name=SwirlOpacity, nodeColor=<rgba 1 0.4219 0 1>}
  }
  MaterialExpressionNamedRerouteDeclaration_5 {
   class=MaterialExpressionNamedRerouteDeclaration
   inputs [None {from=MaterialExpressionMaterialFunctionCall_0, output=Fresnel}]
   properties {name=InFresnel, nodeColor=<rgba 1 0.3047 0 1>}
  }
  MaterialExpressionNamedRerouteDeclaration_6 {
   class=MaterialExpressionNamedRerouteDeclaration
   inputs [None {from=MaterialExpressionMaterialFunctionCall_0, output=Pre skinned Local Position}]
   properties {name=PreSkinPos, nodeColor=<rgba 1 0 0.2109 1>}
  }
  MaterialExpressionNamedRerouteDeclaration_7 {
   class=MaterialExpressionNamedRerouteDeclaration
   inputs [None {from=MaterialExpressionMaterialFunctionCall_6, output=Result}]
   properties {name=TODFactor, nodeColor=<rgba 0.1719 0 1 1>}
  }
  MaterialExpressionNamedRerouteDeclaration_8 {
   class=MaterialExpressionNamedRerouteDeclaration
   inputs [None {from=MaterialExpressionStep_0}]
   properties {name=GemMask, nodeColor=<rgba 1 0.0938 0 1>}
  }
  MaterialExpressionNamedRerouteDeclaration_9 {
   class=MaterialExpressionNamedRerouteDeclaration
   inputs [None {from=MaterialExpressionDesaturation_0}]
   properties {name=Swirl, nodeColor=<rgba 0 0.5078 1 1>}
  }
  MaterialExpressionNamedRerouteUsage_0 {class=MaterialExpressionNamedRerouteUsage}
  MaterialExpressionNamedRerouteUsage_1 {class=MaterialExpressionNamedRerouteUsage}
  MaterialExpressionNamedRerouteUsage_10 {class=MaterialExpressionNamedRerouteUsage}
  MaterialExpressionNamedRerouteUsage_11 {class=MaterialExpressionNamedRerouteUsage}
  MaterialExpressionNamedRerouteUsage_12 {class=MaterialExpressionNamedRerouteUsage}
  MaterialExpressionNamedRerouteUsage_13 {class=MaterialExpressionNamedRerouteUsage}
  MaterialExpressionNamedRerouteUsage_14 {class=MaterialExpressionNamedRerouteUsage}
  MaterialExpressionNamedRerouteUsage_15 {class=MaterialExpressionNamedRerouteUsage}
  MaterialExpressionNamedRerouteUsage_16 {class=MaterialExpressionNamedRerouteUsage}
  MaterialExpressionNamedRerouteUsage_17 {class=MaterialExpressionNamedRerouteUsage}
  MaterialExpressionNamedRerouteUsage_19 {class=MaterialExpressionNamedRerouteUsage}
  MaterialExpressionNamedRerouteUsage_2 {class=MaterialExpressionNamedRerouteUsage}
  MaterialExpressionNamedRerouteUsage_21 {class=MaterialExpressionNamedRerouteUsage}
  MaterialExpressionNamedRerouteUsage_22 {class=MaterialExpressionNamedRerouteUsage}
  MaterialExpressionNamedRerouteUsage_27 {class=MaterialExpressionNamedRerouteUsage}
  MaterialExpressionNamedRerouteUsage_28 {class=MaterialExpressionNamedRerouteUsage}
  MaterialExpressionNamedRerouteUsage_3 {class=MaterialExpressionNamedRerouteUsage}
  MaterialExpressionNamedRerouteUsage_31 {class=MaterialExpressionNamedRerouteUsage}
  MaterialExpressionNamedRerouteUsage_32 {class=MaterialExpressionNamedRerouteUsage}
  MaterialExpressionNamedRerouteUsage_33 {class=MaterialExpressionNamedRerouteUsage}
  MaterialExpressionNamedRerouteUsage_34 {class=MaterialExpressionNamedRerouteUsage}
  MaterialExpressionNamedRerouteUsage_35 {class=MaterialExpressionNamedRerouteUsage}
  MaterialExpressionNamedRerouteUsage_37 {class=MaterialExpressionNamedRerouteUsage}
  MaterialExpressionNamedRerouteUsage_38 {class=MaterialExpressionNamedRerouteUsage}
  MaterialExpressionNamedRerouteUsage_39 {class=MaterialExpressionNamedRerouteUsage}
  MaterialExpressionNamedRerouteUsage_4 {class=MaterialExpressionNamedRerouteUsage}
  MaterialExpressionNamedRerouteUsage_40 {class=MaterialExpressionNamedRerouteUsage}
  MaterialExpressionNamedRerouteUsage_41 {class=MaterialExpressionNamedRerouteUsage}
  MaterialExpressionNamedRerouteUsage_44 {class=MaterialExpressionNamedRerouteUsage}
  MaterialExpressionNamedRerouteUsage_45 {class=MaterialExpressionNamedRerouteUsage}
  MaterialExpressionNamedRerouteUsage_46 {class=MaterialExpressionNamedRerouteUsage}
  MaterialExpressionNamedRerouteUsage_47 {class=MaterialExpressionNamedRerouteUsage}
  MaterialExpressionNamedRerouteUsage_5 {class=MaterialExpressionNamedRerouteUsage}
  MaterialExpressionNamedRerouteUsage_53 {class=MaterialExpressionNamedRerouteUsage}
  MaterialExpressionNamedRerouteUsage_54 {class=MaterialExpressionNamedRerouteUsage}
  MaterialExpressionNamedRerouteUsage_56 {class=MaterialExpressionNamedRerouteUsage}
  MaterialExpressionNamedRerouteUsage_59 {class=MaterialExpressionNamedRerouteUsage}
  MaterialExpressionNamedRerouteUsage_6 {class=MaterialExpressionNamedRerouteUsage}
  MaterialExpressionNamedRerouteUsage_60 {class=MaterialExpressionNamedRerouteUsage}
  MaterialExpressionNamedRerouteUsage_61 {class=MaterialExpressionNamedRerouteUsage}
  MaterialExpressionNamedRerouteUsage_62 {class=MaterialExpressionNamedRerouteUsage}
  MaterialExpressionNamedRerouteUsage_63 {class=MaterialExpressionNamedRerouteUsage}
  MaterialExpressionNamedRerouteUsage_64 {class=MaterialExpressionNamedRerouteUsage}
  MaterialExpressionNamedRerouteUsage_66 {class=MaterialExpressionNamedRerouteUsage}
  MaterialExpressionNamedRerouteUsage_67 {class=MaterialExpressionNamedRerouteUsage}
  MaterialExpressionNamedRerouteUsage_7 {class=MaterialExpressionNamedRerouteUsage}
  MaterialExpressionNamedRerouteUsage_8 {class=MaterialExpressionNamedRerouteUsage}
  MaterialExpressionNamedRerouteUsage_9 {class=MaterialExpressionNamedRerouteUsage}
  MaterialExpressionOneMinus_0 {class=MaterialExpressionOneMinus, inputs [None {from=MaterialExpressionNamedRerouteUsage_4}]}
  MaterialExpressionOneMinus_1 {class=MaterialExpressionOneMinus, inputs [None {from=MaterialExpressionNamedRerouteUsage_1}]}
  MaterialExpressionOneMinus_10 {class=MaterialExpressionOneMinus, inputs [None {from=MaterialExpressionNamedRerouteUsage_62}]}
  MaterialExpressionOneMinus_12 {class=MaterialExpressionOneMinus, inputs [None {from=MaterialExpressionSmoothStep_0}]}
  MaterialExpressionOneMinus_15 {class=MaterialExpressionOneMinus, inputs [None {from=MaterialExpressionSmoothStep_7}]}
  MaterialExpressionOneMinus_19 {class=MaterialExpressionOneMinus, inputs [None {from=MaterialExpressionNamedRerouteUsage_56}]}
  MaterialExpressionOneMinus_2 {class=MaterialExpressionOneMinus, inputs [None {from=MaterialExpressionComponentMask_2}]}
  MaterialExpressionOneMinus_20 {class=MaterialExpressionOneMinus, inputs [None {from=MaterialExpressionMultiply_91}]}
  MaterialExpressionOneMinus_21 {class=MaterialExpressionOneMinus, inputs [None {from=MaterialExpressionFloor_4}]}
  MaterialExpressionOneMinus_22 {class=MaterialExpressionOneMinus, inputs [None {from=MaterialExpressionFloor_5}]}
  MaterialExpressionOneMinus_23 {class=MaterialExpressionOneMinus, inputs [None {from=MaterialExpressionNamedRerouteDeclaration_18}]}
  MaterialExpressionOneMinus_3 {class=MaterialExpressionOneMinus, inputs [None {from=MaterialExpressionNamedRerouteUsage_35}]}
  MaterialExpressionOneMinus_4 {class=MaterialExpressionOneMinus, inputs [None {from=MaterialExpressionMultiply_40}]}
  MaterialExpressionOneMinus_5 {class=MaterialExpressionOneMinus, inputs [None {from=MaterialExpressionSphereMask_0}]}
  MaterialExpressionOneMinus_8 {class=MaterialExpressionOneMinus, inputs [None {from=MaterialExpressionNamedRerouteUsage_33}]}
  MaterialExpressionPinBase_0 {
   class=MaterialExpressionPinBase
   inputs []
   properties {reroutePins [
     Fresnel {expression=&self:MaterialGraph_1.MaterialGraphNode_Composite_0.Thinfilm.MaterialExpressionReroute_3}
     Alpha {expression=&self:MaterialGraph_1.MaterialGraphNode_Composite_0.Thinfilm.MaterialExpressionReroute_4}
     ExponentFactor {expression=&self:MaterialGraph_1.MaterialGraphNode_Composite_0.Thinfilm.MaterialExpressionReroute_5}
    ]}
  }
  MaterialExpressionPinBase_1 {
   class=MaterialExpressionPinBase
   inputs [Return {from=MaterialExpressionMultiply_4, output=""}]
   properties {reroutePins [Return {expression=&self:MaterialGraph_1.MaterialGraphNode_Composite_0.Thinfilm.MaterialExpressionReroute_6}]}
  }
  MaterialExpressionPinBase_10 {class=MaterialExpressionPinBase, inputs [], properties {}}
  MaterialExpressionPinBase_11 {
   class=MaterialExpressionPinBase
   inputs [Output {from=MaterialExpressionSphereMask_2, output=""}]
   properties {reroutePins [Output {expression=&self:MaterialExpressionReroute_5}]}
  }
  MaterialExpressionPinBase_2 {class=MaterialExpressionPinBase, inputs [], properties {}}
  MaterialExpressionPinBase_3 {
   class=MaterialExpressionPinBase
   inputs [Output {from=MaterialExpressionPower_4, output=""}]
   properties {reroutePins [Output {expression=&self:MaterialExpressionReroute_0}]}
  }
  MaterialExpressionPinBase_6 {class=MaterialExpressionPinBase, inputs [], properties {}}
  MaterialExpressionPinBase_7 {
   class=MaterialExpressionPinBase
   inputs [Output {from=MaterialExpressionOneMinus_5, output=""}]
   properties {reroutePins [Output {expression=&self:MaterialExpressionReroute_4}]}
  }
  MaterialExpressionPower_0 {
   class=MaterialExpressionPower
   inputs [
    Base {from=MaterialExpressionReroute_1}
    Exp {from=MaterialExpressionMultiply_34}
   ]
  }
  MaterialExpressionPower_1 {
   class=MaterialExpressionPower
   inputs [
    Base {from=MaterialExpressionReroute_1}
    Exp {from=MaterialExpressionScalarParameter_0}
   ]
  }
  MaterialExpressionPower_10 {
   class=MaterialExpressionPower
   inputs [
    Base {from=MaterialExpressionOneMinus_10}
    Exp {from=MaterialExpressionScalarParameter_25}
   ]
   properties {constExponent=4}
  }
  MaterialExpressionPower_13 {
   class=MaterialExpressionPower
   inputs [
    Base {from=MaterialExpressionNamedRerouteUsage_64}
    Exp {from=MaterialExpressionScalarParameter_91}
   ]
  }
  MaterialExpressionPower_15 {
   class=MaterialExpressionPower
   inputs [
    Base {from=MaterialExpressionNamedRerouteUsage_60}
    Exp {from=MaterialExpressionScalarParameter_29}
   ]
   properties {constExponent=4}
  }
  MaterialExpressionPower_16 {
   class=MaterialExpressionPower
   inputs [
    Base {from=MaterialExpressionOneMinus_8}
    Exp {from=MaterialExpressionAdd_23}
   ]
   properties {constExponent=12}
  }
  MaterialExpressionPower_17 {
   class=MaterialExpressionPower
   inputs [
    Base {from=MaterialExpressionNamedRerouteUsage_40}
    Exp {from=MaterialExpressionScalarParameter_23}
   ]
   properties {constExponent=1}
  }
  MaterialExpressionPower_18 {
   class=MaterialExpressionPower
   inputs [
    Base {from=MaterialExpressionSaturate_8}
    Exp {from=MaterialExpressionScalarParameter_92}
   ]
   properties {constExponent=4}
  }
  MaterialExpressionPower_19 {
   class=MaterialExpressionPower
   inputs [
    Base {from=MaterialExpressionMaterialFunctionCall_20, output=Result}
    Exp {from=MaterialExpressionScalarParameter_57}
   ]
   properties {constExponent=0.5}
  }
  MaterialExpressionPower_20 {
   class=MaterialExpressionPower
   inputs [
    Base {from=MaterialExpressionOneMinus_8}
    Exp {from=MaterialExpressionScalarParameter_49}
   ]
   properties {constExponent=12}
  }
  MaterialExpressionPower_26 {
   class=MaterialExpressionPower
   inputs [
    Base {from=MaterialExpressionMultiply_66}
    Exp {from=MaterialExpressionScalarParameter_61}
   ]
  }
  MaterialExpressionPower_27 {
   class=MaterialExpressionPower
   inputs [
    Base {from=MaterialExpressionNamedRerouteUsage_39}
    Exp {from=MaterialExpressionScalarParameter_65}
   ]
   properties {constExponent=0.5}
  }
  MaterialExpressionPower_3 {
   class=MaterialExpressionPower
   inputs [
    Base {from=MaterialExpressionOneMinus_19}
    Exp {from=MaterialExpressionScalarParameter_43}
   ]
   properties {constExponent=8}
  }
  MaterialExpressionPower_4 {
   class=MaterialExpressionPower
   inputs [
    Base {from=MaterialExpressionSaturate_0}
    Exp {from=""}
   ]
  }
  MaterialExpressionPower_7 {
   class=MaterialExpressionPower
   inputs [
    Base {from=MaterialExpressionNamedRerouteUsage_60}
    Exp {from=MaterialExpressionScalarParameter_38}
   ]
   properties {constExponent=4}
  }
  MaterialExpressionPower_8 {
   class=MaterialExpressionPower
   inputs [
    Base {from=MaterialExpressionNamedRerouteUsage_27}
    Exp {from=MaterialExpressionScalarParameter_45}
   ]
  }
  MaterialExpressionPower_9 {
   class=MaterialExpressionPower
   inputs [
    Base {from=MaterialExpressionNamedRerouteUsage_22}
    Exp {from=MaterialExpressionScalarParameter_74}
   ]
   properties {constExponent=8}
  }
  MaterialExpressionQualitySwitch_1 {
   class=MaterialExpressionQualitySwitch
   inputs [
    Default {from=MaterialExpressionReroute_6, output=Return}
    Low {from=MaterialExpressionConstant_2, output=""}
    High {from="", output=""}
    Medium {from="", output=""}
    Epic {from="", output=""}
   ]
   properties {}
  }
  MaterialExpressionQualitySwitch_2 {
   class=MaterialExpressionQualitySwitch
   inputs [
    Default {from=MaterialExpressionAdd_26, output=""}
    Low {from=MaterialExpressionConstant_3, output=""}
    High {from="", output=""}
    Medium {from="", output=""}
    Epic {from="", output=""}
   ]
   properties {}
  }
  MaterialExpressionReroute_0 {class=MaterialExpressionReroute, inputs [None {from=MaterialExpressionPower_4}]}
  MaterialExpressionReroute_1 {class=MaterialExpressionReroute, inputs [None {from=MaterialExpressionReroute_3, output=Fresnel}]}
  MaterialExpressionReroute_4 {class=MaterialExpressionReroute, inputs [None {from=MaterialExpressionOneMinus_5}]}
  MaterialExpressionReroute_5 {class=MaterialExpressionReroute, inputs [None {from=MaterialExpressionSphereMask_2}]}
  MaterialExpressionReroute_6 {class=MaterialExpressionReroute, inputs [None {from=MaterialExpressionReroute_4, output=Alpha}]}
  MaterialExpressionSaturate_0 {class=MaterialExpressionSaturate, inputs [None {from=MaterialExpressionDivide_0}]}
  MaterialExpressionSaturate_1 {class=MaterialExpressionSaturate, inputs [None {from=MaterialExpressionSmoothStep_1}]}
  MaterialExpressionSaturate_2 {class=MaterialExpressionSaturate, inputs [None {from=MaterialExpressionNamedRerouteUsage_41}]}
  MaterialExpressionSaturate_6 {class=MaterialExpressionSaturate, inputs [None {from=MaterialExpressionSubtract_7}]}
  MaterialExpressionSaturate_8 {class=MaterialExpressionSaturate, inputs [None {from=MaterialExpressionDivide_13}]}
  MaterialExpressionSaturate_9 {class=MaterialExpressionSaturate, inputs [None {from=MaterialExpressionDivide_14}]}
  MaterialExpressionScalarParameter_0 {class=MaterialExpressionScalarParameter, properties {defaultValue=0.5, group=06_Thin Film, parameterName=ThinFilmWarp1, sortPriority=4}}
  MaterialExpressionScalarParameter_1 {class=MaterialExpressionScalarParameter, properties {group=06_Thin Film, parameterName=ThinFilmOffset1}}
  MaterialExpressionScalarParameter_10 {class=MaterialExpressionScalarParameter, properties {defaultValue=0.1, group=Crystal, parameterName=GemBlend}}
  MaterialExpressionScalarParameter_14 {class=MaterialExpressionScalarParameter, properties {defaultValue=0.1, group=06_Thin Film, parameterName=ThinFilmExponentFactor1}}
  MaterialExpressionScalarParameter_18 {class=MaterialExpressionScalarParameter, properties {defaultValue=0.05, parameterName=GemMaskStep}}
  MaterialExpressionScalarParameter_2 {class=MaterialExpressionScalarParameter, properties {defaultValue=0.5, group=06_Thin Film, parameterName=ThinFilmIntensity1, sortPriority=2}}
  MaterialExpressionScalarParameter_23 {class=MaterialExpressionScalarParameter, properties {defaultValue=1.5, group=Swirl, parameterName=LowerPantsBaseColorInfluence_Power}}
  MaterialExpressionScalarParameter_25 {class=MaterialExpressionScalarParameter, properties {defaultValue=4, group=Swirl, parameterName=SwirlFresnelThinFilmErasePower}}
  MaterialExpressionScalarParameter_29 {class=MaterialExpressionScalarParameter, properties {defaultValue=2, group=Swirl, parameterName=SwirlBaseColorBlendFresnelExponent}}
  MaterialExpressionScalarParameter_3 {class=MaterialExpressionScalarParameter, properties {defaultValue=1, group=06_Thin Film, parameterName=ThinFilmScale1, sortPriority=5}}
  MaterialExpressionScalarParameter_31 {class=MaterialExpressionScalarParameter, properties {parameterName=TattooEmissiveIntensity}}
  MaterialExpressionScalarParameter_38 {class=MaterialExpressionScalarParameter, properties {defaultValue=2, group=Swirl, parameterName=SwirlEmissiveBlendFresnelExponent}}
  MaterialExpressionScalarParameter_4 {class=MaterialExpressionScalarParameter, properties {defaultValue=2, group=06_Thin Film, parameterName=ThinFilmExponent1, sortPriority=3}}
  MaterialExpressionScalarParameter_43 {class=MaterialExpressionScalarParameter, properties {defaultValue=8, group=Swirl, parameterName=SwirlFresnelPower}}
  MaterialExpressionScalarParameter_44 {class=MaterialExpressionScalarParameter, properties {parameterName=TattooOutlineEmissiveIntensity}}
  MaterialExpressionScalarParameter_45 {class=MaterialExpressionScalarParameter, properties {defaultValue=1, parameterName=TattooEmissivePower}}
  MaterialExpressionScalarParameter_46 {class=MaterialExpressionScalarParameter, properties {defaultValue=1, group=Swirl, parameterName=SwirlFresnelMult}}
  MaterialExpressionScalarParameter_47 {class=MaterialExpressionScalarParameter, properties {defaultValue=1, group=Swirl, parameterName=SwirlBaseBiasIntensity}}
  MaterialExpressionScalarParameter_48 {class=MaterialExpressionScalarParameter, properties {defaultValue=1, group=SwirlBase, parameterName=SwirlBaseOveralBias}}
  MaterialExpressionScalarParameter_49 {class=MaterialExpressionScalarParameter, properties {defaultValue=2, group=SwirlBase, parameterName=SwirlBasePower}}
  MaterialExpressionScalarParameter_50 {class=MaterialExpressionScalarParameter, properties {defaultValue=1, group=TOD, parameterName=TOD_Day}}
  MaterialExpressionScalarParameter_51 {class=MaterialExpressionScalarParameter, properties {defaultValue=4, group=SwirlBase, parameterName=SwirlBaseExtensionPower}}
  MaterialExpressionScalarParameter_52 {class=MaterialExpressionScalarParameter, properties {defaultValue=-30, group=SwirlBase, parameterName=SwirlBasePosition}}
  MaterialExpressionScalarParameter_53 {class=MaterialExpressionScalarParameter, properties {defaultValue=-10, group=SwirlBase, parameterName=SwirlBaseExtensionHeight}}
  MaterialExpressionScalarParameter_54 {class=MaterialExpressionScalarParameter, properties {group=SwirlBase, parameterName=SwirlBaseSmoothStep_Min}}
  MaterialExpressionScalarParameter_55 {class=MaterialExpressionScalarParameter, properties {defaultValue=1, group=SwirlBase, parameterName=SwirlBaseSmoothStep_Max}}
  MaterialExpressionScalarParameter_56 {class=MaterialExpressionScalarParameter, properties {defaultValue=1, group=SwirlBase, parameterName=SwirlBaseInternalBias}}
  MaterialExpressionScalarParameter_57 {class=MaterialExpressionScalarParameter, properties {defaultValue=4, group=SwirlBase, parameterName=SwirlBaseInternalPower}}
  MaterialExpressionScalarParameter_58 {class=MaterialExpressionScalarParameter, properties {defaultValue=1, group=SwirlBase, parameterName=SwirlBaseInternalDistortion}}
  MaterialExpressionScalarParameter_59 {class=MaterialExpressionScalarParameter, properties {defaultValue=1, group=TOD, parameterName=TOD_Morning}}
  MaterialExpressionScalarParameter_6 {class=MaterialExpressionScalarParameter, properties {parameterName=TattooBlend}}
  MaterialExpressionScalarParameter_60 {class=MaterialExpressionScalarParameter, properties {defaultValue=1, group=TOD, parameterName=TOD_Evening}}
  MaterialExpressionScalarParameter_61 {class=MaterialExpressionScalarParameter, properties {defaultValue=1, group=SwirlBase, parameterName=SwirlBaseOverallPower}}
  MaterialExpressionScalarParameter_62 {class=MaterialExpressionScalarParameter, properties {defaultValue=0.1, group=SwirlBase, parameterName=SwirlBaseInternalContrast}}
  MaterialExpressionScalarParameter_63 {class=MaterialExpressionScalarParameter, properties {defaultValue=18, group=SwirlBase, parameterName=SwirlBaseHeight}}
  MaterialExpressionScalarParameter_64 {class=MaterialExpressionScalarParameter, properties {defaultValue=1, group=TOD, parameterName=TOD_Night}}
  MaterialExpressionScalarParameter_65 {class=MaterialExpressionScalarParameter, properties {defaultValue=0.5, group=SwirlBase, parameterName=SwirlBaseInternalNoisePower}}
  MaterialExpressionScalarParameter_66 {class=MaterialExpressionScalarParameter, properties {defaultValue=0.58, group=SwirlBase, parameterName=SwirlBaseUpperBias}}
  MaterialExpressionScalarParameter_67 {class=MaterialExpressionScalarParameter, properties {defaultValue=5, group=SwirlBase, parameterName=SwirlBaseUpperGradientSize}}
  MaterialExpressionScalarParameter_68 {class=MaterialExpressionScalarParameter, properties {defaultValue=1, group=SwirlBase, parameterName=SwirlBaseUpperGradientPosition}}
  MaterialExpressionScalarParameter_69 {class=MaterialExpressionScalarParameter, properties {defaultValue=0.6, group=Swirl, parameterName=UpperBackBias}}
  MaterialExpressionScalarParameter_7 {class=MaterialExpressionScalarParameter, properties {defaultValue=0.1, parameterName=GemRoughness}}
  MaterialExpressionScalarParameter_71 {class=MaterialExpressionScalarParameter, properties {group=Swirl, parameterName=SwirlColorDesaturation_Pants}}
  MaterialExpressionScalarParameter_72 {class=MaterialExpressionScalarParameter, properties {defaultValue=0.2, group=Swirl, parameterName=SwirlEmissiveIntensity_Pants}}
  MaterialExpressionScalarParameter_73 {class=MaterialExpressionScalarParameter, properties {defaultValue=100, group=Swirl, parameterName=PantsGradient_Radius}}
  MaterialExpressionScalarParameter_74 {class=MaterialExpressionScalarParameter, properties {defaultValue=2.5, group=Swirl, parameterName=PantsGradientPower}}
  MaterialExpressionScalarParameter_75 {class=MaterialExpressionScalarParameter, properties {defaultValue=8, group=Swirl, parameterName=SwirlColorIntensity}}
  MaterialExpressionScalarParameter_76 {class=MaterialExpressionScalarParameter, properties {defaultValue=8, group=Swirl, parameterName=SwirlColorIntensity_Pants}}
  MaterialExpressionScalarParameter_77 {class=MaterialExpressionScalarParameter, properties {parameterName=SwirlBlend}}
  MaterialExpressionScalarParameter_78 {class=MaterialExpressionScalarParameter, properties {parameterName=ThinFilmBlend}}
  MaterialExpressionScalarParameter_8 {class=MaterialExpressionScalarParameter, properties {defaultValue=0.2, group=Swirl, parameterName=SwirlEmissiveIntensity}}
  MaterialExpressionScalarParameter_88 {class=MaterialExpressionScalarParameter, properties {defaultValue=0.1, group=Swirl, parameterName=PantsGradient_Hardness}}
  MaterialExpressionScalarParameter_9 {class=MaterialExpressionScalarParameter, properties {group=Swirl, parameterName=UpperBodySwirlBias}}
  MaterialExpressionScalarParameter_90 {class=MaterialExpressionScalarParameter, properties {defaultValue=0.7, group=Swirl, parameterName=SwirlCenterColorBias}}
  MaterialExpressionScalarParameter_91 {class=MaterialExpressionScalarParameter, properties {defaultValue=2.5, group=Swirl, parameterName=PantsGradientPower}}
  MaterialExpressionScalarParameter_92 {class=MaterialExpressionScalarParameter, properties {defaultValue=3, parameterName=GlowFresnel_MidGradientPower}}
  MaterialExpressionScalarParameter_93 {class=MaterialExpressionScalarParameter, properties {defaultValue=-80, parameterName=GlowFresnel_MidGradientPosition}}
  MaterialExpressionSetMaterialAttributes_1 {
   class=MaterialExpressionSetMaterialAttributes
   inputs [
    MaterialAttributes {from=MaterialExpressionNamedRerouteUsage_3, output=""}
    Base Color {from=MaterialExpressionLinearInterpolate_1, output=""}
    Emissive Color {from=MaterialExpressionAdd_8, output=""}
    Roughness {from=MaterialExpressionLinearInterpolate_2, output=""}
   ]
   properties {attributeSetTypes [#1, #2, #3]}
  }
  MaterialExpressionSmoothStep_0 {
   class=MaterialExpressionSmoothStep
   inputs [
    Min {from="", output=""}
    Max {from="", output=""}
    Value {from=MaterialExpressionDivide_5, output=""}
   ]
   properties {}
  }
  MaterialExpressionSmoothStep_1 {
   class=MaterialExpressionSmoothStep
   inputs [
    Min {from=MaterialExpressionScalarParameter_54, output=""}
    Max {from=MaterialExpressionScalarParameter_55, output=""}
    Value {from=MaterialExpressionDivide_6, output=""}
   ]
   properties {constMax=1.75, constMin=0.2}
  }
  MaterialExpressionSmoothStep_3 {
   class=MaterialExpressionSmoothStep
   inputs [
    Min {from=MaterialExpressionConstant_32, output=""}
    Max {from=MaterialExpressionConstant_33, output=""}
    Value {from=MaterialExpressionComponentMask_3, output=""}
   ]
   properties {}
  }
  MaterialExpressionSmoothStep_4 {
   class=MaterialExpressionSmoothStep
   inputs [
    Min {from="", output=""}
    Max {from="", output=""}
    Value {from=MaterialExpressionOneMinus_20, output=""}
   ]
   properties {constMax=0.8, constMin=0.25}
  }
  MaterialExpressionSmoothStep_6 {
   class=MaterialExpressionSmoothStep
   inputs [
    Min {from="", output=""}
    Max {from="", output=""}
    Value {from=MaterialExpressionPower_26, output=""}
   ]
   properties {constMax=0.7}
  }
  MaterialExpressionSmoothStep_7 {
   class=MaterialExpressionSmoothStep
   inputs [
    Min {from="", output=""}
    Max {from="", output=""}
    Value {from=MaterialExpressionDivide_8, output=""}
   ]
   properties {}
  }
  MaterialExpressionSphereMask_0 {
   class=MaterialExpressionSphereMask
   inputs [
    A {from=MaterialExpressionMaterialFunctionCall_3, output=Result}
    B {from=MaterialExpressionConstant3Vector_1, output=""}
    Radius {from=MaterialExpressionConstant_1, output=""}
    Hardness {from=MaterialExpressionConstant_4, output=""}
   ]
   properties {attenuationRadius=17, hardnessPercent=0}
  }
  MaterialExpressionSphereMask_1 {
   class=MaterialExpressionSphereMask
   inputs [
    A {from=MaterialExpressionMaterialFunctionCall_18, output=Result}
    B {from=MaterialExpressionConstant_11, output=""}
    Radius {from=MaterialExpressionScalarParameter_73, output=""}
    Hardness {from=MaterialExpressionScalarParameter_88, output=""}
   ]
   properties {attenuationRadius=100, hardnessPercent=10}
  }
  MaterialExpressionSphereMask_2 {
   class=MaterialExpressionSphereMask
   inputs [
    A {from=MaterialExpressionMaterialFunctionCall_7, output=Result}
    B {from=MaterialExpressionConstant3Vector_2, output=""}
    Radius {from="", output=""}
    Hardness {from=MaterialExpressionConstant_34, output=""}
   ]
   properties {attenuationRadius=18, hardnessPercent=25}
  }
  MaterialExpressionStep_0 {
   class=MaterialExpressionStep
   inputs [
    Y {from=MaterialExpressionScalarParameter_18, output=""}
    X {from=MaterialExpressionTextureSampleParameter2D_1, output=G}
   ]
   properties {constX=0.1, constY=0.1}
  }
  MaterialExpressionStep_2 {
   class=MaterialExpressionStep
   inputs [
    Y {from="", output=""}
    X {from=MaterialExpressionComponentMask_0, output=""}
   ]
   properties {constX=34, constY=20}
  }
  MaterialExpressionSubtract_10 {
   class=MaterialExpressionSubtract
   inputs [
    A {from=MaterialExpressionComponentMask_1}
    B {from=MaterialExpressionMultiply_93}
   ]
  }
  MaterialExpressionSubtract_5 {
   class=MaterialExpressionSubtract
   inputs [
    A {from=MaterialExpressionSmoothStep_6}
    B {from=MaterialExpressionMultiply_88}
   ]
  }
  MaterialExpressionSubtract_6 {
   class=MaterialExpressionSubtract
   inputs [
    A {from=MaterialExpressionAdd_21}
    B {from=MaterialExpressionScalarParameter_68}
   ]
   properties {constB=4}
  }
  MaterialExpressionSubtract_7 {
   class=MaterialExpressionSubtract
   inputs [
    A {from=MaterialExpressionMaterialFunctionCall_16, output=B}
    B {from=""}
   ]
   properties {constB=108}
  }
  MaterialExpressionSubtract_9 {
   class=MaterialExpressionSubtract
   inputs [
    A {from=MaterialExpressionComponentMask_14}
    B {from=""}
   ]
   properties {constB=20}
  }
  MaterialExpressionTextureSampleParameter2D_1 {
   class=MaterialExpressionTextureSampleParameter2D
   inputs [
    UVs {from="", output=""}
    Apply View MipBias {from="", output=""}
   ]
   properties {group=06_Emissive, parameterName=Emissive, sortPriority=1, texture=&T_BlacksRGB}
  }
  MaterialExpressionTextureSample_6 {
   class=MaterialExpressionTextureSample
   inputs [
    UVs {from=""}
    Tex {from=""}
    Apply View MipBias {from=""}
   ]
   properties {texture=/BRCosmetics/Characters/Player/Female/Medium/Bodies/F_MED_ChicleVeil/Textures/T_ChicleVeil_Body_FX}
  }
  MaterialExpressionVectorParameter_5 {class=MaterialExpressionVectorParameter, properties {parameterName=Pantsbase, defaultValue=<rgba 0.1 0.1 0.1 1>}}
 }
 properties {
  bUseMaterialAttributes=true
  bUsedWithMeshDeformer=true
  bUsedWithMorphTargets=true
  bUsedWithNanite=true
  bUsedWithSkeletalMesh=true
  pixelDepthOffsetMode=PDOM_Legacy
  previewMesh=/BRCosmetics/Characters/Player/Female/Medium/Bodies/F_MED_ChicleVeil/Meshes/F_MED_ChicleVeil
  shadingModel=MSM_FromMaterialExpression
  usedShadingModels=Default Lit | Subsurface | Clear Coat | Subsurface Profile | Cloth
 }
}
