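# Control Rig for the F_MED_ChicleVeil female body cosmetic that applies secondary bone corrections on a three-branch sequence.
# Smoothed spherical pose readers on both hands and the left upper arm weight offsets to hand armor and left pec bones; the right-hand rotation is zero.
# The smoothed `crouch` variable blends the left suspender base toward a fixed local transform.
# Two other SetTransforms and a left-thigh pose reader are unconnected.
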ControlRigBlueprint /BRCosmetics/Characters/Player/Female/Medium/Bodies/F_MED_ChicleVeil/Meshes/F_MED_ChicleVeil_CtrlRig
graphs [
 RigVMFunctionLibrary {links [], nodes []}
 RigVMFunction_Sequence_1_ContainedGraph {nodes [
   n1 = RigVMFunctionEntryNode()
   RigVMFunction_Sequence = &RigVMFunction_Sequence(ExecuteContext=@n1.ExecuteContext)
   n2 = &RigVMFunction_Sequence(ExecuteContext=@RigVMFunction_Sequence.B)
   RigVMFunctionReturnNode(A=@RigVMFunction_Sequence.A, B=@n2.A, C=@n2.B)
  ]}
 RigVMModel {nodes [
   RigUnit_BeginExecution = &RigUnit_BeginExecution()
   RigVMFunction_Sequence = RigVMAggregateNode(ExecuteContext=@RigUnit_BeginExecution.ExecutePin)
   n1 = &RigUnit_SphericalPoseReader(
    ActiveRegionScaleFactors=(PositiveWidth=1,NegativeWidth=1,PositiveHeight=1,NegativeHeight=1)
    ActiveRegionSize=0.1
    Debug=(bDrawDebug=True,bDraw2D=false,bDrawLocalAxes=true,DebugScale=25,DebugSegments=20,DebugThickness=0.25)
    DriverAxis=(X=1,Y=0,Z=0)
    DriverItem=(Type=Bone,Name="hand_l")
    FalloffRegionScaleFactors=(PositiveWidth=1,NegativeWidth=1,PositiveHeight=1,NegativeHeight=1)
    FalloffSize=0.2
    FlipHeightScaling=False
    FlipWidthScaling=False
    OptionalParentItem=(Type=Bone,Name="lowerarm_l")
    RotationOffset=(X=90,Y=0,Z=-40)
    ExecutePin=@RigVMFunction_Sequence.A
   )
   AlphaInterp = &RigVMFunction_AlphaInterp(InterpSpeedDecreasing=8, InterpSpeedIncreasing=8, Scale=1, bClampResult=False, Value=@n1.OutputParam)
   > &RigUnit_ModifyTransforms(
    ItemToModify=((Item=(Type=Bone,Name="extra_armor_hand_l"),Transform=(Rotation=(X=0,Y=0,Z=0,W=1),Translation=(X=0,Y=0,Z=0),Scale3D=(X=1.0,Y=1.0,Z=1.0))))
    Weight=@AlphaInterp.Result
    ItemToModify.0.Transform.Rotation=&RigVMFunction_MathQuaternionFromEuler(Euler=(X=0,Y=0,Z=-5)).Result
   )
   > n2 = &RigUnit_SphericalPoseReader(
    ActiveRegionScaleFactors=(PositiveWidth=1,NegativeWidth=1,PositiveHeight=1,NegativeHeight=1)
    ActiveRegionSize=0.1
    Debug=(bDrawDebug=True,bDraw2D=false,bDrawLocalAxes=true,DebugScale=25,DebugSegments=20,DebugThickness=0.25)
    DriverAxis=(X=-1,Y=0,Z=0)
    DriverItem=(Type=Bone,Name="hand_r")
    FalloffRegionScaleFactors=(PositiveWidth=1,NegativeWidth=1,PositiveHeight=1,NegativeHeight=1)
    FalloffSize=0.2
    FlipHeightScaling=False
    FlipWidthScaling=False
    OptionalParentItem=(Type=Bone,Name="lowerarm_r")
    RotationOffset=(X=-90,Y=0,Z=-40)
   )
   n3 = &RigVMFunction_AlphaInterp(InterpSpeedDecreasing=8, InterpSpeedIncreasing=8, Scale=1, bClampResult=False, Value=@n2.OutputParam)
   > &RigUnit_ModifyTransforms(
    ItemToModify=((Item=(Type=Bone,Name="extra_armor_hand_r"),Transform=(Rotation=(X=0,Y=0,Z=0,W=1),Translation=(X=0,Y=0,Z=0),Scale3D=(X=1.0,Y=1.0,Z=1.0))))
    Weight=@n3.Result
    ItemToModify.0.Transform.Rotation=&RigVMFunction_MathQuaternionFromEuler(Euler=(X=0,Y=0,Z=0)).Result
   )
   n4 = &RigUnit_SphericalPoseReader(
    ActiveRegionScaleFactors=(PositiveWidth=1,NegativeWidth=1,PositiveHeight=1,NegativeHeight=1)
    ActiveRegionSize=0.1
    Debug=(bDrawDebug=True,bDraw2D=false,bDrawLocalAxes=true,DebugScale=25,DebugSegments=20,DebugThickness=0.25)
    DriverAxis=(X=1,Y=0,Z=0)
    DriverItem=(Type=Bone,Name="upperarm_l")
    FalloffRegionScaleFactors=(PositiveWidth=1,NegativeWidth=1,PositiveHeight=1,NegativeHeight=1)
    FalloffSize=0.2
    FlipHeightScaling=False
    FlipWidthScaling=False
    OptionalParentItem=(Type=Bone,Name="lowerarm_r")
    RotationOffset=(X=50,Y=60,Z=0)
    ExecutePin=@RigVMFunction_Sequence.B
   )
   n5 = &RigVMFunction_AlphaInterp(InterpSpeedDecreasing=8, InterpSpeedIncreasing=8, Scale=1, bClampResult=False, Value=@n4.OutputParam)
   > &RigUnit_ModifyTransforms(
    ItemToModify=((Item=(Type=Bone,Name="extra_pec_l"),Transform=(Rotation=(X=0,Y=0,Z=0,W=1),Translation=(X=0,Y=0,Z=0),Scale3D=(X=1.0,Y=1.0,Z=1.0))))
    Weight=@n5.Result
    ItemToModify.0.Transform.Rotation=&RigVMFunction_MathQuaternionFromEuler(Euler=(X=0,Y=-5,Z=10)).Result
   )
   n6 = &RigVMFunction_AlphaInterp(InterpSpeedDecreasing=12, InterpSpeedIncreasing=12, Scale=1, bClampResult=False, Value=RigVMVariableNode() {variable=crouch}.Value)
   &RigUnit_SetTransform(
    Item=(Type=Bone,Name="dyn_suspenders_01_base_l")
    Space=LocalSpace
    Value=(Rotation=(X=-0.2232,Y=0.2403,Z=-0.0676,W=0.9423),Translation=(X=6.9,Y=3,Z=-11),Scale3D=(X=1,Y=1,Z=1))
    bInitial=false
    bPropagateToChildren=false
    Weight=@n6.Result
    ExecutePin=@RigVMFunction_Sequence.C
   )
   &RigUnit_SetTransform(
    ExecutePin=()
    Item=(Type=Bone,Name="dyn_suspenders_01_base_l")
    Space=LocalSpace
    Value=(Rotation=(X=0,Y=0,Z=0,W=1),Translation=(X=8,Y=0,Z=-11),Scale3D=(X=1,Y=1,Z=1))
    Weight=1
    bInitial=false
    bPropagateToChildren=false
   )
   &RigUnit_SetTransform(
    ExecutePin=()
    Item=(Type=Bone,Name="dyn_suspenders_01_base_l")
    Space=LocalSpace
    Value=(Rotation=(X=-0.2232,Y=0.2403,Z=-0.0676,W=0.9423),Translation=(X=6,Y=3,Z=-11),Scale3D=(X=1,Y=1,Z=1))
    Weight=1
    bInitial=false
    bPropagateToChildren=false
   )
   &RigUnit_SphericalPoseReader(
    ActiveRegionScaleFactors=(PositiveWidth=1,NegativeWidth=1,PositiveHeight=1,NegativeHeight=1)
    ActiveRegionSize=0.1
    Debug=(bDrawDebug=True,bDraw2D=false,bDrawLocalAxes=true,DebugScale=25,DebugSegments=20,DebugThickness=0.25)
    DriverAxis=(X=-1,Y=0,Z=0)
    DriverItem=(Type=Bone,Name="thigh_l")
    ExecutePin=()
    FalloffRegionScaleFactors=(PositiveWidth=1,NegativeWidth=1,PositiveHeight=1,NegativeHeight=1)
    FalloffSize=0.2
    FlipHeightScaling=False
    FlipWidthScaling=False
    OptionalParentItem=(Type=Bone,Name="None")
    RotationOffset=(X=115,Y=25,Z=0)
   )
  ]}
]
hierarchy [
 root {loc=<0 0 0>, rot=<0 0 0 1>}
 pelvis {parent=root, loc=<0 -0.5301 98.0571>, rot=<0.0127 0.707 -0.0127 -0.707>}
 spine_01 {parent=pelvis, loc=<6.1208 -0.9328 0>, rot=<0 0 0.034 -0.9994>}
 spine_02 {parent=spine_01, loc=<6.5996 -0.2095 0>, rot=<0 0 0.0238 -0.9997>}
 spine_03 {parent=spine_02, loc=<6.4914 -0.1237 0>, rot=<0 0 0.0149 -0.9999>}
 spine_04 {parent=spine_03, loc=<6.9598 0.0794 0>, rot=<0 0 -0.0132 -0.9999>}
 spine_05 {parent=spine_04, loc=<6.9987 0 0>, rot=<0 0 -0.0015 -1>}
 attach_cape {parent=spine_05, loc=<9.0271 -11.7158 -0.0002>, rot=<0.0616 -0.7044 -0.0616 -0.7044>, scale=<0.9999 1.0001 1>}
 neck_01 {parent=spine_05, loc=<11.3138 0.1529 0>, rot=<0 0 -0.2003 -0.9797>}
 neck_02 {parent=neck_01, loc=<4.5627 0 0>, rot=<0 0 0.0504 -0.9987>}
 head {parent=neck_02, loc=<5.3914 0 0>, rot=<0 0 0.115 -0.9934>}
 attach_backpack {parent=spine_05, loc=<-0.1102 -13.4749 -0.0002>, rot=<-0.0093 -0.707 0.0093 -0.707>, scale=<0.9999 1.0001 1>}
 clavicle_l {parent=spine_05, loc=<5.4831 -0.1053 -3.2946>, rot=<0.0465 -0.7289 -0.009 -0.6829>}
 upperarm_l {parent=clavicle_l, loc=<9.8658 -1.2136 -0.4105>, rot=<0.0164 -0.3245 -0.0269 -0.9454>}
 lowerarm_l {parent=upperarm_l, loc=<26.465 -0.8345 0.259>, rot=<0.0036 0.0161 -0.2298 -0.9731>}
 hand_l {parent=lowerarm_l, loc=<22.7355 -0.3203 -0.6105>, rot=<-0.6784 -0.0223 0.0735 -0.7307>}
 thumb_01_l {parent=hand_l, loc=<1.3118 -0.3191 -2.459>, rot=<0.6546 -0.4042 0.0664 -0.6354>}
 thumb_02_l {parent=thumb_01_l, loc=<5.9336 -0.0448 0.0401>, rot=<-0.0037 0.0213 0.298 -0.9543>}
 thumb_03_l {parent=thumb_02_l, loc=<2.2223 0.3682 -0.0907>, rot=<0.0118 -0.013 0.0066 -0.9998>}
 index_metacarpal_l {parent=hand_l, loc=<1.6535 0.0183 -2.96>, rot=<0.0158 -0.0124 -0.0173 -0.9996>}
 index_01_l {parent=index_metacarpal_l, loc=<8.3008 0.0357 -0.0524>, rot=<-0.0044 0.0485 0.0924 -0.9945>}
 index_02_l {parent=index_01_l, loc=<3.3587 -0.3004 0.0452>, rot=<-0.002 0.0027 0.3225 -0.9466>}
 index_03_l {parent=index_02_l, loc=<2.5578 -0.123 -0.0252>, rot=<-0.002 0.0067 0.2558 -0.9667>}
 middle_metacarpal_l {parent=hand_l, loc=<2.1856 -0.1725 -0.5984>, rot=<0.006 0.0168 -0.0329 -0.9993>}
 middle_01_l {parent=middle_metacarpal_l, loc=<7.3275 -0.2447 0.0129>, rot=<-0.0208 0.0042 0.0858 -0.9961>}
 middle_02_l {parent=middle_01_l, loc=<4.0123 0.1401 0.1175>, rot=<-0.0218 0.0258 0.3512 -0.9357>}
 middle_03_l {parent=middle_02_l, loc=<2.6077 0.0534 0.1373>, rot=<0.0082 0.0188 0.1708 -0.9851>}
 ring_metacarpal_l {parent=hand_l, loc=<2.2074 -0.3625 1.3638>, rot=<-0.0284 0.0722 -0.0458 -0.9959>}
 ring_01_l {parent=ring_metacarpal_l, loc=<7.0295 -0.6197 -0.1446>, rot=<-0.0489 -0.0252 0.1892 -0.9804>}
 ring_02_l {parent=ring_01_l, loc=<3.6174 0.4647 -0.1027>, rot=<0 0 0.2806 -0.9598>}
 ring_03_l {parent=ring_02_l, loc=<2.4152 0.114 0.0199>, rot=<0 0 0.1189 -0.9929>}
 pinky_metacarpal_l {parent=hand_l, loc=<1.7616 -0.9753 3.3814>, rot=<-0.0397 0.0814 -0.069 -0.9935>}
 pinky_01_l {parent=pinky_metacarpal_l, loc=<7.1709 -0.4831 -0.1094>, rot=<-0.0654 -0.0035 0.2242 -0.9723>}
 pinky_02_l {parent=pinky_01_l, loc=<2.8274 0.0893 -0.1043>, rot=<0.0037 0.0011 0.2854 -0.9584>}
 pinky_03_l {parent=pinky_02_l, loc=<2.2046 0.0307 -0.0078>, rot=<-0.0002 0.0098 0.1445 -0.9895>}
 deform_wrist_dn_base_l {parent=hand_l, loc=<0 -1.5326 0>, rot=<0 0 0 1>}
 deform_wrist_dn_l {parent=deform_wrist_dn_base_l, loc=<0 0 0>, rot=<0 0 0 1>}
 deform_wrist_up_base_l {parent=hand_l, loc=<0 2.3009 0>, rot=<0 0 0 1>}
 deform_wrist_up_l {parent=deform_wrist_up_base_l, loc=<0 0 0>, rot=<0 0 0 1>}
 extra_armor_hand_l {parent=hand_l, loc=<-0.8788 1.9634 0.2324>, rot=<0.9966 -0.0109 0.0223 -0.0779>}
 lowerarm_twist_02_l {parent=lowerarm_l, loc=<10.6707 0 0>, rot=<0 0 0 1>}
 lowerarm_twist_01_l {parent=lowerarm_l, loc=<18.316 0 0>, rot=<0 0 0 1>}
 deform_elbow_in_base_l {parent=lowerarm_l, loc=<0.1834 3.6062 -0.2572>, rot=<0 0 0 1>}
 deform_elbow_in_l {parent=deform_elbow_in_base_l, loc=<0 0 0>, rot=<0 0 0 1>}
 deform_elbow_out_base_l {parent=lowerarm_l, loc=<0 -4 0>, rot=<0 0 0 1>}
 deform_elbow_out_l {parent=deform_elbow_out_base_l, loc=<0 0 0>, rot=<0 0 0 1>}
 upperarm_twist_01_l {parent=upperarm_l, loc=<7.2135 -0.1134 0>, rot=<0 0 0 1>}
 deform_upperarm_bk_base_l {parent=upperarm_twist_01_l, loc=<-6.4053 -4.3746 0.314>, rot=<0 0 0 1>}
 deform_upperarm_bk_l {parent=deform_upperarm_bk_base_l, loc=<0 0 0>, rot=<0 0 0 1>}
 deform_upperarm_fr_base_l {parent=upperarm_twist_01_l, loc=<-5.6312 5.092 0.3354>, rot=<0 0 0 1>}
 deform_upperarm_fr_l {parent=deform_upperarm_fr_base_l, loc=<0 0 0>, rot=<0 0 0 1>}
 deform_upperarm_up_base_l {parent=upperarm_twist_01_l, loc=<-8.891 0.4825 3.9517>, rot=<0 0 0 1>}
 deform_upperarm_up_l {parent=deform_upperarm_up_base_l, loc=<0 0 0>, rot=<0 0 0 1>}
 upperarm_twist_02_l {parent=upperarm_l, loc=<19.7998 -0.1134 0>, rot=<0 0 0 1>}
 clavicle_r {parent=spine_05, loc=<5.4831 -0.1053 3.2946>, rot=<0.7289 0.0465 -0.6829 0.009>}
 upperarm_r {parent=clavicle_r, loc=<-9.8658 1.2136 0.4105>, rot=<0.0164 -0.3245 -0.0269 -0.9454>}
 lowerarm_r {parent=upperarm_r, loc=<-26.465 0.8345 -0.259>, rot=<0.0036 0.0161 -0.2298 -0.9731>}
 hand_r {parent=lowerarm_r, loc=<-22.7355 0.3203 0.6106>, rot=<-0.6784 -0.0223 0.0735 -0.7307>}
 thumb_01_r {parent=hand_r, loc=<-1.3118 0.3191 2.459>, rot=<0.6546 -0.4042 0.0664 -0.6354>}
 thumb_02_r {parent=thumb_01_r, loc=<-5.9336 0.0448 -0.0401>, rot=<-0.0037 0.0213 0.298 -0.9543>}
 thumb_03_r {parent=thumb_02_r, loc=<-2.2223 -0.3682 0.0907>, rot=<0.0118 -0.013 0.0066 -0.9998>}
 index_metacarpal_r {parent=hand_r, loc=<-1.6535 -0.0183 2.96>, rot=<0.0158 -0.0124 -0.0173 -0.9996>}
 index_01_r {parent=index_metacarpal_r, loc=<-8.3008 -0.0357 0.0524>, rot=<-0.0044 0.0485 0.0924 -0.9945>}
 index_02_r {parent=index_01_r, loc=<-3.3587 0.3004 -0.0452>, rot=<-0.002 0.0027 0.3225 -0.9466>}
 index_03_r {parent=index_02_r, loc=<-2.5578 0.123 0.0252>, rot=<-0.002 0.0067 0.2558 -0.9667>}
 middle_metacarpal_r {parent=hand_r, loc=<-2.1856 0.1725 0.5984>, rot=<0.006 0.0168 -0.0329 -0.9993>}
 middle_01_r {parent=middle_metacarpal_r, loc=<-7.3275 0.2447 -0.0129>, rot=<-0.0208 0.0042 0.0858 -0.9961>}
 middle_02_r {parent=middle_01_r, loc=<-4.0123 -0.1401 -0.1175>, rot=<-0.0218 0.0258 0.3512 -0.9357>}
 middle_03_r {parent=middle_02_r, loc=<-2.6077 -0.0534 -0.1373>, rot=<0.0082 0.0188 0.1708 -0.9851>}
 ring_metacarpal_r {parent=hand_r, loc=<-2.2074 0.3625 -1.3638>, rot=<-0.0284 0.0722 -0.0458 -0.9959>}
 ring_01_r {parent=ring_metacarpal_r, loc=<-7.0295 0.6197 0.1446>, rot=<-0.0489 -0.0252 0.1892 -0.9804>}
 ring_02_r {parent=ring_01_r, loc=<-3.6174 -0.4647 0.1027>, rot=<0 0 0.2806 -0.9598>}
 ring_03_r {parent=ring_02_r, loc=<-2.4152 -0.114 -0.0199>, rot=<0 0 0.1189 -0.9929>}
 pinky_metacarpal_r {parent=hand_r, loc=<-1.7616 0.9753 -3.3814>, rot=<-0.0397 0.0814 -0.069 -0.9935>}
 pinky_01_r {parent=pinky_metacarpal_r, loc=<-7.1709 0.4831 0.1094>, rot=<-0.0654 -0.0035 0.2242 -0.9723>}
 pinky_02_r {parent=pinky_01_r, loc=<-2.8274 -0.0893 0.1043>, rot=<0.0037 0.0011 0.2854 -0.9584>}
 pinky_03_r {parent=pinky_02_r, loc=<-2.2046 -0.0307 0.0078>, rot=<-0.0002 0.0098 0.1445 -0.9895>}
 deform_wrist_dn_base_r {parent=hand_r, loc=<-0.0001 1.5328 0>, rot=<0 0 0 1>}
 deform_wrist_dn_r {parent=deform_wrist_dn_base_r, loc=<0 0 0>, rot=<0 0 0 1>}
 deform_wrist_up_base_r {parent=hand_r, loc=<-0.0002 -2.3007 -0.0001>, rot=<0 0 0 1>}
 deform_wrist_up_r {parent=deform_wrist_up_base_r, loc=<0 0 0>, rot=<0 0 0 1>}
 extra_armor_hand_r {parent=hand_r, loc=<0.8789 -1.9635 -0.2324>, rot=<-0.0109 -0.9966 -0.0779 -0.0223>}
 lowerarm_twist_02_r {parent=lowerarm_r, loc=<-10.6707 0 0>, rot=<0 0 0 -1>}
 lowerarm_twist_01_r {parent=lowerarm_r, loc=<-18.316 0 0>, rot=<0 0 0 -1>}
 deform_elbow_in_base_r {parent=lowerarm_r, loc=<-0.1834 -3.6062 0.2572>, rot=<0 0 0 -1>}
 deform_elbow_in_r {parent=deform_elbow_in_base_r, loc=<0 0 0>, rot=<0 0 0 1>}
 deform_elbow_out_base_r {parent=lowerarm_r, loc=<0 4 -0.0001>, rot=<0 0 0 -1>}
 deform_elbow_out_r {parent=deform_elbow_out_base_r, loc=<0 0 0>, rot=<0 0 0 1>}
 upperarm_twist_01_r {parent=upperarm_r, loc=<-7.2135 0.1134 0>, rot=<0 0 0 -1>}
 deform_upperarm_bk_base_r {parent=upperarm_twist_01_r, loc=<6.4053 4.3746 -0.314>, rot=<0 0 0 1>}
 deform_upperarm_bk_r {parent=deform_upperarm_bk_base_r, loc=<0 0 0>, rot=<0 0 0 1>}
 deform_upperarm_fr_base_r {parent=upperarm_twist_01_r, loc=<5.6312 -5.092 -0.3354>, rot=<0 0 0 1>}
 deform_upperarm_fr_r {parent=deform_upperarm_fr_base_r, loc=<0 0 0>, rot=<0 0 0 1>}
 deform_upperarm_up_base_r {parent=upperarm_twist_01_r, loc=<8.891 -0.4825 -3.9517>, rot=<0 0 0 1>}
 deform_upperarm_up_r {parent=deform_upperarm_up_base_r, loc=<0 0 0>, rot=<0 0 0 1>}
 upperarm_twist_02_r {parent=upperarm_r, loc=<-19.7998 0.1134 0>, rot=<0 0 0 -1>}
 extra_curve_back_01_c {parent=spine_05, loc=<0.0389 -9.8452 0>, rot=<0.0284 -0.7065 -0.0284 -0.7065>}
 extra_curve_back_02_l {parent=spine_05, loc=<1.1969 -9.9376 -5.2654>, rot=<0.0284 -0.7065 -0.0284 -0.7065>}
 extra_curve_back_02_r {parent=spine_05, loc=<1.1972 -9.9376 5.2655>, rot=<0.0284 -0.7065 -0.0284 -0.7065>}
 extra_pec_l {parent=spine_05, loc=<2.467 7.418 -8.0944>, rot=<0.0819 -0.0735 -0.5508 -0.8273>}
 extra_pec_r {parent=spine_05, loc=<2.4675 7.418 8.0944>, rot=<-0.0819 0.0735 -0.5508 -0.8273>}
 deform_pec_aim_base_l {parent=spine_04, loc=<3.2456 9.3309 -1.2237>, rot=<0.0603 -0.4033 0.3318 -0.8507>}
 deform_pec_aim_l {parent=deform_pec_aim_base_l, loc=<0 0 0>, rot=<0 0 0 1>}
 deform_pec_base_l {parent=deform_pec_aim_l, loc=<9.3665 0 0>, rot=<0 0 0 1>}
 deform_pec_l {parent=deform_pec_base_l, loc=<0 0 0>, rot=<0 0 0 1>}
 deform_pec_aim_base_r {parent=spine_04, loc=<3.2452 9.3309 1.2237>, rot=<-0.4033 -0.0603 0.8507 0.3318>}
 deform_pec_aim_r {parent=deform_pec_aim_base_r, loc=<0 0 0>, rot=<0 0 0 1>}
 deform_pec_base_r {parent=deform_pec_aim_r, loc=<-9.3667 -0.0003 -0.0004>, rot=<0 0 0 1>}
 deform_pec_r {parent=deform_pec_base_r, loc=<0 0 0>, rot=<0 0 0 1>}
 extra_curve_back_02_c {parent=spine_04, loc=<0.7975 -10.236 0>, rot=<0.0294 -0.7065 -0.0294 -0.7065>}
 extra_curve_back_03_c {parent=spine_04, loc=<-3.1996 -9.3067 0>, rot=<0.0294 -0.7065 -0.0294 -0.7065>}
 extra_curve_side_01_l {parent=spine_03, loc=<1.1165 -0.5312 -10.6113>, rot=<0.0387 -0.706 -0.0387 -0.706>}
 extra_curve_side_01_r {parent=spine_03, loc=<1.116 -0.5313 10.6113>, rot=<0.0387 -0.706 -0.0387 -0.706>}
 thigh_l {parent=pelvis, loc=<-1.0855 -0.2109 -8.7053>, rot=<0.0009 0.0355 0.0155 -0.9993>}
 calf_l {parent=thigh_l, loc=<-42.5284 -0.0019 -0.7297>, rot=<0.0006 -0.0161 -0.0391 -0.9991>}
 foot_l {parent=calf_l, loc=<-46.7401 -2.5014 -0.592>, rot=<0.0427 -0.0238 0.0353 -0.9982>}
 ball_l {parent=foot_l, loc=<-3.5658 10.8066 -0.1544>, rot=<0.0001 -0.0001 -0.7063 -0.7079>}
 calf_twist_02_l {parent=calf_l, loc=<-19.1546 -1.7943 -0.2638>, rot=<0.0237 0.0095 -0.0035 -0.9997>}
 calf_twist_01_l {parent=calf_l, loc=<-35.8483 -1.9197 -0.5794>, rot=<0.0237 0.0095 -0.0035 -0.9997>}
 deform_knee_in_base_l {parent=calf_l, loc=<0.8031 -5.0231 0.2456>, rot=<0 0 0 -1>}
 deform_knee_in_l {parent=deform_knee_in_base_l, loc=<0 0 0>, rot=<0 0 0 1>}
 deform_knee_out_base_l {parent=calf_l, loc=<0 4 0>, rot=<0 0 0 -1>}
 deform_knee_out_l {parent=deform_knee_out_base_l, loc=<0 0 0>, rot=<0 0 0 1>}
 leg_strap_A_01_base_l {parent=calf_l, loc=<-3.7792 -2.9119 -6.4785>, rot=<0.0473 -0.9723 -0.0118 -0.2286>}
 leg_strap_A_01_l {parent=leg_strap_A_01_base_l, loc=<0 0 0>, rot=<0 0 0 1>}
 leg_strap_A_02_l {parent=leg_strap_A_01_l, loc=<6.6004 0.0235 0.3247>, rot=<-0.0012 0.0482 -0.0056 -0.9988>}
 leg_strap_A_03_l {parent=leg_strap_A_02_l, loc=<6.2343 0 0>, rot=<0.0265 -0.0581 -0.0282 -0.9976>}
 leg_strap_B_01_base_l {parent=calf_l, loc=<-4.1791 -2.9249 -6.2607>, rot=<0.0405 -0.9826 0.021 -0.1799>}
 leg_strap_B_01_l {parent=leg_strap_B_01_base_l, loc=<0 0 0>, rot=<0 0 0 1>}
 leg_strap_B_02_l {parent=leg_strap_B_01_l, loc=<9.9379 0 0>, rot=<0.0728 0.0001 -0.0024 -0.9973>}
 leg_strap_B_03_l {parent=leg_strap_B_02_l, loc=<11.408 0 0>, rot=<0.0037 -0.0521 -0.0008 -0.9986>}
 leg_strap_B_04_l {parent=leg_strap_B_03_l, loc=<9.1701 0 0>, rot=<-0.0097 -0.0324 -0.0041 -0.9994>}
 extra_curve_calf_l {parent=calf_l, loc=<-3.2674 -2.8433 -6.4606>, rot=<-0.0131 -0.9923 0.1128 -0.0488>}
 dyn_trouser_leg_l {parent=calf_l, loc=<-31.5378 -3.8674 -2.519>, rot=<-0.4399 -0.5472 -0.4842 -0.5222>}
 thigh_twist_01_l {parent=thigh_l, loc=<-27.952 0 0>, rot=<0.0013 0 0 -1>}
 deform_groin_base_l {parent=thigh_l, loc=<0.4801 5.6776 -0.2486>, rot=<0 0 0 -1>}
 deform_groin_l {parent=deform_groin_base_l, loc=<0 0 0>, rot=<0 0 0 1>}
 thigh_r {parent=pelvis, loc=<-1.0855 -0.2109 8.7053>, rot=<0.0355 -0.0009 0.9993 0.0155>}
 calf_r {parent=thigh_r, loc=<42.5284 0.0019 0.7297>, rot=<0.0006 -0.0161 -0.0391 -0.9991>}
 foot_r {parent=calf_r, loc=<46.7401 2.5014 0.592>, rot=<0.0427 -0.0238 0.0353 -0.9982>}
 ball_r {parent=foot_r, loc=<3.5658 -10.8066 0.1544>, rot=<0.0001 -0.0001 -0.7063 -0.7079>}
 calf_twist_02_r {parent=calf_r, loc=<19.1546 1.7943 0.2638>, rot=<0.0237 0.0095 -0.0035 -0.9997>}
 calf_twist_01_r {parent=calf_r, loc=<35.8483 1.9197 0.5794>, rot=<0.0237 0.0095 -0.0035 -0.9997>}
 deform_knee_in_base_r {parent=calf_r, loc=<-0.8031 5.0231 -0.2457>, rot=<0 0 0 -1>}
 deform_knee_in_r {parent=deform_knee_in_base_r, loc=<0 0 0>, rot=<0 0 0 1>}
 deform_knee_out_base_r {parent=calf_r, loc=<0 -4 0>, rot=<0 0 0 -1>}
 deform_knee_out_r {parent=deform_knee_out_base_r, loc=<0 0 0>, rot=<0 0 0 1>}
 leg_strap_B_01_base_r {parent=calf_r, loc=<4.1791 2.9249 6.2607>, rot=<0.021 0.1799 -0.0405 -0.9826>}
 leg_strap_B_01_r {parent=leg_strap_B_01_base_r, loc=<0 0 0>, rot=<0 0 0 1>}
 leg_strap_B_02_r {parent=leg_strap_B_01_r, loc=<9.9379 0 0>, rot=<-0.0728 0.0001 0.0024 -0.9973>}
 leg_strap_B_03_r {parent=leg_strap_B_02_r, loc=<11.4079 0 0.0001>, rot=<-0.0037 -0.0521 0.0008 -0.9986>}
 leg_strap_B_04_r {parent=leg_strap_B_03_r, loc=<9.1702 0 0>, rot=<0.0097 -0.0324 0.0041 -0.9994>}
 leg_strap_A_01_base_r {parent=calf_r, loc=<3.7791 2.9119 6.4786>, rot=<-0.0118 0.2287 -0.0473 -0.9723>}
 leg_strap_A_01_r {parent=leg_strap_A_01_base_r, loc=<0 0 0>, rot=<0 0 0 1>}
 leg_strap_A_02_r {parent=leg_strap_A_01_r, loc=<6.6004 -0.0236 0.3246>, rot=<0.0012 0.0482 0.0056 -0.9988>}
 leg_strap_A_03_r {parent=leg_strap_A_02_r, loc=<6.2342 0 0>, rot=<-0.0265 -0.0581 0.0282 -0.9976>}
 extra_curve_calf_r {parent=calf_r, loc=<3.2674 2.8432 6.4605>, rot=<0.989 -0.0172 -0.087 -0.1185>}
 dyn_trouser_leg_r {parent=calf_r, loc=<31.5378 3.8674 2.5191>, rot=<-0.4842 0.5222 0.4399 -0.5472>}
 thigh_twist_01_r {parent=thigh_r, loc=<27.952 0 0>, rot=<0.0013 0 0 -1>}
 deform_groin_base_r {parent=thigh_r, loc=<-0.4801 -5.6776 0.2486>, rot=<0 0 0 -1>}
 deform_groin_r {parent=deform_groin_base_r, loc=<0 0 0>, rot=<0 0 0 1>}
 deform_glute_base_l {parent=pelvis, loc=<-7.3466 -9.0213 -9.1245>, rot=<0.0172 0.031 -0.4477 -0.8935>}
 deform_glute_l {parent=deform_glute_base_l, loc=<0 0 0>, rot=<0 0 0 1>}
 deform_glute_base_r {parent=pelvis, loc=<-7.3464 -9.0213 9.1245>, rot=<0.031 -0.0172 0.8935 -0.4477>}
 deform_glute_r {parent=deform_glute_base_r, loc=<0 0 0>, rot=<0 0 0 1>}
 dyn_suspenders_01_base_l {parent=pelvis, loc=<7.4861 5.6068 -10.0262>, rot=<0.052 -0.8982 -0.382 -0.2114>}
 dyn_suspenders_01_l {parent=dyn_suspenders_01_base_l, loc=<0 0 0>, rot=<0 0 0 1>}
 dyn_suspenders_02_l {parent=dyn_suspenders_01_l, loc=<7.1298 0 0>, rot=<0.0369 -0.0445 0.008 -0.9983>}
 dyn_suspenders_03_l {parent=dyn_suspenders_02_l, loc=<12.2287 0 0>, rot=<0.232 -0.0127 0.0319 -0.9721>}
 dyn_suspenders_04_l {parent=dyn_suspenders_03_l, loc=<12.0314 0 0>, rot=<0.2151 -0.1986 0.0442 -0.9552>}
 dyn_suspenders_05_l {parent=dyn_suspenders_04_l, loc=<9.4141 0 0>, rot=<-0.1893 -0.8178 0.0911 -0.5358>}
 dyn_suspenders_06_l {parent=dyn_suspenders_05_l, loc=<10.278 0 0>, rot=<0.1625 -0.2299 0.0855 -0.9557>}
 dyn_suspenders_07_l {parent=dyn_suspenders_06_l, loc=<8.1749 0 0>, rot=<-0.0788 -0.0397 0.0237 -0.9958>}
 dyn_suspenders_08_l {parent=dyn_suspenders_07_l, loc=<8.7759 0 0>, rot=<-0.0405 -0.033 0.0338 -0.9981>}
 dyn_suspenders_09_l {parent=dyn_suspenders_08_l, loc=<9.3879 0 0>, rot=<-0.0197 -0.0721 0.0082 -0.9972>}
 dyn_suspenders_10_l {parent=dyn_suspenders_09_l, loc=<7.8864 0 0>, rot=<0 -0.0976 0.9952 0>}
 dyn_suspenders_01_base_r {parent=pelvis, loc=<7.4865 5.6068 10.0262>, rot=<0.2114 0.382 0.8982 -0.052>}
 dyn_suspenders_01_r {parent=dyn_suspenders_01_base_r, loc=<0 0 0>, rot=<0 0 0 1>}
 dyn_suspenders_02_r {parent=dyn_suspenders_01_r, loc=<7.1301 -0.0001 -0.0001>, rot=<-0.0369 -0.0445 -0.008 -0.9983>}
 dyn_suspenders_03_r {parent=dyn_suspenders_02_r, loc=<12.2287 0.0001 0.0001>, rot=<-0.232 -0.0127 -0.0319 -0.9721>}
 dyn_suspenders_04_r {parent=dyn_suspenders_03_r, loc=<12.0314 -0.0001 0>, rot=<-0.2151 -0.1986 -0.0442 -0.9552>}
 dyn_suspenders_05_r {parent=dyn_suspenders_04_r, loc=<9.4141 0 0.0001>, rot=<0.1893 -0.8178 -0.0911 -0.5358>}
 dyn_suspenders_06_r {parent=dyn_suspenders_05_r, loc=<10.278 0.0001 0>, rot=<-0.1625 -0.2299 -0.0855 -0.9557>}
 dyn_suspenders_07_r {parent=dyn_suspenders_06_r, loc=<8.175 0 0>, rot=<0.0788 -0.0397 -0.0237 -0.9958>}
 dyn_suspenders_08_r {parent=dyn_suspenders_07_r, loc=<8.7758 0 0>, rot=<0.0405 -0.033 -0.0338 -0.9981>}
 dyn_suspenders_09_r {parent=dyn_suspenders_08_r, loc=<9.388 0 0>, rot=<0.0197 -0.0721 -0.0082 -0.9972>}
 dyn_suspenders_10_r {parent=dyn_suspenders_09_r, loc=<7.8861 0 -0.0001>, rot=<0 0.0976 0.9952 0>}
 extra_glute {parent=pelvis, loc=<-4.6231 -4.1937 0>, rot=<0 0 0 1>}
 dyn_root {parent=root, loc=<0 0 0>, rot=<0 0 0 1>}
]
variables [crouch {type=double, default=0}]
