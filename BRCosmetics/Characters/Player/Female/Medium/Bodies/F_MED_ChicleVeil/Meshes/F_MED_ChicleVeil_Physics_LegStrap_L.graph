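# Physics asset for the two left leg straps on the ChicleVeil female medium player body, covering only the bones those straps need.
# Strap chains A and B hang from the calf through tiny base capsules, with capsule segments joined by swing-limited, twist-locked constraints.
# Pelvis, left thigh and calf capsules form the body volume above them.
# A huge zero-length root capsule sits below the origin, its top just above it like a floor.
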PhysicsAsset /BRCosmetics/Characters/Player/Female/Medium/Bodies/F_MED_ChicleVeil/Meshes/F_MED_ChicleVeil_Physics_LegStrap_L
bodies {
 calf_l [{
   shapeType=Capsule
   length=30.8586
   radius=10.7352
   shapeName=calf_l_capsule
   center=<-25 -2.2358 1.1593>
   rotation=<rot 79.8876 175.0537 165.1687>
  }]
 leg_strap_A_01_base_l [{shapeType=Capsule, length=0.505, radius=0.505, shapeName=leg_strap_A_01_base_l_capsule, rotation=<rot 0 -90 0>}]
 leg_strap_A_01_l [{shapeType=Capsule, length=0.5, radius=2, shapeName=leg_strap_A_01_l_capsule, center=<4.0389 0 1.1729>, rotation=<rot 90 -180 -180>}]
 leg_strap_A_02_l [{shapeType=Capsule, length=1, radius=1.5, shapeName=leg_strap_A_02_l_capsule, center=<3.2521 0 0>, rotation=<rot 90 -180 -180>}]
 leg_strap_A_03_l [{shapeType=Capsule, length=6, radius=2.5, shapeName=leg_strap_A_03_l_capsule, center=<5.9736 0 0>, rotation=<rot 90 -180 -180>}]
 leg_strap_B_01_base_l [{shapeType=Capsule, length=0.505, radius=0.505, shapeName=leg_strap_B_01_base_l_capsule, rotation=<rot 0 -90 0>}]
 leg_strap_B_01_l [{shapeType=Capsule, length=6, radius=1, shapeName=leg_strap_B_01_l_capsule, center=<5.0896 0 0>, rotation=<rot 90 -180 -180>}]
 leg_strap_B_02_l [{shapeType=Capsule, length=8, radius=2, shapeName=leg_strap_B_02_l_capsule, center=<6.0102 0 0>, rotation=<rot 90 -180 -180>}]
 leg_strap_B_03_l [{shapeType=Capsule, length=4, radius=2.5, shapeName=leg_strap_B_03_l_capsule, center=<4.8093 0 0>, rotation=<rot 90 -180 -180>}]
 leg_strap_B_04_l [{shapeType=Capsule, length=7, radius=3, shapeName=leg_strap_B_04_l_capsule, center=<6.2264 0 0>, rotation=<rot 90 -180 -180>}]
 pelvis [{shapeType=Capsule, length=16.0294, radius=15.8372, center=<3.3281 -1.2887 0.2958>, rotation=<rot 1.5959 -90 -90.1849>}]
 root [{shapeType=Capsule, length=0, radius=300, shapeName=root_capsule, center=<0 0 -290>, rotation=<rot 0 -90 0>}]
 thigh_l [{shapeType=Capsule, length=37.5181, radius=9.7461, shapeName=thigh_l_capsule, center=<-20 0.0001 0>, rotation=<rot -90 -179.9998 179.9997>}]
}
constraints [
 {
  bone1Name=thigh_l
  bone2Name=pelvis
  swing1LimitDegrees=45
  swing1Motion=Free
  swing2LimitDegrees=45
  swing2Motion=Free
  twistLimitDegrees=45
  twistMotion=Free
 }
 {bone1Name=calf_l, bone2Name=thigh_l, swing1LimitDegrees=45, swing2LimitDegrees=45, twistLimitDegrees=5}
 {
  bone1Name=pelvis
  bone2Name=root
  swing1LimitDegrees=45
  swing1Motion=Free
  swing2LimitDegrees=45
  swing2Motion=Free
  twistLimitDegrees=45
  twistMotion=Free
 }
 {
  bone1Name=leg_strap_B_04_l
  bone2Name=leg_strap_B_03_l
  swing1LimitDegrees=45
  swing2LimitDegrees=15
  twistLimitDegrees=45
  twistMotion=Locked
 }
 {bone1Name=leg_strap_A_01_base_l, bone2Name=calf_l, swing1LimitDegrees=45, swing2LimitDegrees=15, twistLimitDegrees=45, twistMotion=Locked}
 {
  bone1Name=leg_strap_A_01_l
  bone2Name=leg_strap_A_01_base_l
  swing1LimitDegrees=45
  swing2LimitDegrees=15
  twistLimitDegrees=45
  twistMotion=Locked
 }
 {
  bone1Name=leg_strap_A_02_l
  bone2Name=leg_strap_A_01_l
  swing1LimitDegrees=45
  swing2LimitDegrees=15
  twistLimitDegrees=45
  twistMotion=Locked
 }
 {
  bone1Name=leg_strap_A_03_l
  bone2Name=leg_strap_A_02_l
  swing1LimitDegrees=45
  swing2LimitDegrees=15
  twistLimitDegrees=45
  twistMotion=Locked
 }
 {bone1Name=leg_strap_B_01_base_l, bone2Name=calf_l, swing1LimitDegrees=45, swing2LimitDegrees=15, twistLimitDegrees=45, twistMotion=Locked}
 {
  bone1Name=leg_strap_B_01_l
  bone2Name=leg_strap_B_01_base_l
  swing1LimitDegrees=45
  swing2LimitDegrees=15
  twistLimitDegrees=45
  twistMotion=Locked
 }
 {
  bone1Name=leg_strap_B_02_l
  bone2Name=leg_strap_B_01_l
  swing1LimitDegrees=45
  swing2LimitDegrees=15
  twistLimitDegrees=45
  twistMotion=Locked
 }
 {
  bone1Name=leg_strap_B_03_l
  bone2Name=leg_strap_B_02_l
  swing1LimitDegrees=45
  swing2LimitDegrees=15
  twistLimitDegrees=45
  twistMotion=Locked
 }
]
