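# Physics asset for the right-leg straps on the female medium ChicleVeil body, a variant named for the crouching pose.
# Capsules on pelvis, thigh, calf and foot form the leg, while a huge zero-length capsule offset from dyn_root apparently approximates the ground.
# Two strap chains, A with three segments and B with four, hang from small base bodies constrained to the calf.
# Strap segments have limited swing and locked twist, while leg and base joints rotate freely.
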PhysicsAsset /BRCosmetics/Characters/Player/Female/Medium/Bodies/F_MED_ChicleVeil/Meshes/F_MED_ChicleVeil_Physics_LegStrap_R_crouching
bodies {
 calf_r [{shapeType=Capsule, length=40, radius=12, shapeName=calf_l_capsule, center=<25 2.2358 -1.1593>, rotation=<rot 79.8876 175.0537 -14.8313>}]
 dyn_root [{shapeType=Capsule, length=0, radius=296, shapeName=dyn_root_capsule, center=<8.5742 -32.1423 -290>, rotation=<rot 0 -90 0>}]
 foot_r [{
   shapeType=Capsule
   length=13.2592
   radius=11.5653
   shapeName=foot_r_capsule
   center=<5.2632 -8.3565 -0.4121>
   rotation=<rot -0.648 1.4432 -88.2195>
  }]
 leg_strap_A_01_base_r [{shapeType=Capsule, length=0.505, radius=0.505, shapeName=leg_strap_A_01_base_r_capsule, rotation=<rot 0 -90 0>}]
 leg_strap_A_01_r [{shapeType=Capsule, length=0.5, radius=2, shapeName=leg_strap_A_01_r_capsule, center=<4.0389 0 1.1729>, rotation=<rot -90 -179.9975 179.9974>}]
 leg_strap_A_02_r [{shapeType=Capsule, length=1, radius=2, shapeName=leg_strap_A_02_r_capsule, center=<3.2521 0 0>, rotation=<rot -90 179.9604 -179.9605>}]
 leg_strap_A_03_r [{shapeType=Capsule, length=6, radius=2, shapeName=leg_strap_A_03_r_capsule, center=<5.9736 0.6724 0.9195>, rotation=<rot -90 0.0001 -0.0002>}]
 leg_strap_B_01_base_r [{shapeType=Capsule, length=0.505, radius=0.505, shapeName=leg_strap_B_01_base_r_capsule, rotation=<rot 0 -90 0>}]
 leg_strap_B_01_r [{shapeType=Capsule, length=6, radius=2, shapeName=leg_strap_B_01_r_capsule, center=<5.0896 0 0>, rotation=<rot -90 -90 89.9998>}]
 leg_strap_B_02_r [{shapeType=Capsule, length=8, radius=3, shapeName=leg_strap_B_02_r_capsule, center=<6.0102 0 0>, rotation=<rot -90 -90 89.9998>}]
 leg_strap_B_03_r [{shapeType=Capsule, length=4, radius=2.5, shapeName=leg_strap_B_03_r_capsule, center=<4.8093 0 0>, rotation=<rot -90 179.9999 179.9999>}]
 leg_strap_B_04_r [{shapeType=Capsule, length=7, radius=2, shapeName=leg_strap_B_04_r_capsule, center=<6.2264 0 0>, rotation=<rot -90 -90 89.9998>}]
 pelvis [{shapeType=Capsule, length=16.0294, radius=15.8372, center=<3.3281 -1.2887 0.2958>, rotation=<rot 1.5959 -90 -90.1849>}]
 thigh_r [{shapeType=Capsule, length=37.5181, radius=9.7461, shapeName=thigh_l_capsule, center=<20 -0.0001 0>, rotation=<rot -90 -179.9998 -0.0003>}]
}
constraints [
 {
  bone1Name=thigh_r
  bone2Name=pelvis
  swing1LimitDegrees=45
  swing1Motion=Free
  swing2LimitDegrees=45
  swing2Motion=Free
  twistLimitDegrees=45
  twistMotion=Free
 }
 {
  bone1Name=calf_r
  bone2Name=thigh_r
  swing1LimitDegrees=45
  swing1Motion=Free
  swing2LimitDegrees=45
  swing2Motion=Free
  twistLimitDegrees=45
  twistMotion=Free
 }
 {
  bone1Name=foot_r
  bone2Name=calf_r
  swing1LimitDegrees=45
  swing1Motion=Free
  swing2LimitDegrees=45
  swing2Motion=Free
  twistLimitDegrees=45
  twistMotion=Free
 }
 {
  bone1Name=leg_strap_B_04_r
  bone2Name=leg_strap_B_03_r
  swing1LimitDegrees=15
  swing2LimitDegrees=30
  twistLimitDegrees=45
  twistMotion=Locked
 }
 {
  bone1Name=leg_strap_A_01_base_r
  bone2Name=calf_r
  swing1LimitDegrees=45
  swing1Motion=Free
  swing2LimitDegrees=45
  swing2Motion=Free
  twistLimitDegrees=45
  twistMotion=Free
 }
 {
  bone1Name=leg_strap_A_01_r
  bone2Name=leg_strap_A_01_base_r
  swing1LimitDegrees=45
  swing2LimitDegrees=45
  twistLimitDegrees=45
  twistMotion=Locked
 }
 {
  bone1Name=leg_strap_A_02_r
  bone2Name=leg_strap_A_01_r
  swing1LimitDegrees=45
  swing2LimitDegrees=15
  twistLimitDegrees=45
  twistMotion=Locked
 }
 {
  bone1Name=leg_strap_A_03_r
  bone2Name=leg_strap_A_02_r
  swing1LimitDegrees=45
  swing2LimitDegrees=15
  twistLimitDegrees=45
  twistMotion=Locked
 }
 {
  bone1Name=leg_strap_B_01_base_r
  bone2Name=calf_r
  swing1LimitDegrees=45
  swing1Motion=Free
  swing2LimitDegrees=45
  swing2Motion=Free
  twistLimitDegrees=45
  twistMotion=Free
 }
 {
  bone1Name=leg_strap_B_01_r
  bone2Name=leg_strap_B_01_base_r
  swing1LimitDegrees=45
  swing2LimitDegrees=45
  twistLimitDegrees=45
  twistMotion=Locked
 }
 {
  bone1Name=leg_strap_B_02_r
  bone2Name=leg_strap_B_01_r
  swing1LimitDegrees=15
  swing2LimitDegrees=30
  twistLimitDegrees=45
  twistMotion=Locked
 }
 {
  bone1Name=leg_strap_B_03_r
  bone2Name=leg_strap_B_02_r
  swing1LimitDegrees=15
  swing2LimitDegrees=30
  twistLimitDegrees=45
  twistMotion=Locked
 }
]
