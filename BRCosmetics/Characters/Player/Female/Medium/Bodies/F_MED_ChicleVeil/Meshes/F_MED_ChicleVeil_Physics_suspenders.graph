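# Physics asset for the dangling suspenders on the F_MED_ChicleVeil female character body, giving the dyn_suspenders bones collision and joint limits.
# Left and right capsule chains hang from the pelvis, linked by swing- and twist-limited constraints, with looser joints at segment 09 and locked twist at the tips.
# Pelvis, thigh and calf capsules form the body shapes with free leg joints, while a huge zero-length root capsule sits below as a near-flat floor collider.
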PhysicsAsset /BRCosmetics/Characters/Player/Female/Medium/Bodies/F_MED_ChicleVeil/Meshes/F_MED_ChicleVeil_Physics_suspenders
bodies {
 calf_l [{
   shapeType=Capsule
   length=30.8586
   radius=10.7352
   shapeName=calf_l_capsule
   center=<-22.2227 -8.0177 1.7309>
   rotation=<rot 83.5899 174.867 167.5275>
  }]
 calf_r [{
   shapeType=Capsule
   length=30.8586
   radius=10.7352
   shapeName=calf_l_capsule
   center=<22.2227 8.0177 -1.7309>
   rotation=<rot 83.5899 174.867 -12.4725>
  }]
 dyn_suspenders_01_base_l [{shapeType=Capsule, length=1, radius=2, shapeName=dyn_suspenders_01_base_l_capsule, rotation=<rot 90 0 0>}]
 dyn_suspenders_01_base_r [{shapeType=Capsule, length=1, radius=2, shapeName=dyn_suspenders_01_base_l_capsule, rotation=<rot 90 180 0>}]
 dyn_suspenders_01_l [{shapeType=Capsule, length=1, radius=1, shapeName=dyn_suspenders_01_l_capsule, center=<3 0 0>, rotation=<rot 90 0 0>}]
 dyn_suspenders_01_r [{shapeType=Capsule, length=1, radius=1, shapeName=dyn_suspenders_01_l_capsule, center=<3 0 0>, rotation=<rot 90 180 0>}]
 dyn_suspenders_02_l [{shapeType=Capsule, length=5, radius=2, shapeName=dyn_suspenders_02_l_capsule, center=<5 0 0>, rotation=<rot 90 0 0>}]
 dyn_suspenders_02_r [{shapeType=Capsule, length=5, radius=2, shapeName=dyn_suspenders_02_l_capsule, center=<5 0 0>, rotation=<rot 90 180 0>}]
 dyn_suspenders_03_l [{shapeType=Capsule, length=5, radius=2, shapeName=dyn_suspenders_03_l_capsule, center=<5 0 0>, rotation=<rot 90 0 0>}]
 dyn_suspenders_03_r [{shapeType=Capsule, length=5, radius=2, shapeName=dyn_suspenders_03_r_capsule, center=<5 0 0>, rotation=<rot 90 0 0>}]
 dyn_suspenders_04_l [{shapeType=Capsule, length=5, radius=2, shapeName=dyn_suspenders_04_l_capsule, center=<5 0 0>, rotation=<rot 90 0 0>}]
 dyn_suspenders_04_r [{shapeType=Capsule, length=5, radius=2, shapeName=dyn_suspenders_04_r_capsule, center=<5 0 0>, rotation=<rot 90 0 0>}]
 dyn_suspenders_05_l [{shapeType=Capsule, length=5, radius=2, shapeName=dyn_suspenders_05_l_capsule, center=<5 0 0>, rotation=<rot 90 0 0>}]
 dyn_suspenders_05_r [{shapeType=Capsule, length=5, radius=2, shapeName=dyn_suspenders_05_r_capsule, center=<5 0 0>, rotation=<rot 90 0 0>}]
 dyn_suspenders_06_l [{shapeType=Capsule, length=5, radius=2, shapeName=dyn_suspenders_06_l_capsule, center=<5 0 0>, rotation=<rot 90 0 0>}]
 dyn_suspenders_06_r [{shapeType=Capsule, length=5, radius=2, shapeName=dyn_suspenders_06_r_capsule, center=<5 0 0>, rotation=<rot 90 0 0>}]
 dyn_suspenders_07_l [{shapeType=Capsule, length=5, radius=4, shapeName=dyn_suspenders_07_l_capsule, center=<5 0 0>, rotation=<rot 90 0 0>}]
 dyn_suspenders_07_r [{shapeType=Capsule, length=5, radius=2, shapeName=dyn_suspenders_07_r_capsule, center=<5 0 0>, rotation=<rot 90 0 0>}]
 dyn_suspenders_08_l [{shapeType=Capsule, length=5, radius=3, shapeName=dyn_suspenders_08_l_capsule, center=<5 0 0>, rotation=<rot 90 0 0>}]
 dyn_suspenders_08_r [{shapeType=Capsule, length=5, radius=2.5, shapeName=dyn_suspenders_08_r_capsule, center=<5 0 0>, rotation=<rot 90 0 0>}]
 dyn_suspenders_09_l [{shapeType=Capsule, length=5, radius=1, shapeName=dyn_suspenders_09_l_capsule, center=<5 0 0>, rotation=<rot 90 0 0>}]
 dyn_suspenders_09_r [{shapeType=Capsule, length=5, radius=1, shapeName=dyn_suspenders_09_l_capsule, center=<5 0 0>, rotation=<rot 90 180 0>}]
 dyn_suspenders_10_l [{shapeType=Capsule, length=0.505, radius=0.505, shapeName=dyn_suspenders_10_l_capsule, rotation=<rot 0 -90 0>}]
 dyn_suspenders_10_r [{shapeType=Capsule, length=0.505, radius=0.505, shapeName=dyn_suspenders_10_l_capsule, rotation=<rot 0 -90 180>}]
 pelvis [
  {shapeType=Capsule, length=9.5, radius=10.836, center=<0 -1.2267 0>}
  {shapeType=Capsule, length=5, radius=15, shapeName=pelvis_capsule, center=<-16.0249 -18.9836 0>}
  {shapeType=Capsule, length=21, radius=15, shapeName=pelvis_capsule, center=<-9.5254 21.5366 0>}
 ]
 root [{shapeType=Capsule, length=0, radius=300, shapeName=root_capsule, center=<0 0 -290>, rotation=<rot 0 -90 0>}]
 thigh_l [{shapeType=Capsule, length=37.5181, radius=9.7461, shapeName=thigh_l_capsule, center=<-20 -2.9455 0>, rotation=<rot -90 -179.9998 179.9997>}]
 thigh_r [{shapeType=Capsule, length=37.5181, radius=9.7461, shapeName=thigh_l_capsule, center=<20 2.9455 0>, rotation=<rot -90 -179.9998 -0.0003>}]
}
constraints [
 {
  bone1Name=thigh_l
  bone2Name=pelvis
  swing1LimitDegrees=45
  swing1Motion=Free
  swing2LimitDegrees=45
  swing2Motion=Free
  twistLimitDegrees=45
  twistMotion=Free
 }
 {bone1Name=dyn_suspenders_01_base_l, bone2Name=pelvis, swing1LimitDegrees=45, swing2LimitDegrees=45, twistLimitDegrees=5}
 {bone1Name=dyn_suspenders_01_l, bone2Name=dyn_suspenders_01_base_l, swing1LimitDegrees=45, swing2LimitDegrees=45, twistLimitDegrees=5}
 {bone1Name=dyn_suspenders_02_l, bone2Name=dyn_suspenders_01_l, swing1LimitDegrees=45, swing2LimitDegrees=45, twistLimitDegrees=5}
 {bone1Name=dyn_suspenders_03_l, bone2Name=dyn_suspenders_02_l, swing1LimitDegrees=45, swing2LimitDegrees=45, twistLimitDegrees=5}
 {bone1Name=dyn_suspenders_04_l, bone2Name=dyn_suspenders_03_l, swing1LimitDegrees=45, swing2LimitDegrees=45, twistLimitDegrees=5}
 {bone1Name=dyn_suspenders_05_l, bone2Name=dyn_suspenders_04_l, swing1LimitDegrees=45, swing2LimitDegrees=45, twistLimitDegrees=5}
 {bone1Name=dyn_suspenders_06_l, bone2Name=dyn_suspenders_05_l, swing1LimitDegrees=45, swing2LimitDegrees=45, twistLimitDegrees=5}
 {bone1Name=dyn_suspenders_07_l, bone2Name=dyn_suspenders_06_l, swing1LimitDegrees=45, swing2LimitDegrees=45, twistLimitDegrees=5}
 {bone1Name=dyn_suspenders_08_l, bone2Name=dyn_suspenders_07_l, swing1LimitDegrees=45, swing2LimitDegrees=45, twistLimitDegrees=5}
 {bone1Name=dyn_suspenders_09_l, bone2Name=dyn_suspenders_08_l, swing1LimitDegrees=60, swing2LimitDegrees=60, twistLimitDegrees=15}
 {bone1Name=dyn_suspenders_03_r, bone2Name=dyn_suspenders_02_r, swing1LimitDegrees=45, swing2LimitDegrees=45, twistLimitDegrees=5}
 {bone1Name=dyn_suspenders_04_r, bone2Name=dyn_suspenders_03_r, swing1LimitDegrees=45, swing2LimitDegrees=45, twistLimitDegrees=5}
 {bone1Name=dyn_suspenders_05_r, bone2Name=dyn_suspenders_04_r, swing1LimitDegrees=45, swing2LimitDegrees=45, twistLimitDegrees=5}
 {bone1Name=dyn_suspenders_06_r, bone2Name=dyn_suspenders_05_r, swing1LimitDegrees=45, swing2LimitDegrees=45, twistLimitDegrees=5}
 {bone1Name=dyn_suspenders_07_r, bone2Name=dyn_suspenders_06_r, swing1LimitDegrees=45, swing2LimitDegrees=45, twistLimitDegrees=5}
 {bone1Name=dyn_suspenders_08_r, bone2Name=dyn_suspenders_07_r, swing1LimitDegrees=45, swing2LimitDegrees=45, twistLimitDegrees=5}
 {
  bone1Name=dyn_suspenders_10_l
  bone2Name=dyn_suspenders_09_l
  swing1LimitDegrees=30
  swing2LimitDegrees=30
  twistLimitDegrees=45
  twistMotion=Locked
 }
 {
  bone1Name=thigh_r
  bone2Name=pelvis
  swing1LimitDegrees=45
  swing1Motion=Free
  swing2LimitDegrees=45
  swing2Motion=Free
  twistLimitDegrees=45
  twistMotion=Free
 }
 {bone1Name=dyn_suspenders_01_base_r, bone2Name=pelvis, swing1LimitDegrees=45, swing2LimitDegrees=45, twistLimitDegrees=5}
 {bone1Name=dyn_suspenders_01_r, bone2Name=dyn_suspenders_01_base_r, swing1LimitDegrees=45, swing2LimitDegrees=45, twistLimitDegrees=5}
 {bone1Name=dyn_suspenders_02_r, bone2Name=dyn_suspenders_01_r, swing1LimitDegrees=45, swing2LimitDegrees=45, twistLimitDegrees=5}
 {bone1Name=dyn_suspenders_09_r, bone2Name=dyn_suspenders_08_r, swing1LimitDegrees=60, swing2LimitDegrees=60, twistLimitDegrees=15}
 {
  bone1Name=dyn_suspenders_10_r
  bone2Name=dyn_suspenders_09_r
  swing1LimitDegrees=30
  swing2LimitDegrees=30
  twistLimitDegrees=45
  twistMotion=Locked
 }
 {
  bone1Name=calf_l
  bone2Name=thigh_l
  swing1LimitDegrees=45
  swing1Motion=Free
  swing2LimitDegrees=45
  swing2Motion=Free
  twistLimitDegrees=45
  twistMotion=Free
 }
 {
  bone1Name=calf_r
  bone2Name=thigh_r
  swing1LimitDegrees=45
  swing1Motion=Free
  swing2LimitDegrees=45
  swing2Motion=Free
  twistLimitDegrees=45
  twistMotion=Free
 }
]
