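# Post-process style AnimBP for the ChicleVeil female body cosmetic, copying the attached parent's pose and adding secondary motion.
# Deform and crouch-driven Control Rigs feed a suspenders RigidBody sim, bracketed by a temporary thigh offset, into a cached pose.
# Two leg-strap sim branches blend on thigh twist or vertical velocity, then constraints, trouser AnimDynamics and glute adjustments follow.
# The event graph updates those states and drives material WPO cloth wind and drag.
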AnimBlueprint /BRCosmetics/Characters/Player/Female/Medium/Bodies/F_MED_ChicleVeil/Meshes/F_MED_ChicleVeil_AnimBP
target_skeleton=&/F_MED_ChicleVeil_Skeleton
graphs [
 AnimGraph {nodes [
   n1 = "Transform (Modify) Bone - Bone: leg_strap_A_01_base_r"!(
    ComponentPose=Local To Component(LocalPose=Use cached pose 'pre_sim'())
    bAlphaBoolEnabled:bool=True
    Translation="0,0,0"
    Rotation="3.101,-61.165,-0.382"
    Scale="1,1,1"
   ) -> (Pose) {
    alphaInputType=Bool
    rotationMode=BMM_Additive
    rotationSpace=BCS_BoneSpace
    translationMode=BMM_Additive
    translationSpace=BCS_BoneSpace
    alphaBoolBlend {blendInTime=0.2, blendOption=Linear, blendOutTime=0.2}
    boneToModify {boneName=leg_strap_A_01_base_r}
    rotation=<rot 3.101 -61.1651 -0.382>
   }
   n2 = "Transform (Modify) Bone - Bone: leg_strap_A_01_base_l"!(ComponentPose=@n1, bAlphaBoolEnabled:bool=True, Translation="0,0,0", Rotation="2.917,91.208,0.984", Scale="1,1,1") -> (Pose) {
    alphaInputType=Bool
    rotationMode=BMM_Additive
    rotationSpace=BCS_BoneSpace
    translationMode=BMM_Additive
    translationSpace=BCS_BoneSpace
    alphaBoolBlend {blendInTime=0.2, blendOption=Linear, blendOutTime=0.2}
    boneToModify {boneName=leg_strap_A_01_base_l}
    rotation=<rot 2.917 91.2081 0.984>
   }
   n3 = RigidBody!(ComponentPose=@n2, Alpha, OverrideWorldGravity:Vector="0,0,-980", ExternalForce) -> (Pose) {
    bOverrideWorldGravity=true
    overridePhysicsAsset=&/F_MED_ChicleVeil_Physics_F_MED_ChicleVeil_Physics_LegStrap_L_crouching
    simulationSpace=BaseBoneSpace
    alphaBoolBlend {blendOption=Linear}
    baseBoneRef {boneName=calf_l}
    componentLinearAccScale=<0.3 0.3 0.3>
    componentLinearVelScale=<0.3 0.3 0.3>
    simSpaceSettings {worldAlpha=0.05}
   }
   n4 = RigidBody(ComponentPose=@n3, OverrideWorldGravity="0,0,-980") {
    bOverrideWorldGravity=true
    overridePhysicsAsset=&/F_MED_ChicleVeil_Physics_LegStrap_R_crouching
    simulationSpace=BaseBoneSpace
    alphaBoolBlend {blendOption=Linear}
    baseBoneRef {boneName=calf_r}
    componentLinearAccScale=<0.3 0.3 0.3>
    componentLinearVelScale=<0.3 0.3 0.3>
    simSpaceSettings {worldAlpha=0.05}
   }
   n5 = Component To Local(ComponentPose=@n4)
   n6 = RigidBody(ComponentPose=Local To Component(LocalPose=Use cached pose 'pre_sim'()), OverrideWorldGravity="0,0,-980") {
    bOverrideWorldGravity=true
    overridePhysicsAsset=&/F_MED_ChicleVeil_Physics_LegStrap_L
    simulationSpace=BaseBoneSpace
    alphaBoolBlend {blendOption=Linear}
    baseBoneRef {boneName=calf_l}
    componentLinearAccScale=<0.3 0.3 0.3>
    componentLinearVelScale=<0.3 0.3 0.3>
    simSpaceSettings {worldAlpha=0.05}
   }
   n7 = "Transform (Modify) Bone - Bone: dyn_root"!(ComponentPose=@n6, bAlphaBoolEnabled:bool=True, Translation="0,0,-4", Rotation="0,0,0.079", Scale="1,1,1") -> (Pose) {
    alphaInputType=Bool
    rotationMode=BMM_Additive
    rotationSpace=BCS_ParentBoneSpace
    translationMode=BMM_Additive
    alphaBoolBlend {blendInTime=0.2, blendOption=Linear, blendOutTime=0.2}
    boneToModify {boneName=dyn_root}
    rotation {roll=0.079}
    translation {z=-4}
   }
   n8 = RigidBody(ComponentPose=@n7, OverrideWorldGravity="0,0,-980") {
    bOverrideWorldGravity=true
    overridePhysicsAsset=&/F_MED_ChicleVeil_Physics_LegStrap_R
    simulationSpace=BaseBoneSpace
    alphaBoolBlend {blendOption=Linear}
    baseBoneRef {boneName=calf_r}
    componentLinearAccScale=<0.3 0.3 0.3>
    componentLinearVelScale=<0.3 0.3 0.3>
    simSpaceSettings {worldAlpha=0.05}
   }
   n9 = Component To Local(ComponentPose=@n8)
   n10 = Get is_move() -> (is_move:bool)
   n11 = "Constraint - extra_curve_back_01_c (PP)"!(
    ComponentPose=Local To Component(LocalPose=Blend Poses by bool(BlendPose_0=@n5, BlendPose_1=@n9, BlendTime_0=0.4, BlendTime_1=0.4, bActiveValue=@n10) {blendType=Linear})
    Alpha=1
    ConstraintWeights_0=0.3914
    ConstraintWeights_1:float=0.6086
   ) -> (Pose) {
    alphaBoolBlend {blendOption=Linear}
    boneToModify {boneName=extra_curve_back_01_c}
    constraintSetup [{targetBone {boneName=spine_04}}, {targetBone {boneName=spine_05}}]
    constraintWeights [1,1]
   }
   n12 = "Constraint - extra_curve_back_02_c (PP)"!(ComponentPose=@n11, Alpha=1, ConstraintWeights_0=0.749, ConstraintWeights_1:float=0.251) -> (Pose) {
    alphaBoolBlend {blendOption=Linear}
    boneToModify {boneName=extra_curve_back_02_c}
    constraintSetup [{targetBone {boneName=spine_04}}, {targetBone {boneName=spine_05}}]
    constraintWeights [1,1]
   }
   n13 = "Constraint - extra_curve_back_03_c (PPP)"!(ComponentPose=@n12, Alpha=1, ConstraintWeights_0=0.2172, ConstraintWeights_1:float=0.2984, ConstraintWeights_2:float=0.4845) -> (Pose) {
    alphaBoolBlend {blendOption=Linear}
    boneToModify {boneName=extra_curve_back_03_c}
    constraintSetup [{targetBone {boneName=spine_02}}, {targetBone {boneName=spine_03}}, {targetBone {boneName=spine_04}}]
    constraintWeights [1,1,1]
   }
   n14 = "Constraint - extra_curve_back_02_r (PP)"!(ComponentPose=@n13, Alpha=1, ConstraintWeights_0=0.2995, ConstraintWeights_1:float=0.7005) -> (Pose) {
    alphaBoolBlend {blendOption=Linear}
    boneToModify {boneName=extra_curve_back_02_r}
    constraintSetup [{targetBone {boneName=spine_04}}, {targetBone {boneName=spine_05}}]
    constraintWeights [1,1]
   }
   n15 = "Constraint - extra_curve_back_02_l (PP)"!(ComponentPose=@n14, Alpha=1, ConstraintWeights_0=0.205, ConstraintWeights_1:float=0.795) -> (Pose) {
    alphaBoolBlend {blendOption=Linear}
    boneToModify {boneName=extra_curve_back_02_l}
    constraintSetup [{targetBone {boneName=spine_04}}, {targetBone {boneName=spine_05}}]
    constraintWeights [1,1]
   }
   n16 = "Constraint - extra_curve_side_01_l (PPPP)"!(
    ComponentPose=@n15
    Alpha=1
    ConstraintWeights_0=0.2489
    ConstraintWeights_1:float=0.1751
    ConstraintWeights_2:float=0.5522
    ConstraintWeights_3:float=0.0238
   ) -> (Pose) {
    alphaBoolBlend {blendOption=Linear}
    boneToModify {boneName=extra_curve_side_01_l}
    constraintSetup [{targetBone {boneName=spine_02}}, {targetBone {boneName=spine_03}}, {targetBone {boneName=spine_04}}, {targetBone {boneName=spine_05}}]
    constraintWeights [1,1,1,1]
   }
   n17 = "Constraint - extra_curve_side_01_r (PPPP)"!(
    ComponentPose=@n16
    Alpha=1
    ConstraintWeights_0=0.2489
    ConstraintWeights_1:float=0.1751
    ConstraintWeights_2:float=0.5522
    ConstraintWeights_3:float=0.0238
   ) -> (Pose) {
    alphaBoolBlend {blendOption=Linear}
    boneToModify {boneName=extra_curve_side_01_r}
    constraintSetup [{targetBone {boneName=spine_02}}, {targetBone {boneName=spine_03}}, {targetBone {boneName=spine_04}}, {targetBone {boneName=spine_05}}]
    constraintWeights [1,1,1,1]
   }
   n18 = "Constraint - extra_curve_back_02_c (PPPPPPPP)"!(
    ComponentPose=@n17
    Alpha=1
    ConstraintWeights_0=0.1153
    ConstraintWeights_1:float=0.0873
    ConstraintWeights_2:float=0.2538
    ConstraintWeights_3:float=0.2919
    ConstraintWeights_4:float=0.008
    ConstraintWeights_5:float=0.0993
    ConstraintWeights_6:float=0.1143
    ConstraintWeights_7:float=0.0301
   ) -> (Pose) {
    alphaBoolBlend {blendOption=Linear}
    boneToModify {boneName=extra_curve_back_02_c}
    constraintSetup [
     {targetBone {boneName=spine_05}}
     {targetBone {boneName=spine_02}}
     {targetBone {boneName=extra_curve_back_02_c}}
     {targetBone {boneName=spine_04}}
     {targetBone {boneName=extra_curve_back_01_c}}
     {targetBone {boneName=extra_curve_back_03_c}}
     {targetBone {boneName=spine_03}}
     {targetBone {boneName=spine_01}}
    ]
    constraintWeights [1,1,1,1,1,1,1,1]
   }
   n19 = Anim Dynamics - Bone: dyn_trouser_leg_r(ComponentPose=@n18, Alpha=1, ExternalForce="0,0,0") {
    bLinearSpring=true
    bOverrideLinearDamping=true
    bUseGravityOverride=true
    linearDampingOverride=0.9999
    linearSpringConstant=175
    simulationSpace=BoneRelative
    alphaBoolBlend {blendOption=Linear}
    boundBone {boneName=dyn_trouser_leg_r}
    componentAppliedLinearAccClamp=<3000 3000 3000>
    componentLinearAccScale=<0.1 0.1 0.1>
    componentLinearVelScale=<0.9 0.9 0.9>
    gravityOverride {z=-980}
    physicsBodyDefinitions [{
      boundBone {boneName=dyn_trouser_leg_r}
      boxExtents=<60 60 60>
      constraintSetup {angularLimitsMax=<0 0 0>, angularLimitsMin=<0 0 0>, angularTarget=<0 0 0>, linearAxesMax=<5 4 5>, linearAxesMin=<-1.5 -4 -5>}
      localJointOffset=<0 0 0>
     }]
    relativeSpaceBone {boneName=foot_l}
   }
   n20 = Anim Dynamics - Bone: dyn_trouser_leg_l(ComponentPose=@n19, Alpha=1, ExternalForce="0,0,0") {
    bLinearSpring=true
    bOverrideLinearDamping=true
    bUseGravityOverride=true
    linearDampingOverride=0.9999
    linearSpringConstant=175
    simulationSpace=BoneRelative
    alphaBoolBlend {blendOption=Linear}
    boundBone {boneName=dyn_trouser_leg_l}
    componentAppliedLinearAccClamp=<3000 3000 3000>
    componentLinearAccScale=<0.1 0.1 0.1>
    componentLinearVelScale=<0.9 0.9 0.9>
    gravityOverride {z=-980}
    physicsBodyDefinitions [{
      boundBone {boneName=dyn_trouser_leg_l}
      boxExtents=<60 60 60>
      constraintSetup {angularLimitsMax=<0 0 0>, angularLimitsMin=<0 0 0>, angularTarget=<0 0 0>, linearAxesMax=<5 4 5>, linearAxesMin=<-1.5 -4 -5>}
      localJointOffset=<0 0 0>
     }]
    relativeSpaceBone {boneName=foot_l}
   }
   n21 = "Constraint - deform_glute_l (T)"!(ComponentPose=@n20, bAlphaBoolEnabled:bool=True, ConstraintWeights_0=1) -> (Pose) {
    alphaInputType=Bool
    alphaBoolBlend {blendOption=Linear}
    boneToModify {boneName=deform_glute_l}
    constraintSetup [{transformType=Translation, targetBone {boneName=pelvis}}]
    constraintWeights [1]
   }
   n22 = "Constraint - deform_glute_r (T)"!(ComponentPose=@n21, bAlphaBoolEnabled:bool=True, ConstraintWeights_0=1) -> (Pose) {
    alphaInputType=Bool
    alphaBoolBlend {blendOption=Linear}
    boneToModify {boneName=deform_glute_r}
    constraintSetup [{transformType=Translation, targetBone {boneName=pelvis}}]
    constraintWeights [1]
   }
   n23 = "Transform (Modify) Bone - Bone: deform_glute_l"!(ComponentPose=@n22, bAlphaBoolEnabled:bool=True, Translation="0,0,0", Rotation="0,0,0", Scale="2,2,2") -> (Pose) {
    alphaInputType=Bool
    scaleMode=BMM_Additive
    scaleSpace=BCS_BoneSpace
    translationMode=BMM_Additive
    alphaBoolBlend {blendInTime=0.2, blendOption=Linear, blendOutTime=0.2}
    boneToModify {boneName=deform_glute_l}
    scale=<2 2 2>
   }
   n24 = "Transform (Modify) Bone - Bone: deform_glute_r"!(ComponentPose=@n23, bAlphaBoolEnabled:bool=True, Translation="0,0,0", Rotation="0,0,0", Scale="2,2,2") -> (Pose) {
    alphaInputType=Bool
    scaleMode=BMM_Additive
    scaleSpace=BCS_BoneSpace
    translationMode=BMM_Additive
    alphaBoolBlend {blendInTime=0.2, blendOption=Linear, blendOutTime=0.2}
    boneToModify {boneName=deform_glute_r}
    scale=<2 2 2>
   }
   n25 = "Transform (Modify) Bone - Bone: extra_glute"!(ComponentPose=@n24, bAlphaBoolEnabled:bool=True, Translation="0.000,-6.510,3.614", Rotation="0,0,0", Scale="1.5,1.5,1.5") -> (Pose) {
    alphaInputType=Bool
    rotationMode=BMM_Additive
    rotationSpace=BCS_BoneSpace
    scaleMode=BMM_Additive
    scaleSpace=BCS_BoneSpace
    translationMode=BMM_Additive
    alphaBoolBlend {blendInTime=0.2, blendOption=Linear, blendOutTime=0.2}
    boneToModify {boneName=extra_glute}
    scale=<1.5 1.5 1.5>
    translation {y=-6.51, z=3.614}
   }
   Output Pose(Result=Component To Local(ComponentPose=@n25)) {layerGroup=None}
   n26 = Control Rig!(
    Source=Copy Pose From Mesh() {bUseAttachedParent=true}
    Alpha
    deform_value:FN Deform Rig Value Struct Structure=Get deform() -> (deform:FN Deform Rig Value Struct Structure)
   ) -> (Pose) {
    alphaBoolBlend {blendOption=Linear}
    controlRigAssetReference {blueprintRigClass=&/F_MED_ChicleVeil_DeformRig.F_MED_ChicleVeil_DeformRig_C}
   }
   n27 = "Constraint - extra_pec_l (PPP)"!(
    ComponentPose=Local To Component(LocalPose=@n26)
    Alpha=1
    ConstraintWeights_0=0.5
    ConstraintWeights_1:float=0.3
    ConstraintWeights_2:float=0.2
   ) -> (Pose) {
    alphaBoolBlend {blendOption=Linear}
    boneToModify {boneName=extra_pec_l}
    constraintSetup [{targetBone {boneName=spine_05}}, {targetBone {boneName=spine_04}}, {targetBone {boneName=clavicle_l}}]
    constraintWeights [1,1,1]
   }
   n28 = "Constraint - extra_pec_r (PPP)"!(ComponentPose=@n27, Alpha=1, ConstraintWeights_0=0.5, ConstraintWeights_1:float=0.3, ConstraintWeights_2:float=0.2) -> (Pose) {
    alphaBoolBlend {blendOption=Linear}
    boneToModify {boneName=extra_pec_r}
    constraintSetup [{targetBone {boneName=spine_05}}, {targetBone {boneName=spine_04}}, {targetBone {boneName=clavicle_r}}]
    constraintWeights [1,1,1]
   }
   n29 = "Constraint - extra_armor_hand_l (PPPP)"!(
    ComponentPose=@n28
    Alpha=1
    ConstraintWeights_0=0.25
    ConstraintWeights_1:float=0.25
    ConstraintWeights_2:float=0.25
    ConstraintWeights_3:float=0.25
   ) -> (Pose) {
    alphaBoolBlend {blendOption=Linear}
    boneToModify {boneName=extra_armor_hand_l}
    constraintSetup [
     {targetBone {boneName=index_metacarpal_l}}
     {targetBone {boneName=middle_metacarpal_l}}
     {targetBone {boneName=ring_metacarpal_l}}
     {targetBone {boneName=pinky_metacarpal_l}}
    ]
    constraintWeights [1,1,1,1]
   }
   n30 = "Constraint - extra_armor_hand_r (PPPP)"!(
    ComponentPose=@n29
    Alpha=1
    ConstraintWeights_0=0.25
    ConstraintWeights_1:float=0.25
    ConstraintWeights_2:float=0.25
    ConstraintWeights_3:float=0.25
   ) -> (Pose) {
    alphaBoolBlend {blendOption=Linear}
    boneToModify {boneName=extra_armor_hand_r}
    constraintSetup [
     {targetBone {boneName=index_metacarpal_r}}
     {targetBone {boneName=middle_metacarpal_r}}
     {targetBone {boneName=ring_metacarpal_r}}
     {targetBone {boneName=pinky_metacarpal_r}}
    ]
    constraintWeights [1,1,1,1]
   }
   n31 = Control Rig!(Source=Component To Local(ComponentPose=@n30), Alpha, crouch:double=Get crouch() -> (crouch:double)) -> (Pose) {
    alphaBoolBlend {blendOption=Linear}
    controlRigAssetReference {blueprintRigClass=&/F_MED_ChicleVeil_CtrlRig.F_MED_ChicleVeil_CtrlRig_C}
   }
   n32 = MakeVector(Y=0, Z=-10)
   n33 = "Transform (Modify) Bone - Bone: thigh_l"!(ComponentPose=Local To Component(LocalPose=@n31), bAlphaBoolEnabled:bool=True, Translation=@n32, Rotation="0,0,0", Scale="1,1,1") -> (Pose) {
    alphaInputType=Bool
    translationMode=BMM_Additive
    alphaBoolBlend {blendInTime=0.2, blendOption=Linear, blendOutTime=0.2}
    boneToModify {boneName=thigh_l}
    translation {z=-10}
   }
   n34 = RigidBody!(
    ComponentPose=@n33
    Alpha
    OverrideWorldGravity="0,0,-980"
    ExternalForce
    ComponentLinearAccScale:Vector="0.3,0.3,0.3"
    ComponentLinearVelScale:Vector="0,0,0"
   ) -> (Pose) {
    bOverrideWorldGravity=true
    overridePhysicsAsset=&/F_MED_ChicleVeil_Physics_suspenders
    simulationSpace=BaseBoneSpace
    alphaBoolBlend {blendOption=Linear}
    baseBoneRef {boneName=pelvis}
    simSpaceSettings {worldAlpha=0.05}
   }
   n35 = "Transform (Modify) Bone - Bone: thigh_l"(
    ComponentPose=@n34
    bAlphaBoolEnabled=True
    Translation=NegateVector!(A:Vector=@n32) -> (ReturnValue:Vector)
    Rotation="0,0,0"
    Scale="1,1,1"
   ) {
    alphaInputType=Bool
    translationMode=BMM_Additive
    alphaBoolBlend {blendInTime=0.2, blendOption=Linear, blendOutTime=0.2}
    boneToModify {boneName=thigh_l}
    translation {z=10}
   }
   Save cached pose 'pre_sim'(Pose=Component To Local(ComponentPose=@n35)) {cacheName=pre_sim}
  ]}
 EventGraph {nodes [
   n1 = Event BlueprintUpdateAnimation()
   > n2 = Sequence!(execute) -> (then_0, then_1, then_2:Exec, then_3:Exec, then_4:Exec)
   n3 = Get bIsCrouching()
   n4 = Branch(execute=@n2.then_1, Condition=@n3)
   > Set suspenders_component_linear(suspenders_component_linear:Vector="0.6,0.6,0.6") -> (Output_Get:Vector)
   Set suspenders_component_linear(execute=@n4.else, suspenders_component_linear="0, 0, 0")
   n5 = Branch(execute=@n2.then_2, Condition=@n3)
   > Set crouch(crouch:double=1) -> (Output_Get:double)
   Set crouch(execute=@n5.else, crouch=0)
   Set WPO Globals Function(
    execute=@n2.then_4
    Material=Get Mat Instance Dynamic()
    Overlay Material=Get Overlay Mat Instance Dynamic()
    Skeletal Mesh=Get SkelMesh()
    "Local Noise Transform  Location"="Get Local Noise Transform  Location"() -> ("Local Noise Transform  Location":Vector)
    Local Noise Offset=Get Local Noise Offset()
   )
   > Set WPO Regional Function!(
    execute
    Material:Material Instance Dynamic Object Reference=Get Mat Instance Dynamic()
    Overlay Material:Material Instance Dynamic Object Reference=Get Overlay Mat Instance Dynamic()
    Skeletal Mesh:Skeletal Mesh Component Object Reference=Get SkelMesh()
    Velocity Transform:Name=foot_r
    Velocity Interp Speed:double=Get Velocity Interp Speed()
    Collision Transform:Name=calf_r
    Collision Transform Axis:EAxis Enum
    Invert Collision Transform Axis:bool=false
    Delta Time X:double=@n1.DeltaTimeX
    Previous Velocity Transform Location:"Vector (by ref)"=Get R1 Velocity Transform Location()
    Previous Velocity Transform Velocity:"Vector (by ref)"=Get R1 Velocity Transform Velocity()
    Region:int=1
   ) -> (then)
   > Set WPO Regional Function(
    Material=Get Mat Instance Dynamic()
    Overlay Material=Get Overlay Mat Instance Dynamic()
    Skeletal Mesh=Get SkelMesh()
    Velocity Transform=foot_l
    Velocity Interp Speed=Get Velocity Interp Speed()
    Collision Transform=calf_l
    Invert Collision Transform Axis=true
    Delta Time X=@n1.DeltaTimeX
    Previous Velocity Transform Location=Get R2 Velocity Transform Location()
    Previous Velocity Transform Velocity=Get R2 Velocity Transform Velocity()
    Region=2
   )
   > n6 = Sequence!(execute) -> (then_0, then_1)
   > n7 = Branch(Condition=Get bIsSkydiveFloating())
   > n8 = Set WPO Pho Wind Intensity(WPO Pho Wind Intensity=1)
   SetScalarParameterValue(
    execute=[@n8.then, @n9.then]
    self=[Get Mat Instance Dynamic(), Get Overlay Mat Instance Dynamic()]
    ParameterName=ClothSim_WindStrength
    Value=Get WPO Pho Wind Intensity()
   )
   n9 = Set WPO Pho Wind Intensity(execute=@n7.else, WPO Pho Wind Intensity=0)
   n10 = Get bIsSkyDiving()
   n11 = Branch(execute=@n6.then_1, Condition=AND Boolean(A=Get bIsJumping(), B=NOT Boolean(A=@n10)))
   > n12 = Set WPO Pho Wind Intensity(WPO Pho Wind Intensity=5)
   n13 = Get Mat Instance Dynamic()
   n14 = Get Overlay Mat Instance Dynamic()
   n15 = Get WPO Pho Wind Intensity()
   SetScalarParameterValue(execute=[@n16.then, @n12.then], self=[@n13, @n14], ParameterName=ClothSim_R1_DragStrength, Value=@n15)
   > SetScalarParameterValue(self=[@n13, @n14], ParameterName=ClothSim_R2_DragStrength, Value=@n15)
   n16 = Set WPO Pho Wind Intensity(execute=@n11.else, WPO Pho Wind Intensity=6)
   n17 = Gravity Override State Function DEPRECATED(
    execute=@n2.then_0
    CharacterPartAnimInstance=Self-Reference()
    GameTimeInSeconds=GetGameTimeInSeconds()
    GravityOverrideParams=Get Gravity Override Params()
    RBAN Params=Get RBAN Params()
   )
   n18 = Get bIsSkydiveDiveMode()
   n19 = Get bIsFalling()
   n20 = Get bIsJumping()
   > n21 = Branch(Condition=AND Boolean(A=OR Boolean(A=@n20, B=@n19), B=NOT Boolean(A=@n18)))
   > Set Gravity(Gravity:Vector=vector + vector(A=@n17, B="0,0,600")) -> (Output_Get:Vector)
   Set Gravity(execute=@n21.else, Gravity=@n17)
   Set is_sky(
    execute=@n2.then_3
    is_sky:bool="Not Equal (Float)"(A=Get!() -> (PawnVelocity_X:double, PawnVelocity_Y:double, PawnVelocity_Z:double).PawnVelocity_Z)
   ) -> (Output_Get:bool)
   n22 = GetAlphaFromReferenceTransformDelta(
    SkeletalMeshComponent=GetOwnerSkeletalMeshComponent()
    BoneName=thigh_l
    BaseName=calf_l
    DeltaComponentType=QuaternionTwist
    InputRange=(X=30,Y=35)
    TwistAxis="0,0,1"
   )
   > Set is_move(is_move:bool=OR Boolean(A="Equal (Float)"(A=@n22, B=1), B=Get is_sky() -> (is_sky:bool))) -> (Output_Get:bool)
   Event OnCharacterCustomizationLoaded()
   > n23 = Set SkelMesh(SkelMesh=GetPartSkeletalMeshComponent())
   > n24 = Get Dynamic Material Instance(SkeletalMeshComponent=@n23)
   > Set Mat Instance Dynamic(Mat Instance Dynamic=@n24)
   > n25 = GetOrCreateDynamicOverlayMaterial(MeshComponent=@n23, MaterialName=None)
   > Set Overlay Mat Instance Dynamic(Overlay Mat Instance Dynamic=@n25)
  ]}
]
variables [
 crouch {type=number}
 deform {
  type=FN_DeformRig_Value_Struct
  default {
   bicep {location=<-3 4 0>, rotation=<rot 0 0 0>, scale=<1 1 1>}
   elbow_in {location=<0 9 0>, rotation=<rot 0 0 0>, scale=<0.5 1 1>}
   elbow_out {location=<0 -2 0>, rotation=<rot 0 0 0>, scale=<0.75 1 1>}
   glute {location=<-7 0 0>, rotation=<rot 0 0 0>, scale=<1 1.5 1>}
   groin {location=<-3 12 0>, rotation=<rot 0 0 0>, scale=<0.5 1 1>}
   knee_in {location=<0 -12 0>, rotation=<rot 0 0 0>, scale=<0.5 1 1>}
   knee_out {location=<0 3 0>, rotation=<rot 0 0 0>, scale=<1.5 1 1.5>}
   lat {location=<0 0 10>, rotation=<rot 0 0 0>, scale=<1 1 1>}
   pec {location=<0 2 3>, rotation=<rot 0 0 0>, scale=<1 1 1>}
   upperarm_bk {location=<0 -4 0>, rotation=<rot 0 0 0>, scale=<1 1 1>}
   upperarm_fwd_bk {location=<0 -6 0>, rotation=<rot 0 0 0>, scale=<1 1 1>}
   upperarm_fwd_fr {location=<0 10 0>, rotation=<rot 0 0 0>, scale=<0.5 1 1>}
   upperarm_up {location=<0 0 8>, rotation=<rot 0 0 0>, scale=<0.7 1 1>}
   wrist_dn {location=<0 -6 0>, rotation=<rot 0 0 0>, scale=<0.7 1 1>}
   wrist_up {location=<0 9 0>, rotation=<rot 0 0 0>, scale=<0.7 1 1>}
  }
 }
 Gravity {type=Vector}
 Gravity Override Params {
  type=GravityOverrideParamsStruct
  default {
   skydive_FloatingDown {gravityOverride=<0 -700 500>, windAmplitude=<200 -500 200>, windFrequency=<1.5 7 4>, windOffset=<0 -600 200>}
   skydive_GoingUp {gravityOverride=<0 -150 400>, windAmplitude=<100 -250 100>, windFrequency=<1.5 7 4>}
   skydive_Parachuting {gravityOverride=<0 -400 -980>, windAmplitude=<100 -250 200>, windFrequency=<0.75 3.5 4>}
  }
 }
 is_move {type=boolean}
 is_sky {type=boolean}
 Local Noise Offset {type=Vector}
 "Local Noise Transform  Location" {type=Vector}
 Mat Instance Dynamic {type=MaterialInstanceDynamic}
 Overlay Mat Instance Dynamic {type=MaterialInstanceDynamic}
 R1 Velocity Transform Location {type=Vector}
 R1 Velocity Transform Velocity {type=Vector}
 R2 Velocity Transform Location {type=Vector}
 R2 Velocity Transform Velocity {type=Vector}
 RBAN Params {
  type=RBANParamsStruct
  default {
   dBNO {simSpaceSettings {
     maxAngularAcceleration=1000
     maxAngularVelocity=3
     maxLinearAcceleration=300
     maxLinearVelocity=700
     velocityScaleZ=0
     externalLinearDragV=<0 0.5 0>
    }}
   dBNO_Carried {simSpaceSettings {
     maxAngularAcceleration=1000
     maxAngularVelocity=3
     maxLinearAcceleration=300
     maxLinearVelocity=700
     velocityScaleZ=0
     externalLinearDragV=<0 0.5 0>
    }}
   emoteOrMelee {simSpaceSettings {
     maxAngularAcceleration=1000
     maxAngularVelocity=3
     maxLinearAcceleration=300
     maxLinearVelocity=700
     velocityScaleZ=0
     externalLinearDragV=<0 0.5 0>
    }}
   front_End {simSpaceSettings {
     maxAngularAcceleration=1000
     maxAngularVelocity=3
     maxLinearAcceleration=300
     maxLinearVelocity=700
     velocityScaleZ=0
     externalLinearDragV=<0 0.5 0>
    }}
   onGround_Crouch_Moving {simSpaceSettings {
     maxAngularAcceleration=1000
     maxAngularVelocity=3
     maxLinearAcceleration=300
     maxLinearVelocity=700
     velocityScaleZ=0
     externalLinearDragV=<0 0.5 0>
    }}
   onGround_Falling {simSpaceSettings {
     maxAngularAcceleration=1000
     maxAngularVelocity=3
     maxLinearAcceleration=300
     maxLinearVelocity=700
     velocityScaleZ=0
     externalLinearDragV=<0 0.5 0>
    }}
   onGround_Moving {simSpaceSettings {
     maxAngularAcceleration=1000
     maxAngularVelocity=3
     maxLinearAcceleration=300
     maxLinearVelocity=700
     velocityScaleZ=0
     externalLinearDragV=<0 0.5 0>
    }}
   onGround_Standing {simSpaceSettings {
     maxAngularAcceleration=1000
     maxAngularVelocity=3
     maxLinearAcceleration=300
     maxLinearVelocity=700
     velocityScaleZ=0
     externalLinearDragV=<0 0.5 0>
    }}
   rail_Grinding {simSpaceSettings {
     maxAngularAcceleration=1000
     maxAngularVelocity=3
     maxLinearAcceleration=300
     maxLinearVelocity=700
     velocityScaleZ=0
     worldAlpha=0.1
     externalLinearDragV=<0 0.5 0>
    }}
   riding_Moving {simSpaceSettings {
     maxAngularAcceleration=1000
     maxAngularVelocity=3
     maxLinearAcceleration=300
     maxLinearVelocity=700
     velocityScaleZ=0
     externalLinearDragV=<0 0.5 0>
    }}
   riding_Moving_Targeting {simSpaceSettings {
     maxAngularAcceleration=1000
     maxAngularVelocity=3
     maxLinearAcceleration=300
     maxLinearVelocity=700
     velocityScaleZ=0
     externalLinearDragV=<0 0.5 0>
    }}
   riding_Standing {simSpaceSettings {
     maxAngularAcceleration=1000
     maxAngularVelocity=3
     maxLinearAcceleration=300
     maxLinearVelocity=700
     velocityScaleZ=0
     externalLinearDragV=<0 0.5 0>
    }}
   skydive_DivingDown {simSpaceSettings {
     maxAngularAcceleration=1000
     maxAngularVelocity=3
     maxLinearAcceleration=300
     maxLinearVelocity=700
     velocityScaleZ=0
     externalLinearDragV=<0 0.5 0>
    }}
   skydive_FloatingDown {
    componentLinearAccScale=<0 0.5 0>
    componentLinearVelScale=<0 0.5 0>
    simSpaceSettings {
     maxAngularAcceleration=1000
     maxAngularVelocity=3
     maxLinearAcceleration=300
     maxLinearVelocity=700
     velocityScaleZ=0
     externalLinearDragV=<0 0.5 0>
    }
   }
   skydive_GoingUp {simSpaceSettings {
     maxAngularAcceleration=1000
     maxAngularVelocity=3
     maxLinearAcceleration=300
     maxLinearVelocity=700
     velocityScaleZ=0
     externalLinearDragV=<0 0.5 0>
    }}
   skydive_Parachuting {
    componentLinearAccScale=<0 0.1 0.3>
    componentLinearVelScale=<0 0.1 0.3>
    simSpaceSettings {
     maxAngularAcceleration=1000
     maxAngularVelocity=3
     maxLinearAcceleration=300
     maxLinearVelocity=700
     velocityScaleZ=0
     externalLinearDragV=<0 0.5 0>
    }
   }
   skydive_Surfing {simSpaceSettings {
     maxAngularAcceleration=1000
     maxAngularVelocity=3
     maxLinearAcceleration=300
     maxLinearVelocity=700
     velocityScaleZ=0
     externalLinearDragV=<0 0.5 0>
    }}
   swimming {simSpaceSettings {
     maxAngularAcceleration=1000
     maxAngularVelocity=3
     maxLinearAcceleration=300
     maxLinearVelocity=700
     velocityScaleZ=0
     externalLinearDragV=<0 0.5 0>
    }}
   tactical_Sprint {simSpaceSettings {
     maxAngularAcceleration=1000
     maxAngularVelocity=3
     maxLinearAcceleration=300
     maxLinearVelocity=700
     velocityScaleZ=0
     externalLinearDragV=<0 0.5 0>
    }}
   ziplining {simSpaceSettings {
     maxAngularAcceleration=1000
     maxAngularVelocity=3
     maxLinearAcceleration=300
     maxLinearVelocity=700
     velocityScaleZ=0
     externalLinearDragV=<0 0.5 0>
    }}
  }
 }
 SkelMesh {type=SkeletalMeshComponent}
 suspenders_component_linear {type=Vector, default=<0.3 0.3 0>}
 Velocity Interp Speed {type=number, default=1}
 WPO Pho Wind Intensity {type=number}
]
