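# Control Rig for the F_MED_ChicleVeil face accessory that applies pose-driven corrective offsets to its dyn_hair bones.
# Three sequenced branches use spherical pose readers on head with right upper arm, head alone, and both clavicles multiplied together.
# Smoothed reader outputs weight Modify Transforms nodes that move, rotate or scale back, side and right-front hair bones.
# The first branch maxes head×arm against head, so the arm reader adds nothing for 0–1 outputs.
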ControlRigBlueprint /BRCosmetics/Characters/Player/Female/Medium/Bodies/F_MED_ChicleVeil/Meshes/Parts/F_MED_ChicleVeil_FaceAcc_CtrlRig
graphs [
 RigVMFunctionLibrary {links [], nodes []}
 RigVMModel {nodes [
   RigUnit_BeginExecution = &RigUnit_BeginExecution()
   SequenceAggregate = RigVMAggregateNode(ExecuteContext=@RigUnit_BeginExecution.ExecutePin)
   n1 = &RigUnit_SphericalPoseReader(
    ActiveRegionScaleFactors=(PositiveWidth=1,NegativeWidth=0.5,PositiveHeight=0.75,NegativeHeight=0.75)
    ActiveRegionSize=0.01
    Debug=(bDrawDebug=true,bDraw2D=False,bDrawLocalAxes=False,DebugScale=25,DebugSegments=20,DebugThickness=0.25)
    DriverAxis=(X=0,Y=1,Z=0)
    DriverItem=(Type=Bone,Name="head")
    FalloffRegionScaleFactors=(PositiveWidth=1,NegativeWidth=0.5,PositiveHeight=1,NegativeHeight=1)
    FalloffSize=0.3
    OptionalParentItem=(Type=Bone)
    RotationOffset=(X=90,Y=0,Z=-30)
    ExecutePin=@SequenceAggregate.A
   )
   > n2 = &RigUnit_SphericalPoseReader(
    ActiveRegionScaleFactors=(PositiveWidth=1,NegativeWidth=1,PositiveHeight=1,NegativeHeight=1)
    ActiveRegionSize=0.01
    Debug=(bDrawDebug=true,bDraw2D=False,bDrawLocalAxes=False,DebugScale=25,DebugSegments=20,DebugThickness=0.25)
    DriverAxis=(X=-1,Y=0,Z=0)
    DriverItem=(Type=Bone,Name="upperarm_r")
    FalloffRegionScaleFactors=(PositiveWidth=1,NegativeWidth=1,PositiveHeight=1,NegativeHeight=1)
    FalloffSize=0.3
    OptionalParentItem=(Type=Bone)
    RotationOffset=(X=-105,Y=0,Z=25)
   )
   Multiply = &RigVMFunction_MathDoubleMul(A=@n1.OutputParam, B=@n2.OutputParam)
   Maximum = &RigVMFunction_MathDoubleMax(A=@Multiply.Result, B=@n1.OutputParam)
   AlphaInterp = &RigVMFunction_AlphaInterp(InterpSpeedDecreasing=6, InterpSpeedIncreasing=6, Scale=1, bClampResult=False, Value=@Maximum.Result)
   > &RigUnit_ModifyTransforms(
    ItemToModify=((Item=(Type=Bone,Name="dyn_hair_02_front_r"),Transform=(Rotation=(X=0,Y=0,Z=0,W=1),Translation=(X=0,Y=0,Z=0),Scale3D=(X=1,Y=1,Z=0.5))))
    Weight=@AlphaInterp.Result
    ItemToModify.0.Transform.Rotation=&RigVMFunction_MathQuaternionFromEuler(Euler=(X=0,Y=0,Z=0)).Result
   )
   > &RigUnit_ModifyTransforms(
    ItemToModify=((Item=(Type=Bone,Name="dyn_hair_02_side_r"),Transform=(Rotation=(X=0,Y=0,Z=0,W=1),Translation=(X=-5,Y=0,Z=0),Scale3D=(X=1,Y=1,Z=1))))
    Weight=@AlphaInterp.Result
    ItemToModify.0.Transform.Rotation=&RigVMFunction_MathQuaternionFromEuler(Euler=(X=0,Y=0,Z=0)).Result
   )
   > &RigUnit_ModifyTransforms(
    ItemToModify=((Item=(Type=Bone,Name="dyn_hair_02_back_c"),Transform=(Rotation=(X=0,Y=0,Z=0,W=1),Translation=(X=-3,Y=0,Z=0),Scale3D=(X=1,Y=1,Z=1))))
    Weight=@AlphaInterp.Result
    ItemToModify.0.Transform.Rotation=&RigVMFunction_MathQuaternionFromEuler(Euler=(X=0,Y=0,Z=0)).Result
   )
   > &RigUnit_ModifyTransforms(
    ItemToModify=((Item=(Type=Bone,Name="dyn_hair_02_back_c"),Transform=(Rotation=(X=0,Y=0,Z=0,W=1),Translation=(X=-1.5,Y=0,Z=0),Scale3D=(X=1,Y=1,Z=1))))
    Weight=@AlphaInterp.Result
    ItemToModify.0.Transform.Rotation=&RigVMFunction_MathQuaternionFromEuler(Euler=(X=0,Y=0,Z=0)).Result
   )
   n3 = &RigUnit_SphericalPoseReader(
    ActiveRegionScaleFactors=(PositiveWidth=1,NegativeWidth=0.5,PositiveHeight=0.75,NegativeHeight=0.75)
    ActiveRegionSize=0.05
    Debug=(bDrawDebug=true,bDraw2D=False,bDrawLocalAxes=False,DebugScale=25,DebugSegments=20,DebugThickness=0.25)
    DriverAxis=(X=0,Y=1,Z=0)
    DriverItem=(Type=Bone,Name="head")
    FalloffRegionScaleFactors=(PositiveWidth=1,NegativeWidth=0.5,PositiveHeight=1,NegativeHeight=1)
    FalloffSize=0.2
    OptionalParentItem=(Type=Bone)
    RotationOffset=(X=105,Y=0,Z=-50)
    ExecutePin=@SequenceAggregate.B
   )
   n4 = &RigVMFunction_AlphaInterp(InterpSpeedDecreasing=4, InterpSpeedIncreasing=4, Scale=1, bClampResult=False, Value=@n3.OutputParam)
   > &RigUnit_ModifyTransforms(
    ItemToModify=((Item=(Type=Bone,Name="dyn_hair_01_back_c"),Transform=(Rotation=(X=0,Y=0,Z=0,W=1),Translation=(X=0,Y=0,Z=0),Scale3D=(X=0.7,Y=1,Z=1))))
    Weight=@n4.Result
    ItemToModify.0.Transform.Rotation=&RigVMFunction_MathQuaternionFromEuler(Euler=(X=0,Y=-30,Z=0)).Result
   )
   > &RigUnit_ModifyTransforms(
    ItemToModify=((Item=(Type=Bone,Name="dyn_hair_01_side_r"),Transform=(Rotation=(X=0,Y=0,Z=0,W=1),Translation=(X=0,Y=0,Z=0),Scale3D=(X=0.7,Y=1,Z=1))))
    Weight=@n4.Result
    ItemToModify.0.Transform.Rotation=&RigVMFunction_MathQuaternionFromEuler(Euler=(X=0,Y=-15,Z=0)).Result
   )
   > &RigUnit_ModifyTransforms(
    ItemToModify=((Item=(Type=Bone,Name="dyn_hair_01_side_l"),Transform=(Rotation=(X=0,Y=0,Z=0,W=1),Translation=(X=0,Y=0,Z=0),Scale3D=(X=0.7,Y=1,Z=1))))
    Weight=@n4.Result
    ItemToModify.0.Transform.Rotation=&RigVMFunction_MathQuaternionFromEuler(Euler=(X=0,Y=0,Z=0)).Result
   )
   n5 = &RigUnit_SphericalPoseReader(
    ActiveRegionScaleFactors=(PositiveWidth=1,NegativeWidth=1,PositiveHeight=1,NegativeHeight=1)
    ActiveRegionSize=0.05
    Debug=(bDrawDebug=true,bDraw2D=False,bDrawLocalAxes=False,DebugScale=25,DebugSegments=20,DebugThickness=0.25)
    DriverAxis=(X=1,Y=0,Z=0)
    DriverItem=(Type=Bone,Name="clavicle_l")
    FalloffRegionScaleFactors=(PositiveWidth=1,NegativeWidth=1,PositiveHeight=1,NegativeHeight=1)
    FalloffSize=0.2
    OptionalParentItem=(Type=Bone)
    RotationOffset=(X=50,Y=0,Z=-60)
    ExecutePin=@SequenceAggregate.C
   )
   > n6 = &RigUnit_SphericalPoseReader(
    ActiveRegionScaleFactors=(PositiveWidth=1,NegativeWidth=1,PositiveHeight=1,NegativeHeight=1)
    ActiveRegionSize=0.05
    Debug=(bDrawDebug=true,bDraw2D=False,bDrawLocalAxes=false,DebugScale=25,DebugSegments=20,DebugThickness=0.25)
    DriverAxis=(X=-1,Y=0,Z=0)
    DriverItem=(Type=Bone,Name="clavicle_r")
    FalloffRegionScaleFactors=(PositiveWidth=1,NegativeWidth=1,PositiveHeight=1,NegativeHeight=1)
    FalloffSize=0.2
    OptionalParentItem=(Type=Bone)
    RotationOffset=(X=-5,Y=135,Z=0)
   )
   n7 = &RigVMFunction_AlphaInterp(
    InterpSpeedDecreasing=4
    InterpSpeedIncreasing=4
    Scale=1
    bClampResult=False
    Value=&RigVMFunction_MathDoubleMul(B=@n5.OutputParam, A=@n6.OutputParam).Result
   )
   > &RigUnit_ModifyTransforms(
    ItemToModify=((Item=(Type=Bone,Name="dyn_hair_01_back_c"),Transform=(Rotation=(X=0,Y=0,Z=0,W=1),Translation=(X=-3,Y=0,Z=3),Scale3D=(X=1,Y=1,Z=1))))
    Weight=@n7.Result
    ItemToModify.0.Transform.Rotation=&RigVMFunction_MathQuaternionFromEuler(Euler=(X=0,Y=30,Z=0)).Result
   )
  ]}
 SequenceAggregate_1_ContainedGraph {nodes [
   n1 = RigVMFunctionEntryNode()
   SequenceAggregate = &RigVMFunction_Sequence(ExecuteContext=@n1.ExecuteContext)
   SequenceAggregate_1 = &RigVMFunction_Sequence(ExecuteContext=@SequenceAggregate.B)
   SequenceAggregate_2 = &RigVMFunction_Sequence(ExecuteContext=@SequenceAggregate_1.B)
   RigVMFunctionReturnNode(A=@SequenceAggregate.A, B=@SequenceAggregate_1.A, C=@SequenceAggregate_2.A, D=@SequenceAggregate_2.B)
  ]}
]
hierarchy [
 root {loc=<0 0 0>, rot=<0 0 0 1>}
 pelvis {parent=root, loc=<0 -0.5301 98.0571>, rot=<0.0127 0.707 -0.0127 -0.707>}
 spine_01 {parent=pelvis, loc=<6.1208 -0.9328 0>, rot=<0 0 0.034 -0.9994>}
 spine_02 {parent=spine_01, loc=<6.5996 -0.2095 0>, rot=<0 0 0.0238 -0.9997>}
 spine_03 {parent=spine_02, loc=<6.4914 -0.1237 0>, rot=<0 0 0.0149 -0.9999>}
 spine_04 {parent=spine_03, loc=<6.9598 0.0794 0>, rot=<0 0 -0.0132 -0.9999>}
 spine_05 {parent=spine_04, loc=<6.9987 0 0>, rot=<0 0 -0.0015 -1>}
 neck_01 {parent=spine_05, loc=<11.3138 0.1529 0>, rot=<0 0 -0.2003 -0.9797>}
 neck_02 {parent=neck_01, loc=<4.5627 0 0>, rot=<0 0 0.0504 -0.9987>}
 head {parent=neck_02, loc=<5.3914 0 0>, rot=<0 0 0.115 -0.9934>}
 dyn_hair_01_front_base_l {parent=head, loc=<7.2538 3.5871 -6.8058>, rot=<0.0433 -0.9938 -0.0468 -0.0914>}
 dyn_hair_01_front_l {parent=dyn_hair_01_front_base_l, loc=<0 0 0>, rot=<0 0 0 1>}
 dyn_hair_02_front_l {parent=dyn_hair_01_front_l, loc=<6.0277 0 0>, rot=<0 0 0 1>}
 dyn_hair_01_side_base_l {parent=head, loc=<7.4751 -3.0502 -5.4661>, rot=<0.0773 -0.9606 0.2412 -0.1144>}
 dyn_hair_01_side_l {parent=dyn_hair_01_side_base_l, loc=<0 0 0>, rot=<0 0 0 1>}
 dyn_hair_02_side_l {parent=dyn_hair_01_side_l, loc=<6.1316 0 0>, rot=<0 0 0 1>}
 dyn_hair_01_back_base_c {parent=head, loc=<7.3413 -6.9996 0>, rot=<0.1067 -0.699 0.699 -0.1067>}
 dyn_hair_01_back_c {parent=dyn_hair_01_back_base_c, loc=<0 0 0>, rot=<0 0 0 1>}
 dyn_hair_02_back_c {parent=dyn_hair_01_back_c, loc=<6.1741 0 0>, rot=<0 0 0 1>}
 dyn_hair_01_side_base_r {parent=head, loc=<7.4752 -3.0502 5.4661>, rot=<0.1144 -0.2412 0.9606 -0.0773>}
 dyn_hair_01_side_r {parent=dyn_hair_01_side_base_r, loc=<0 0 0>, rot=<0 0 0 1>}
 dyn_hair_02_side_r {parent=dyn_hair_01_side_r, loc=<6.1317 0 0>, rot=<0 0 0 1>}
 dyn_hair_01_front_base_r {parent=head, loc=<7.2534 3.5871 6.8058>, rot=<0.0914 0.0468 0.9938 -0.0433>}
 dyn_hair_01_front_r {parent=dyn_hair_01_front_base_r, loc=<0 0 0>, rot=<0 0 0 1>}
 dyn_hair_02_front_r {parent=dyn_hair_01_front_r, loc=<6.0277 0 0>, rot=<0 0 0 1>}
 dyn_hair_bang_base_r {parent=head, loc=<10.1112 6.4469 5.4666>, rot=<0.1869 0.2168 0.9574 0.0384>}
 dyn_hair_bang_01_r {parent=dyn_hair_bang_base_r, loc=<0 0 0>, rot=<0 0 0 1>}
 dyn_hair_bang_02_r {parent=dyn_hair_bang_01_r, loc=<4.1286 0 0>, rot=<0 -0.1827 0 -0.9832>}
 dyn_hair_bang_base_c {parent=head, loc=<11.7799 7.7929 2.0051>, rot=<-0.2827 -0.6491 -0.6467 -0.2838>}
 dyn_hair_bang_01_c {parent=dyn_hair_bang_base_c, loc=<0 0 0>, rot=<0 0 0 1>}
 dyn_hair_bang_02_c {parent=dyn_hair_bang_01_c, loc=<3.066 0 0>, rot=<0 -0.3961 0 -0.9182>}
 dyn_hair_bang_base_l {parent=head, loc=<11.1626 7.5604 -2.3768>, rot=<-0.0851 -0.9102 -0.2606 -0.3105>}
 dyn_hair_bang_01_l {parent=dyn_hair_bang_base_l, loc=<0 0 0>, rot=<0 0 0 1>}
 dyn_hair_bang_02_l {parent=dyn_hair_bang_01_l, loc=<3.1797 0 0>, rot=<0 -0.2612 0 -0.9653>}
 dyn_topknot_base_c {parent=head, loc=<16.5967 -6.4588 -0.1026>, rot=<-0.6214 0.0042 -0.1128 -0.7753>}
 dyn_topknot_c {parent=dyn_topknot_base_c, loc=<0 0 0>, rot=<0 0 0 1>}
 dyn_topknot_base_l {parent=head, loc=<14.2701 -5.4316 -3.3458>, rot=<-0.576 -0.3574 0.3814 -0.6286>}
 dyn_topknot_l {parent=dyn_topknot_base_l, loc=<0 0 0>, rot=<0 0 0 1>}
 dyn_topknot_base_r {parent=head, loc=<13.5367 -6.787 3.3717>, rot=<0.6325 -0.4052 0.5176 0.4097>}
 dyn_topknot_r {parent=dyn_topknot_base_r, loc=<0 0 0>, rot=<0 0 0 1>}
 clavicle_l {parent=spine_05, loc=<5.4831 -0.1053 -3.2946>, rot=<0.0465 -0.7289 -0.009 -0.6829>}
 upperarm_l {parent=clavicle_l, loc=<9.8658 -1.2136 -0.4105>, rot=<0.0164 -0.3245 -0.0269 -0.9454>}
 clavicle_r {parent=spine_05, loc=<5.4831 -0.1053 3.2946>, rot=<0.7289 0.0465 -0.6829 0.009>}
 upperarm_r {parent=clavicle_r, loc=<-9.8658 1.2136 0.4105>, rot=<0.0164 -0.3245 -0.0269 -0.9454>}
]
