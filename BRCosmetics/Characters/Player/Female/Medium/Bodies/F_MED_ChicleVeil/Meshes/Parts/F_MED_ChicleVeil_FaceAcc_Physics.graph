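# Physics asset for the ChicleVeil female medium face-accessory mesh, setting up secondary motion for its dynamic hair bones.
# Back, front, side, bang and topknot chains of capsule bodies each start at a small base body constrained to the head with free swing and twist.
# Segments further down each chain have angle limits, with tight twist and swing varying by chain.
# Capsules on head, neck, upper spine and upper arms approximate the body around the hair.
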PhysicsAsset /BRCosmetics/Characters/Player/Female/Medium/Bodies/F_MED_ChicleVeil/Meshes/Parts/F_MED_ChicleVeil_FaceAcc_Physics
bodies {
 dyn_hair_01_back_base_c [{shapeType=Capsule, length=0.5, radius=0.5, shapeName=dyn_hair_01_back_base_c_capsule}]
 dyn_hair_01_back_c [{shapeType=Capsule, length=2, radius=1, shapeName=dyn_hair_01_back_c_capsule, center=<3 0 0>, rotation=<rot -90 -180 -180>}]
 dyn_hair_01_front_base_l [{shapeType=Capsule, length=0.5, radius=0.5, shapeName=dyn_hair_01_front_base_l_capsule}]
 dyn_hair_01_front_base_r [{shapeType=Capsule, length=0.5, radius=0.5, shapeName=dyn_hair_01_front_base_r_capsule}]
 dyn_hair_01_front_l [{shapeType=Capsule, length=2, radius=1, shapeName=dyn_hair_01_front_l_capsule, center=<3 0 0>, rotation=<rot -90 65.382 -65.3811>}]
 dyn_hair_01_front_r [{shapeType=Capsule, length=2, radius=1, shapeName=dyn_hair_01_front_r_capsule, center=<3 0 0.4517>, rotation=<rot -90 -90 89.9991>}]
 dyn_hair_01_side_base_l [{shapeType=Capsule, length=0.5, radius=0.5, shapeName=dyn_hair_01_side_base_l_capsule}]
 dyn_hair_01_side_base_r [{shapeType=Capsule, length=0.5, radius=0.5, shapeName=dyn_hair_01_side_base_r_capsule}]
 dyn_hair_01_side_l [{shapeType=Capsule, length=4, radius=2, shapeName=dyn_hair_02_side_r_capsule, center=<4 0 0>, rotation=<rot -90 -160.5291 -19.4716>}]
 dyn_hair_01_side_r [{shapeType=Capsule, length=2, radius=1, shapeName=dyn_hair_01_side_r_capsule, center=<3 0 0>, rotation=<rot -90 -159.2198 159.219>}]
 dyn_hair_02_back_c [{shapeType=Capsule, length=4.5, radius=2, shapeName=dyn_hair_02_back_c_capsule, center=<4 0 0>, rotation=<rot -90 -180 -180>}]
 dyn_hair_02_front_l [{shapeType=Capsule, length=4.5, radius=1.5, shapeName=dyn_hair_02_front_l_capsule, center=<4 0 0>, rotation=<rot -90 65.382 -65.3811>}]
 dyn_hair_02_front_r [{shapeType=Capsule, length=4.5, radius=1.5, shapeName=dyn_hair_02_front_r_capsule, center=<4 0 0>, rotation=<rot -90 179.9981 179.9997>}]
 dyn_hair_02_side_l [{shapeType=Capsule, length=4, radius=2.5, shapeName=dyn_hair_02_side_r_capsule, center=<4 0 0>, rotation=<rot -90 -160.5291 -19.4716>}]
 dyn_hair_02_side_r [{shapeType=Capsule, length=6, radius=2, shapeName=dyn_hair_02_side_r_capsule, center=<4 0 0>, rotation=<rot -90 -159.2198 159.219>}]
 dyn_hair_bang_01_c [{shapeType=Capsule, length=2, radius=1, shapeName=dyn_hair_bang_01_c_capsule, center=<3 0 0>, rotation=<rot -90 0.1511 -0.1511>}]
 dyn_hair_bang_01_l [{shapeType=Capsule, length=2, radius=1, shapeName=dyn_hair_bang_01_l_capsule, center=<3 0 0>, rotation=<rot -90 44.6563 -44.6561>}]
 dyn_hair_bang_01_r [{shapeType=Capsule, length=2, radius=1, shapeName=dyn_hair_bang_01_r_capsule, center=<3 0 0>, rotation=<rot -90 -38.4646 38.4641>}]
 dyn_hair_bang_02_c [{shapeType=Capsule, length=2, radius=1, shapeName=dyn_hair_bang_02_c_capsule, center=<3 0 0>, rotation=<rot -90 0.0037 -0.0037>}]
 dyn_hair_bang_02_l [{shapeType=Capsule, length=2, radius=1, shapeName=dyn_hair_bang_02_l_capsule, center=<3 0 0>, rotation=<rot -90 11.9313 -11.9304>}]
 dyn_hair_bang_02_r [{shapeType=Capsule, length=2, radius=1, shapeName=dyn_hair_bang_02_r_capsule, center=<3 0 0>, rotation=<rot -90 -2.4099 2.4088>}]
 dyn_hair_bang_base_c [{shapeType=Capsule, length=0.5, radius=0.5, shapeName=dyn_hair_bang_base_c_capsule}]
 dyn_hair_bang_base_l [{shapeType=Capsule, length=0.5, radius=0.5, shapeName=dyn_hair_bang_base_l_capsule}]
 dyn_hair_bang_base_r [{shapeType=Capsule, length=0.5, radius=0.5, shapeName=dyn_hair_bang_base_r_capsule}]
 dyn_topknot_base_c [{shapeType=Capsule, length=0.5, radius=0.5, shapeName=dyn_topknot_base_c_capsule}]
 dyn_topknot_base_l [{shapeType=Capsule, length=0.5, radius=0.5, shapeName=dyn_topknot_base_l_capsule}]
 dyn_topknot_base_r [{shapeType=Capsule, length=0.5, radius=0.5, shapeName=dyn_topknot_base_r_capsule}]
 dyn_topknot_c [{shapeType=Capsule, length=2, radius=1, shapeName=dyn_topknot_c_capsule, center=<3 0 0>, rotation=<rot -90 -8.6653 8.6652>}]
 dyn_topknot_l [{shapeType=Capsule, length=2, radius=1, shapeName=dyn_topknot_l_capsule, center=<3 0 0>, rotation=<rot -90 63.0666 -63.0677>}]
 dyn_topknot_r [{shapeType=Capsule, length=2, radius=1, shapeName=dyn_topknot_r_capsule, center=<3 0 0>, rotation=<rot -90 -84.288 84.289>}]
 head [
  {shapeType=Capsule, length=5.2516, radius=5.1, center=<-2.0746 -0.8172 0>, rotation=<rot -90 180 -160.0001>}
  {shapeType=Capsule, length=0, radius=6.5, shapeName=head_capsule, center=<6 3.3899 0>}
  {shapeType=Capsule, length=0, radius=7, shapeName=head_capsule, center=<2.3417 5.5711 0>}
 ]
 neck_01 [{
   shapeType=Capsule
   length=6.0949
   radius=5.2884
   shapeName=neck_01_capsule
   center=<1.5312 0.874 -0.0664>
   rotation=<rot 87.5172 -0.0032 -11.1613>
  }]
 spine_05 [
  {shapeType=Capsule, length=2, radius=7.3542, shapeName=spine_05_capsule, center=<11.2294 -2.0025 0>}
  {
   shapeType=Capsule
   length=4.3766
   radius=10.8207
   shapeName=spine_05_capsule
   center=<0.9477 1.8096 -0.0043>
   rotation=<rot -89.5488 -69.2055 73.8815>
  }
 ]
 upperarm_l [{shapeType=Capsule, length=21.75, radius=8, shapeName=upperarm_l_capsule, center=<12.3273 0 0>, rotation=<rot 90 0 -0.0002>}]
 upperarm_r [{shapeType=Capsule, length=21.75, radius=4, shapeName=upperarm_r_capsule, center=<-12.3273 0 -2.3872>, rotation=<rot 90 0 -0.0005>}]
}
constraints [
 {
  bone1Name=dyn_hair_01_back_base_c
  bone2Name=head
  swing1LimitDegrees=45
  swing1Motion=Free
  swing2LimitDegrees=45
  swing2Motion=Free
  twistLimitDegrees=45
  twistMotion=Free
 }
 {bone1Name=dyn_hair_01_back_c, bone2Name=dyn_hair_01_back_base_c, swing1LimitDegrees=5, swing2LimitDegrees=45, twistLimitDegrees=5}
 {bone1Name=dyn_hair_02_back_c, bone2Name=dyn_hair_01_back_c, swing1LimitDegrees=5, swing2LimitDegrees=45, twistLimitDegrees=5}
 {
  bone1Name=dyn_hair_01_front_base_l
  bone2Name=head
  swing1LimitDegrees=45
  swing1Motion=Free
  swing2LimitDegrees=45
  swing2Motion=Free
  twistLimitDegrees=45
  twistMotion=Free
 }
 {bone1Name=dyn_hair_01_front_l, bone2Name=dyn_hair_01_front_base_l, swing1LimitDegrees=30, swing2LimitDegrees=30, twistLimitDegrees=5}
 {bone1Name=dyn_hair_02_front_l, bone2Name=dyn_hair_01_front_l, swing1LimitDegrees=30, swing2LimitDegrees=30, twistLimitDegrees=5}
 {
  bone1Name=dyn_hair_01_front_base_r
  bone2Name=head
  swing1LimitDegrees=45
  swing1Motion=Free
  swing2LimitDegrees=45
  swing2Motion=Free
  twistLimitDegrees=45
  twistMotion=Free
 }
 {bone1Name=dyn_hair_01_front_r, bone2Name=dyn_hair_01_front_base_r, swing1LimitDegrees=45, swing2LimitDegrees=45, twistLimitDegrees=5}
 {bone1Name=dyn_hair_02_front_r, bone2Name=dyn_hair_01_front_r, swing1LimitDegrees=45, swing2LimitDegrees=45, twistLimitDegrees=5}
 {
  bone1Name=dyn_hair_01_side_base_l
  bone2Name=head
  swing1LimitDegrees=45
  swing1Motion=Free
  swing2LimitDegrees=45
  swing2Motion=Free
  twistLimitDegrees=45
  twistMotion=Free
 }
 {
  bone1Name=dyn_hair_01_side_base_r
  bone2Name=head
  swing1LimitDegrees=45
  swing1Motion=Free
  swing2LimitDegrees=45
  swing2Motion=Free
  twistLimitDegrees=45
  twistMotion=Free
 }
 {bone1Name=dyn_hair_01_side_r, bone2Name=dyn_hair_01_side_base_r, swing1LimitDegrees=45, swing2LimitDegrees=45, twistLimitDegrees=5}
 {bone1Name=dyn_hair_02_side_r, bone2Name=dyn_hair_01_side_r, swing1LimitDegrees=45, swing2LimitDegrees=45, twistLimitDegrees=5}
 {
  bone1Name=dyn_hair_bang_base_c
  bone2Name=head
  swing1LimitDegrees=45
  swing1Motion=Free
  swing2LimitDegrees=45
  swing2Motion=Free
  twistLimitDegrees=45
  twistMotion=Free
 }
 {bone1Name=dyn_hair_bang_01_c, bone2Name=dyn_hair_bang_base_c, swing1LimitDegrees=15, swing2LimitDegrees=45, twistLimitDegrees=5}
 {bone1Name=dyn_hair_bang_02_c, bone2Name=dyn_hair_bang_01_c, swing1LimitDegrees=15, swing2LimitDegrees=45, twistLimitDegrees=5}
 {
  bone1Name=dyn_hair_bang_base_l
  bone2Name=head
  swing1LimitDegrees=45
  swing1Motion=Free
  swing2LimitDegrees=45
  swing2Motion=Free
  twistLimitDegrees=45
  twistMotion=Free
 }
 {bone1Name=dyn_hair_bang_01_l, bone2Name=dyn_hair_bang_base_l, swing1LimitDegrees=15, swing2LimitDegrees=45, twistLimitDegrees=5}
 {bone1Name=dyn_hair_bang_02_l, bone2Name=dyn_hair_bang_01_l, swing1LimitDegrees=15, swing2LimitDegrees=45, twistLimitDegrees=5}
 {
  bone1Name=dyn_hair_bang_base_r
  bone2Name=head
  swing1LimitDegrees=45
  swing1Motion=Free
  swing2LimitDegrees=45
  swing2Motion=Free
  twistLimitDegrees=45
  twistMotion=Free
 }
 {bone1Name=dyn_hair_bang_01_r, bone2Name=dyn_hair_bang_base_r, swing1LimitDegrees=15, swing2LimitDegrees=45, twistLimitDegrees=5}
 {bone1Name=dyn_hair_bang_02_r, bone2Name=dyn_hair_bang_01_r, swing1LimitDegrees=15, swing2LimitDegrees=45, twistLimitDegrees=5}
 {
  bone1Name=dyn_topknot_base_c
  bone2Name=head
  swing1LimitDegrees=45
  swing1Motion=Free
  swing2LimitDegrees=45
  swing2Motion=Free
  twistLimitDegrees=45
  twistMotion=Free
 }
 {bone1Name=dyn_topknot_c, bone2Name=dyn_topknot_base_c, swing1LimitDegrees=45, swing2LimitDegrees=45, twistLimitDegrees=5}
 {
  bone1Name=dyn_topknot_base_l
  bone2Name=head
  swing1LimitDegrees=45
  swing1Motion=Free
  swing2LimitDegrees=45
  swing2Motion=Free
  twistLimitDegrees=45
  twistMotion=Free
 }
 {bone1Name=dyn_topknot_l, bone2Name=dyn_topknot_base_l, swing1LimitDegrees=45, swing2LimitDegrees=45, twistLimitDegrees=5}
 {
  bone1Name=dyn_topknot_base_r
  bone2Name=head
  swing1LimitDegrees=45
  swing1Motion=Free
  swing2LimitDegrees=45
  swing2Motion=Free
  twistLimitDegrees=45
  twistMotion=Free
 }
 {bone1Name=dyn_topknot_r, bone2Name=dyn_topknot_base_r, swing1LimitDegrees=45, swing2LimitDegrees=45, twistLimitDegrees=5}
 {
  bone1Name=upperarm_r
  bone2Name=spine_05
  swing1LimitDegrees=45
  swing1Motion=Free
  swing2LimitDegrees=45
  swing2Motion=Free
  twistLimitDegrees=45
  twistMotion=Free
 }
 {
  bone1Name=upperarm_l
  bone2Name=spine_05
  swing1LimitDegrees=45
  swing1Motion=Free
  swing2LimitDegrees=45
  swing2Motion=Free
  twistLimitDegrees=45
  twistMotion=Free
 }
 {bone1Name=dyn_hair_02_side_l, bone2Name=dyn_hair_01_side_l, swing1LimitDegrees=15, swing2LimitDegrees=30, twistLimitDegrees=5}
 {bone1Name=dyn_hair_01_side_l, bone2Name=dyn_hair_01_side_base_l, swing1LimitDegrees=15, swing2LimitDegrees=15, twistLimitDegrees=5}
]
