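# Animation Blueprint for the ChicleVeil female character's face accessory, adding secondary physics motion to the part.
# It blends the attached parent's pose with a captured body mesh's pose from the upper arms down, then runs head-space RigidBody simulation and a Control Rig.
# Each update stores a deprecated gravity-override result and toggles a material parameter to hide pins when a worn part's mesh name matches GalliumConcrete.
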AnimBlueprint /BRCosmetics/Characters/Player/Female/Medium/Bodies/F_MED_ChicleVeil/Meshes/Parts/F_MED_ChicleVeil_FaceAcc_AnimBP
target_skeleton=&/F_MED_ChicleVeil_FaceAcc_Skeleton
graphs [
 AnimGraph {nodes [
   n1 = Copy Pose From Mesh(SourceMeshComponent=Get BodyMesh() -> (BodyMesh:Skeletal Mesh Component Object Reference)) {bUseAttachedParent=true}
   n2 = Copy Pose From Mesh() {bUseAttachedParent=true}
   n3 = RigidBody!(
    ComponentPose=Local To Component(LocalPose=Layered blend per bone(BasePose=@n2, BlendPoses_0=@n1) {layerSetup [{branchFilters [{boneName=upperarm_l}, {boneName=upperarm_r}]}]})
    Alpha
    OverrideWorldGravity:Vector="0,0,-980"
    ExternalForce
   ) -> (Pose) {
    bOverrideWorldGravity=true
    overridePhysicsAsset=&/F_MED_ChicleVeil_FaceAcc_Physics
    simulationSpace=BaseBoneSpace
    alphaBoolBlend {blendOption=Linear}
    baseBoneRef {boneName=head}
    componentAppliedLinearAccClamp=<2500 2500 2500>
    componentLinearAccScale=<0.35 0.35 0.35>
    componentLinearVelScale=<0.35 0.35 0.35>
    simSpaceSettings {maxAngularAcceleration=10, maxAngularVelocity=10, maxLinearAcceleration=10, maxLinearVelocity=10, worldAlpha=1}
   }
   n4 = Control Rig(Source=Component To Local(ComponentPose=@n3)) {
    alphaBoolBlend {blendOption=Linear}
    controlRigAssetReference {blueprintRigClass=&/F_MED_ChicleVeil_FaceAcc_CtrlRig.F_MED_ChicleVeil_FaceAcc_CtrlRig_C}
   }
   Output Pose(Result=@n4) {layerGroup=None}
  ]}
 EventGraph {nodes [
   n1 = Event OnCharacterCustomizationLoaded()
   > Set BodyMesh(BodyMesh="Get (a copy)"(Array:Array of Skeletal Mesh Component Budgeted Object References=@n1.CustomMeshParts, Dimension 1=1) -> (Output:Skeletal Mesh Component Budgeted Object Reference))
   Event BlueprintUpdateAnimation()
   > n2 = Sequence!(execute) -> (then_0, then_1, then_2:Exec)
   n3 = Get Dynamic Material Instance(execute=@n2.then_2, SkeletalMeshComponent=GetPartSkeletalMeshComponent())
   > n4 = Branch(Condition=Get HidePins() -> (HidePins:bool))
   > SetScalarParameterValue(self=@n3, ParameterName=Hide Element 01, Value=1)
   SetScalarParameterValue(execute=@n4.else, self=@n3, ParameterName=Hide Element 01, Value=0)
   n5 = Gravity Override State Function DEPRECATED(
    execute=@n2.then_0
    CharacterPartAnimInstance=Self-Reference()
    GameTimeInSeconds=GetGameTimeInSeconds()
    GravityOverrideParams=Get Gravity Override Params()
   )
   > Set Gravity(Gravity:Vector=@n5) -> (Output_Get:Vector)
   n6 = Cast To FortPlayerPawn(execute=@n2.then_1, Object=TryGetPawnOwner())
   n7 = GetSkeletalMeshForPartType(self=@n6)
   > Set HidePins(HidePins:bool=MatchesWildcard(SourceString=GetDisplayName(Object=@n7), Wildcard=*F_MED_GalliumConcrete*, SearchCase=CaseSensitive)) -> (Output_Get:bool)
  ]}
]
variables [
 BodyMesh {type=SkeletalMeshComponent}
 Gravity {type=Vector}
 Gravity Override Params {type=GravityOverrideParamsStruct}
 HidePins {type=boolean}
]
